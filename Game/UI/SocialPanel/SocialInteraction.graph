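# Social panel interaction button widget, built on SocialInteractionButton, showing an action name, hover bubble with input prompt and icon, and action menu anchor.
# One OnHover animation drives slides, fades and material parameters, playing forward on hover, reversed at double speed on unhover, and reversed on construct.
# On interaction set, AddFriend recolors the indicator, and AddFriend or AcceptFriendInvite shows an Epic-friend encouragement badge for platform-only friends.
# Hover also swaps that badge's color.
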WidgetBlueprint /Game/UI/SocialPanel/SocialInteraction
animations [OnHover {
  binding_count=7
  display_label=""
  length_seconds=0.2
  object=&self:OnHover
  sequence {bindings [
    HoverBubble {
     id=#1
     possessed_class=SizeBox
     tracks [{
       type=MovieScene2DTransformTrack
       sections [{
         type=MovieScene2DTransformSection
         end_frame=4
         start_frame=0
         channels [
          Translation.X {type=MovieSceneScriptingFloatChannel, keys [[0,0],[2,8]]}
          Translation.Y {type=MovieSceneScriptingFloatChannel}
          Angle {type=MovieSceneScriptingFloatChannel}
          Scale.X {type=MovieSceneScriptingFloatChannel}
          Scale.Y {type=MovieSceneScriptingFloatChannel}
          Shear.X {type=MovieSceneScriptingFloatChannel}
          Shear.Y {type=MovieSceneScriptingFloatChannel}
         ]
        }]
      }]
    }
    InteractionIcon {
     id=#2
     possessed_class=Image
     tracks [
      {
       type=MovieScene2DTransformTrack
       sections [{
         type=MovieScene2DTransformSection
         end_frame=4
         start_frame=0
         channels [
          Translation.X {type=MovieSceneScriptingFloatChannel}
          Translation.Y {type=MovieSceneScriptingFloatChannel}
          Angle {type=MovieSceneScriptingFloatChannel}
          Scale.X {type=MovieSceneScriptingFloatChannel, keys [[0,0.875],[3,1.125],[4,1]]}
          Scale.Y {type=MovieSceneScriptingFloatChannel, keys [[0,0.875],[3,1.125],[4,1]]}
          Shear.X {type=MovieSceneScriptingFloatChannel}
          Shear.Y {type=MovieSceneScriptingFloatChannel}
         ]
        }]
      }
      {
       type=MovieSceneFloatTrack
       display_name=Render Opacity
       sections [{type=MovieSceneFloatSection, end_frame=4, start_frame=0, channels [None {type=MovieSceneScriptingFloatChannel, keys [[0,0],[2,1]]}]}]
      }
     ]
    }
    EntryBorder {
     id=#3
     possessed_class=Border
     tracks [
      {
       type=MovieSceneWidgetMaterialTrack
       sections [{
         type=MovieSceneParameterSection
         end_frame=4
         start_frame=0
         channels [AnimateHighlight {type=MovieSceneScriptingFloatChannel, keys [[0,0],[3,1],[4,1]]}]
        }]
      }
      {
       type=MovieScene2DTransformTrack
       sections [{
         type=MovieScene2DTransformSection
         end_frame=4
         start_frame=0
         channels [
          Translation.X {type=MovieSceneScriptingFloatChannel, keys [[0,0],[2,-8]]}
          Translation.Y {type=MovieSceneScriptingFloatChannel}
          Angle {type=MovieSceneScriptingFloatChannel}
          Scale.X {type=MovieSceneScriptingFloatChannel}
          Scale.Y {type=MovieSceneScriptingFloatChannel}
          Shear.X {type=MovieSceneScriptingFloatChannel}
          Shear.Y {type=MovieSceneScriptingFloatChannel}
         ]
        }]
      }
     ]
    }
    InputActionWidget {
     id=#4
     possessed_class=CommonActionWidget
     tracks [{
       type=MovieSceneFloatTrack
       display_name=Render Opacity
       sections [{type=MovieSceneFloatSection, end_frame=4, start_frame=0, channels [None {type=MovieSceneScriptingFloatChannel, keys [[0,0],[1,0],[2,1]]}]}]
      }]
    }
    AnimatedInputBorder {
     id=#5
     possessed_class=Border
     tracks [
      {
       type=MovieSceneWidgetMaterialTrack
       sections [{type=MovieSceneParameterSection, channels [AnimateStateChange {type=MovieSceneScriptingFloatChannel, keys [[0,0],[3,1],[4,1]]}]}]
      }
      {
       type=MovieSceneColorTrack
       display_name=Brush Color
       sections [{
         type=MovieSceneColorSection
         channels [
          Color.R {type=MovieSceneScriptingFloatChannel}
          Color.G {type=MovieSceneScriptingFloatChannel}
          Color.B {type=MovieSceneScriptingFloatChannel}
          Color.A {type=MovieSceneScriptingFloatChannel, keys [[0,0],[2,1]]}
         ]
        }]
      }
     ]
    }
    Text_InteractionName {
     id=#6
     possessed_class=CommonTextBlock
     tracks [
      {
       type=MovieSceneFloatTrack
       display_name=Render Opacity
       sections [{type=MovieSceneFloatSection, end_frame=4, start_frame=0, channels [None {type=MovieSceneScriptingFloatChannel, keys [[0,0.5],[4,1]]}]}]
      }
      {
       type=MovieScene2DTransformTrack
       sections [{
         type=MovieScene2DTransformSection
         end_frame=4
         start_frame=0
         channels [
          Translation.X {type=MovieSceneScriptingFloatChannel, keys [[0,0],[2,-8]]}
          Translation.Y {type=MovieSceneScriptingFloatChannel}
          Angle {type=MovieSceneScriptingFloatChannel}
          Scale.X {type=MovieSceneScriptingFloatChannel}
          Scale.Y {type=MovieSceneScriptingFloatChannel}
          Shear.X {type=MovieSceneScriptingFloatChannel}
          Shear.Y {type=MovieSceneScriptingFloatChannel}
         ]
        }]
      }
     ]
    }
    Border_InteractionIndicator {
     id=#7
     possessed_class=Border
     tracks [{
       type=MovieSceneFloatTrack
       display_name=Render Opacity
       sections [{type=MovieSceneFloatSection, end_frame=4, start_frame=0, channels [None {type=MovieSceneScriptingFloatChannel, keys [[0,0.5],[3,1],[4,1]]}]}]
      }]
    }
   ]}
 }]
graphs [EventGraph {nodes [
   Event Construct()
   > Parent: Construct()
   n1 = Get OnHover() -> (OnHover:Widget Animation Object Reference)
   > PlayAnimation(InAnimation=@n1, StartAtTime=GetEndTime(self=@n1), PlayMode=Reverse)
   Event OnInteractionSet() {is=title K2Node_Event}
   > SetVisibility(self=Get EncourageEpicFriend() -> (EncourageEpicFriend:Image Object Reference), InVisibility=Collapsed)
   n2 = GetInteractionName!(self:Social Interaction Button Object Reference) -> (ReturnValue:Name=None)
   n3 = "Equal (Name)"!(A:Name=@n2, B:Name=AddFriend) -> (ReturnValue:bool) {is=title K2Node_CallFunction}
   > n4 = Branch(Condition=@n3)
   > n5 = SetBrushColor(
    self=Get Border_InteractionIndicator() -> (Border_InteractionIndicator:Border Object Reference)
    InBrushColor=Get EncourageEpicFriend_IdleColor() -> (EncourageEpicFriend_IdleColor:Linear Color Structure)
   )
   Branch(execute=[@n5.then, @n4.else], Condition=OR Boolean(A=@n3, B="Equal (Name)"(A=@n2, B=AcceptFriendInvite)))
   > SetVisibility(
    self=Get EncourageEpicFriend()
    InVisibility=Select(
     Option 0=Collapsed
     Option 1=SelfHitTestInvisible
     Index=IsPlatformOnlyFriend!(self:Social Interaction Button Object Reference) -> (ReturnValue:bool)
    )
   )
   Event On Hovered()
   > PlayAnimation(InAnimation=Get OnHover())
   > SetColorAndOpacity(
    self:Image Object Reference=Get EncourageEpicFriend()
    InColorAndOpacity:Linear Color Structure=Get EncourageEpicFriend_HoveredColor() -> (EncourageEpicFriend_HoveredColor:Linear Color Structure)
   )
   Event On Unhovered()
   > PlayAnimation(InAnimation=Get OnHover(), PlayMode=Reverse, PlaybackSpeed=2)
   > SetColorAndOpacity(self=Get EncourageEpicFriend(), InColorAndOpacity=Get EncourageEpicFriend_IdleColor())
  ]}]
variables [
 AddFriendColor {type=LinearColor, default=<rgba 0 0 0 0>}
 EncourageEpicFriend_HoveredColor {type=LinearColor, default=<rgba 1 1 1 1>}
 EncourageEpicFriend_IdleColor {type=LinearColor, default=<rgba 0.9558 1 0.0672 1>}
]
widgets {
 info {inheritedWidgetCount=3, parentClass=/Script/SocialUMG.SocialInteractionButton, rootWidgetClass=&SizeBox, widgetCount=17}
 widgets [
  SizeBox_0 {widgetClassPath=&SizeBox, widget=SizeBox_0, properties {bOverride_MinDesiredHeight=true, heightOverride=48, minDesiredHeight=48}}
  EntryBorder {
   widgetClassPath=&Border
   parent=SizeBox_0
   widget=EntryBorder
   properties {
    visibility=SelfHitTestInvisible
    background {
     bIsDynamicallyLoaded=false
     drawAs=Image
     imageType=NoImage
     mirroring=NoMirror
     resourceName=None
     resourceObject=/Game/UI/SocialPanel/Art/Materials/M_UI_SocialUserItemBorder_Interaction
     tiling=NoTile
     imageSize=<32 32>
     margin=<ltrb 0 0 0 0>
     outlineSettings {
      bUseBrushTransparency=false
      roundingType=HalfHeightRadius
      width=0
      color {colorUseRule=UseColor_Specified, specifiedColor=<rgba 0 0 0 0>}
      cornerRadii=<0 0 0 0>
     }
     tintColor {colorUseRule=UseColor_Specified, specifiedColor=<rgba 1 1 1 1>}
     uVRegion {bIsValid=false, max=<0 0>, min=<0 0>}
    }
    padding=<ltrb 0 0 0 0>
   }
   slot {class=SizeBoxSlot, properties {padding=<ltrb 8 0 8 0>}}
  }
  HorizontalBox_2 {
   widgetClassPath=&HorizontalBox
   parent=EntryBorder
   widget=HorizontalBox_2
   slot {class=BorderSlot, properties {padding=<ltrb 0 0 0 0>}}
  }
  Border_InteractionIndicator {
   widgetClassPath=&Border
   bInherited=true
   bIsVariable=true
   parent=HorizontalBox_2
   widget=Border_InteractionIndicator
   properties {renderOpacity=0.5, brushColor=<rgba 0.4287 0.6654 0.7991 1>, padding=<ltrb 0 0 0 0>}
   slot {class=HorizontalBoxSlot}
  }
  Spacer_0 {
   widgetClassPath=&Spacer
   parent=Border_InteractionIndicator
   widget=Spacer_0
   properties {size=<8 8>}
   slot {class=BorderSlot, properties {padding=<ltrb 0 0 0 0>}}
  }
  SizeBox_3 {
   widgetClassPath=&SizeBox
   parent=HorizontalBox_2
   widget=SizeBox_3
   properties {bOverride_MinDesiredHeight=true, heightOverride=24, minDesiredHeight=24}
   slot {
    class=HorizontalBoxSlot
    properties {horizontalAlignment=HAlign_Left, verticalAlignment=VAlign_Center, padding=<ltrb 24 0 0 0>, size {sizeRule=Fill, value=1}}
   }
  }
  Text_InteractionName {
   widgetClassPath=&CommonTextBlock
   bInherited=true
   bIsVariable=true
   parent=SizeBox_3
   widget=Text_InteractionName
   properties {
    mobileFontSizeMultiplier=1.2
    renderOpacity=0.5
    text="{INTERACTION NAME}"
    font {
     bForceMonospaced=false
     bMaterialIsStencil=false
     fontMaterial=null
     fontObject=&BurbankSmall
     letterSpacing=0
     monospacedWidth=1
     size=18
     skewAmount=0
     typefaceFontName=Black
     outlineSettings {
      bApplyOutlineToDropShadows=false
      bMiteredCorners=false
      bSeparateFillAlpha=false
      outlineMaterial=null
      outlineSize=0
      outlineColor=<rgba 0 0 0 1>
     }
    }
   }
   slot {class=SizeBoxSlot, properties {verticalAlignment=VAlign_Center, padding=<ltrb 0 -4 0 -4>}}
  }
  HoverBubble {
   widgetClassPath=&SizeBox
   parent=HorizontalBox_2
   widget=HoverBubble
   properties {bOverride_MinDesiredWidth=true, minDesiredWidth=80}
   slot {class=HorizontalBoxSlot, properties {horizontalAlignment=HAlign_Right}}
  }
  Overlay_1 {widgetClassPath=&Overlay, parent=HoverBubble, widget=Overlay_1, slot {class=SizeBoxSlot}}
  AnimatedInputBorder {
   widgetClassPath=&Border
   parent=Overlay_1
   widget=AnimatedInputBorder
   properties {
    visibility=SelfHitTestInvisible
    background {
     bIsDynamicallyLoaded=false
     drawAs=Image
     imageType=NoImage
     mirroring=NoMirror
     resourceName=None
     resourceObject=/Game/UI/SocialPanel/Art/Materials/M_UI_UserHover_Inst
     tiling=NoTile
     imageSize=<64 64>
     margin=<ltrb 1 0 0 0>
     outlineSettings {
      bUseBrushTransparency=false
      roundingType=HalfHeightRadius
      width=0
      color {colorUseRule=UseColor_Specified, specifiedColor=<rgba 0 0 0 0>}
      cornerRadii=<0 0 0 0>
     }
     tintColor {colorUseRule=UseColor_Specified, specifiedColor=<rgba 1 1 1 1>}
     uVRegion {bIsValid=false, max=<0 0>, min=<0 0>}
    }
    padding=<ltrb 0 0 0 0>
   }
   slot {class=OverlaySlot, properties {horizontalAlignment=HAlign_Right, verticalAlignment=VAlign_Fill}}
  }
  HorizontalBox_6 {
   widgetClassPath=&HorizontalBox
   parent=AnimatedInputBorder
   widget=HorizontalBox_6
   slot {class=BorderSlot, properties {padding=<ltrb 0 0 0 0>}}
  }
  InputActionWidget {
   widgetClassPath=&CommonActionWidget
   bInherited=true
   parent=HorizontalBox_6
   widget=InputActionWidget
   properties {visibility=SelfHitTestInvisible, inputActions [{dataTable=&InputActionDataTable, rowName=Select}]}
   slot {class=HorizontalBoxSlot, properties {horizontalAlignment=HAlign_Left, verticalAlignment=VAlign_Center, padding=<ltrb 8 0 -16 0>}}
  }
  Spacer_7 {
   widgetClassPath=&Spacer
   parent=HorizontalBox_6
   widget=Spacer_7
   properties {size=<8 8>}
   slot {class=HorizontalBoxSlot, properties {size {sizeRule=Fill, value=1}}}
  }
  InteractionIconSB {
   widgetClassPath=&SizeBox
   parent=HorizontalBox_6
   widget=InteractionIconSB
   properties {bOverride_HeightOverride=true, bOverride_WidthOverride=true, heightOverride=56, widthOverride=56}
   slot {class=HorizontalBoxSlot, properties {horizontalAlignment=HAlign_Right, verticalAlignment=VAlign_Center}}
  }
  InteractionIcon {
   widgetClassPath=&Image
   bIsVariable=true
   parent=InteractionIconSB
   widget=InteractionIcon
   properties {
    renderOpacity=0
    visibility=SelfHitTestInvisible
    brush {
     bIsDynamicallyLoaded=false
     drawAs=Image
     imageType=NoImage
     mirroring=NoMirror
     resourceName=None
     resourceObject=/Game/UI/Foundation/Textures/Icons/Checks/T-Icon-Check-64
     tiling=NoTile
     imageSize=<32 32>
     margin=<ltrb 0 0 0 0>
     outlineSettings {
      bUseBrushTransparency=false
      roundingType=HalfHeightRadius
      width=0
      color {colorUseRule=UseColor_Specified, specifiedColor=<rgba 0 0 0 0>}
      cornerRadii=<0 0 0 0>
     }
     tintColor {colorUseRule=UseColor_Specified, specifiedColor=<rgba 1 1 1 1>}
     uVRegion {bIsValid=false, max=<0 0>, min=<0 0>}
    }
   }
   slot {class=SizeBoxSlot, properties {horizontalAlignment=HAlign_Center, verticalAlignment=VAlign_Center}}
  }
  MenuAnchor_Actions {
   widgetClassPath=/Script/UMG.MenuAnchor
   bIsVariable=true
   parent=Overlay_1
   widget=MenuAnchor_Actions
   properties {menuClass=&/UserActionMenu.UserActionMenu_C, placement=MenuPlacement_RightLeftCenter}
   slot {class=OverlaySlot, properties {horizontalAlignment=HAlign_Right, verticalAlignment=VAlign_Center}}
  }
  EncourageEpicFriend {
   widgetClassPath=&Image
   bIsVariable=true
   parent=Overlay_1
   widget=EncourageEpicFriend
   properties {
    bFlipForRightToLeftFlowDirection=true
    visibility=Collapsed
    brush {
     bIsDynamicallyLoaded=false
     drawAs=Image
     imageType=NoImage
     mirroring=NoMirror
     resourceName=None
     resourceObject=/Game/UI/SocialPanel/Art/Materials/M_UI_EpicFriendOpportunity_Badge_UserEntry
     tiling=NoTile
     imageSize=<32 32>
     margin=<ltrb 0 0 0 0>
     outlineSettings {
      bUseBrushTransparency=false
      roundingType=HalfHeightRadius
      width=0
      color {colorUseRule=UseColor_Specified, specifiedColor=<rgba 0 0 0 0>}
      cornerRadii=<0 0 0 0>
     }
     tintColor {colorUseRule=UseColor_Specified, specifiedColor=<rgba 1 1 1 1>}
     uVRegion {bIsValid=false, max=<0 0>, min=<0 0>}
    }
    colorAndOpacity=<rgba 1 0.9552 0 1>
   }
   slot {class=OverlaySlot, properties {horizontalAlignment=HAlign_Right}}
  }
 ]
}
