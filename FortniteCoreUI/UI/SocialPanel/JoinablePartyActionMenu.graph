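# Joinable-party action menu widget for the social panel, built on the native FortJoinablePartyActionMenu parent.
# A header row shows party display names, above a switcher holding positive and negative SocialInteraction entry boxes or a yes/no confirmation prompt.
# The event graph only plays animations: Opened slides and fades in the header and interactions, and ConfirmationOpened animates the prompt when it becomes visible.
# Several inherited bindings, such as portraits and the member list, have no widget here.
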WidgetBlueprint /FortniteCoreUI/UI/SocialPanel/JoinablePartyActionMenu
animations [
 Opened {
  binding_count=3
  display_label=""
  length_seconds=0.3
  object=&self:Opened
  sequence {bindings [
    UserCellBorder {
     id=#1
     possessed_class=SizeBox
     tracks [
      {
       type=MovieSceneFloatTrack
       display_name=Render Opacity
       sections [{
         type=MovieSceneFloatSection
         end_frame=6
         start_frame=0
         channels [None {type=MovieSceneScriptingFloatChannel, thinned_from=4, keys [[0,0],[3,1],[6,1]]}]
        }]
      }
      {
       type=MovieScene2DTransformTrack
       sections [{
         type=MovieScene2DTransformSection
         end_frame=6
         start_frame=0
         channels [
          Translation.X {type=MovieSceneScriptingFloatChannel, thinned_from=4, keys [[0,-256],[3,0],[6,0]]}
          Translation.Y {type=MovieSceneScriptingFloatChannel}
          Angle {type=MovieSceneScriptingFloatChannel}
          Scale.X {type=MovieSceneScriptingFloatChannel}
          Scale.Y {type=MovieSceneScriptingFloatChannel}
          Shear.X {type=MovieSceneScriptingFloatChannel, thinned_from=4, keys [[0,16],[3,0],[6,0]]}
          Shear.Y {type=MovieSceneScriptingFloatChannel}
         ]
        }]
      }
     ]
    }
    InteractionsBorder {
     id=#2
     possessed_class=Border
     tracks [
      {
       type=MovieScene2DTransformTrack
       sections [{
         type=MovieScene2DTransformSection
         channels [
          Translation.X {type=MovieSceneScriptingFloatChannel}
          Translation.Y {type=MovieSceneScriptingFloatChannel, keys [[0,0],[1,-72],[3,0],[4,0]]}
          Angle {type=MovieSceneScriptingFloatChannel}
          Scale.X {type=MovieSceneScriptingFloatChannel}
          Scale.Y {type=MovieSceneScriptingFloatChannel}
          Shear.X {type=MovieSceneScriptingFloatChannel}
          Shear.Y {type=MovieSceneScriptingFloatChannel}
         ]
        }]
      }
      {
       type=MovieSceneFloatTrack
       display_name=Render Opacity
       sections [{
         type=MovieSceneFloatSection
         end_frame=6
         start_frame=0
         channels [None {type=MovieSceneScriptingFloatChannel, keys [[0,0],[1,0],[3,1],[6,1]]}]
        }]
      }
     ]
    }
    EntryBorder {
     id=#3
     possessed_class=Border
     tracks [{
       type=MovieSceneWidgetMaterialTrack
       sections [{type=MovieSceneParameterSection, channels [AnimateHighlight {type=MovieSceneScriptingFloatChannel, keys [[0,0],[4,1],[6,1]]}]}]
      }]
    }
   ]}
 }
 ConfirmationOpened {
  binding_count=1
  display_label=""
  length_seconds=0.3
  object=&self:ConfirmationOpened
  sequence {bindings [Border_Confirmation {
     id=#4
     possessed_class=Border
     tracks [
      {
       type=MovieScene2DTransformTrack
       sections [{
         type=MovieScene2DTransformSection
         channels [
          Translation.X {type=MovieSceneScriptingFloatChannel}
          Translation.Y {type=MovieSceneScriptingFloatChannel, keys [[0,0],[1,-72],[3,0],[4,0]]}
          Angle {type=MovieSceneScriptingFloatChannel}
          Scale.X {type=MovieSceneScriptingFloatChannel}
          Scale.Y {type=MovieSceneScriptingFloatChannel}
          Shear.X {type=MovieSceneScriptingFloatChannel}
          Shear.Y {type=MovieSceneScriptingFloatChannel}
         ]
        }]
      }
      {
       type=MovieSceneFloatTrack
       display_name=Render Opacity
       sections [{
         type=MovieSceneFloatSection
         end_frame=6
         start_frame=0
         channels [None {type=MovieSceneScriptingFloatChannel, keys [[0,0],[1,0],[3,1],[6,1]]}]
        }]
      }
     ]
    }]}
 }
]
graphs [EventGraph {nodes [
   n1 = Event OnToggleConfirmation!() -> (OutputDelegate, then, bIsVisible:bool) {is=title K2Node_Event}
   > Branch(Condition=@n1.bIsVisible)
   > PlayAnimation(InAnimation=Get ConfirmationOpened() -> (ConfirmationOpened:Widget Animation Object Reference))
   Event OnOpened() {is=title K2Node_Event}
   > PlayAnimation(InAnimation=Get Opened() -> (Opened:Widget Animation Object Reference))
  ]}]
variables []
widgets {
 info {inheritedWidgetCount=13, parentClass=/Script/FortniteUI.FortJoinablePartyActionMenu, rootWidgetClass=&Overlay, widgetCount=30}
 widgets [
  Overlay_0 {widgetClassPath=&Overlay, widget=Overlay_0}
  SizeBox_0 {
   widgetClassPath=&SizeBox
   parent=Overlay_0
   widget=SizeBox_0
   properties {bOverride_WidthOverride=true, widthOverride=520}
   slot {class=OverlaySlot}
  }
  VB_Content {widgetClassPath=&VerticalBox, parent=SizeBox_0, widget=VB_Content, slot {class=SizeBoxSlot}}
  Overlay_1 {widgetClassPath=&Overlay, parent=VB_Content, widget=Overlay_1, slot {class=VerticalBoxSlot}}
  UserCellBorder {
   widgetClassPath=&SizeBox
   parent=Overlay_1
   widget=UserCellBorder
   properties {bOverride_MinDesiredHeight=true, minDesiredHeight=72}
   slot {class=OverlaySlot, properties {horizontalAlignment=HAlign_Fill, verticalAlignment=VAlign_Fill}}
  }
  Border_4 {
   widgetClassPath=&Border
   parent=UserCellBorder
   widget=Border_4
   properties {brushColor=<rgba 0 0.0168 0.0887 1>, padding=<ltrb 8 0 8 0>}
   slot {class=SizeBoxSlot}
  }
  EntryBorder {
   widgetClassPath=&Border
   bIsVariable=true
   parent=Border_4
   widget=EntryBorder
   properties {
    visibility=SelfHitTestInvisible
    background {
     bIsDynamicallyLoaded=false
     drawAs=Image
     imageType=NoImage
     mirroring=NoMirror
     resourceName=None
     resourceObject=/Game/UI/SocialPanel/Art/Materials/M_UI_SocialUserItemBorder_UserListItem
     tiling=NoTile
     imageSize=<32 32>
     margin=<ltrb 0 0 0 0>
     outlineSettings {
      bUseBrushTransparency=false
      roundingType=HalfHeightRadius
      width=0
      color {colorUseRule=UseColor_Specified, specifiedColor=<rgba 0 0 0 0>}
      cornerRadii=<0 0 0 0>
     }
     tintColor {colorUseRule=UseColor_Specified, specifiedColor=<rgba 1 1 1 1>}
     uVRegion {bIsValid=false, max=<0 0>, min=<0 0>}
    }
    padding=<ltrb 0 0 0 0>
   }
   slot {class=BorderSlot, properties {padding=<ltrb 8 0 8 0>}}
  }
  HorizontalBox_2 {
   widgetClassPath=&HorizontalBox
   parent=EntryBorder
   widget=HorizontalBox_2
   properties {flowDirectionPreference=Culture}
   slot {class=BorderSlot, properties {padding=<ltrb 0 0 0 0>}}
  }
  Border_OnlineStatusIndicator {
   widgetClassPath=&Border
   bIsVariable=true
   parent=HorizontalBox_2
   widget=Border_OnlineStatusIndicator
   properties {brushColor=<rgba 0.4287 0.6654 0.7991 0.6>, padding=<ltrb 0 0 0 0>}
   slot {class=HorizontalBoxSlot}
  }
  Spacer_0 {
   widgetClassPath=&Spacer
   parent=Border_OnlineStatusIndicator
   widget=Spacer_0
   properties {size=<8 8>}
   slot {class=BorderSlot, properties {padding=<ltrb 0 0 0 0>}}
  }
  UserTextInfoVB {
   widgetClassPath=&VerticalBox
   parent=HorizontalBox_2
   widget=UserTextInfoVB
   slot {class=HorizontalBoxSlot, properties {verticalAlignment=VAlign_Center, padding=<ltrb 24 20 20 20>, size {sizeRule=Fill, value=1}}}
  }
  SizeBox_2 {
   widgetClassPath=&SizeBox
   parent=UserTextInfoVB
   widget=SizeBox_2
   properties {
    bOverride_MaxDesiredWidth=true
    bOverride_MinDesiredHeight=true
    heightOverride=24
    maxDesiredWidth=326
    minDesiredHeight=24
    widthOverride=326
   }
   slot {class=VerticalBoxSlot, properties {size {sizeRule=Fill, value=1}}}
  }
  RichText_DisplayNames {
   widgetClassPath=&CommonRichTextBlock
   bInherited=true
   bIsVariable=true
   parent=SizeBox_2
   widget=RichText_DisplayNames
   properties {
    bOverrideDefaultStyle=true
    bTintInlineIcon=true
    lineHeightPercentage=1.2
    mobileTextBlockScale=1.34
    scrollStyle=/FortUILibrary/Styles/Typography/Foundation/TextScrollStyle-NoFade.TextScrollStyle-NoFade_C
    text="User Name<PC/>\r\nUser Name<PC/>\r\nUser Name<PC/>"
    defaultTextStyleOverride {
     overflowPolicy=Clip
     transformPolicy=None
     colorAndOpacity {colorUseRule=UseColor_Specified, specifiedColor=<rgba 0.521 1 0.8228 1>}
     font {
      bForceMonospaced=false
      bMaterialIsStencil=false
      fontMaterial=null
      fontObject=&BurbankSmall
      letterSpacing=0
      monospacedWidth=1
      size=16
      skewAmount=0
      typefaceFontName=Black
      outlineSettings {
       bApplyOutlineToDropShadows=false
       bMiteredCorners=false
       bSeparateFillAlpha=false
       outlineMaterial=null
       outlineSize=0
       outlineColor=<rgba 0 0 0 1>
      }
     }
     highlightColor {colorUseRule=UseColor_Specified, specifiedColor=<rgba 0 0 0 1>}
     highlightShape {
      bIsDynamicallyLoaded=false
      drawAs=Image
      imageType=NoImage
      mirroring=NoMirror
      resourceName=None
      resourceObject=null
      tiling=NoTile
      imageSize=<32 32>
      margin=<ltrb 0 0 0 0>
      outlineSettings {
       bUseBrushTransparency=false
       roundingType=HalfHeightRadius
       width=0
       color {colorUseRule=UseColor_Specified, specifiedColor=<rgba 0 0 0 0>}
       cornerRadii=<0 0 0 0>
      }
      tintColor {colorUseRule=UseColor_Specified, specifiedColor=<rgba 1 1 1 1>}
      uVRegion {bIsValid=false, max=<0 0>, min=<0 0>}
     }
     selectedBackgroundColor {colorUseRule=UseColor_Foreground, specifiedColor=<rgba 1 0 1 1>}
     shadowColorAndOpacity=<rgba 0 0 0 1>
     shadowOffset=<0 0>
     strikeBrush {
      bIsDynamicallyLoaded=false
      drawAs=Image
      imageType=NoImage
      mirroring=NoMirror
      resourceName=None
      resourceObject=null
      tiling=NoTile
      imageSize=<32 32>
      margin=<ltrb 0 0 0 0>
      outlineSettings {
       bUseBrushTransparency=false
       roundingType=HalfHeightRadius
       width=0
       color {colorUseRule=UseColor_Specified, specifiedColor=<rgba 0 0 0 0>}
       cornerRadii=<0 0 0 0>
      }
      tintColor {colorUseRule=UseColor_Specified, specifiedColor=<rgba 1 1 1 1>}
      uVRegion {bIsValid=false, max=<0 0>, min=<0 0>}
     }
     underlineBrush {
      bIsDynamicallyLoaded=false
      drawAs=Image
      imageType=NoImage
      mirroring=NoMirror
      resourceName=None
      resourceObject=null
      tiling=NoTile
      imageSize=<32 32>
      margin=<ltrb 0 0 0 0>
      outlineSettings {
       bUseBrushTransparency=false
       roundingType=HalfHeightRadius
       width=0
       color {colorUseRule=UseColor_Specified, specifiedColor=<rgba 0 0 0 0>}
       cornerRadii=<0 0 0 0>
      }
      tintColor {colorUseRule=UseColor_Specified, specifiedColor=<rgba 1 1 1 1>}
      uVRegion {bIsValid=false, max=<0 0>, min=<0 0>}
     }
    }
   }
   slot {class=SizeBoxSlot}
  }
  Switcher_Confirmation {
   widgetClassPath=&CommonWidgetSwitcherLegacy
   bInherited=true
   bIsVariable=true
   parent=VB_Content
   widget=Switcher_Confirmation
   properties {transitionDuration=0.2}
   slot {class=VerticalBoxSlot}
  }
  InteractionsBorder {
   widgetClassPath=&Border
   parent=Switcher_Confirmation
   widget=InteractionsBorder
   properties {brushColor=<rgba 0 0.0168 0.0887 1>, padding=<ltrb 0 4 0 0>}
   slot {class=WidgetSwitcherSlot}
  }
  VerticalBox_1 {
   widgetClassPath=&VerticalBox
   parent=InteractionsBorder
   widget=VerticalBox_1
   slot {class=BorderSlot, properties {padding=<ltrb 0 4 0 0>}}
  }
  EntryBox_PositiveInteractions {
   widgetClassPath=&DynamicEntryBox
   bInherited=true
   bIsVariable=true
   parent=VerticalBox_1
   widget=EntryBox_PositiveInteractions
   properties {
    entryBoxType=Vertical
    entryWidgetClass=/Game/UI/SocialPanel/SocialInteraction.SocialInteraction_C
    numDesignerPreviewEntries=1
    entrySpacing=<0 4>
   }
   slot {class=VerticalBoxSlot}
  }
  Spacer_InteractionSeparator {
   widgetClassPath=&Spacer
   bInherited=true
   parent=VerticalBox_1
   widget=Spacer_InteractionSeparator
   properties {size=<1 20>}
   slot {class=VerticalBoxSlot}
  }
  EntryBox_NegativeInteractions {
   widgetClassPath=&DynamicEntryBox
   bInherited=true
   bIsVariable=true
   parent=VerticalBox_1
   widget=EntryBox_NegativeInteractions
   properties {
    entryBoxType=Vertical
    entryWidgetClass=&SocialInteraction_C
    numDesignerPreviewEntries=1
    entrySpacing=<0 4>
   }
   slot {class=VerticalBoxSlot, properties {padding=<ltrb 0 0 0 4>}}
  }
  Border_Confirmation {
   widgetClassPath=&Border
   parent=Switcher_Confirmation
   widget=Border_Confirmation
   properties {
    navigation=&self:WidgetTree.Border_Confirmation.WidgetNavigation_0
    verticalAlignment=VAlign_Center
    visibility=SelfHitTestInvisible
    brushColor=<rgba 0 0.0168 0.0887 1>
    padding=<ltrb 0 8 0 8>
    renderTransformPivot=<0 0.5>
   }
   slot {class=WidgetSwitcherSlot}
  }
  VerticalBox_0 {
   widgetClassPath=&VerticalBox
   parent=Border_Confirmation
   widget=VerticalBox_0
   slot {class=BorderSlot, properties {verticalAlignment=VAlign_Center, padding=<ltrb 0 8 0 8>}}
  }
  Text_ConfirmationLabel {
   widgetClassPath=&CommonTextBlock
   bInherited=true
   parent=VerticalBox_0
   widget=Text_ConfirmationLabel
   properties {
    justification=Center
    renderOpacity=0.3
    scrollStyle=&TextScrollStyle-Base_C
    text=CONFIRM ACTION?
    font {
     bForceMonospaced=false
     bMaterialIsStencil=false
     fontMaterial=null
     fontObject=&BurbankSmall
     letterSpacing=0
     monospacedWidth=1
     size=24
     skewAmount=0
     typefaceFontName=NotoSans-Regular
     outlineSettings {
      bApplyOutlineToDropShadows=false
      bMiteredCorners=false
      bSeparateFillAlpha=false
      outlineMaterial=null
      outlineSize=0
      outlineColor=<rgba 0 0 0 1>
     }
    }
   }
   slot {class=VerticalBoxSlot, properties {horizontalAlignment=HAlign_Center, padding=<ltrb 0 8 0 -6>}}
  }
  HorizontalBox_0 {
   widgetClassPath=&HorizontalBox
   parent=VerticalBox_0
   widget=HorizontalBox_0
   properties {flowDirectionPreference=Culture}
   slot {class=VerticalBoxSlot, properties {padding=<ltrb 0 16 0 0>}}
  }
  Button_Confirm {
   widgetClassPath=/Game/UI/SocialPanel/SocialConfirmationButton
   bInherited=true
   bIsVariable=true
   parent=HorizontalBox_0
   widget=Button_Confirm
   properties {
    bIsAccept=true
    displayText=YES
    navigation=&self:WidgetTree.Button_Confirm.WidgetNavigation_0
    onButtonBaseClicked=(&self:WidgetTree.Button_Confirm.HandleOnButtonClicked)
    onButtonBaseDoubleClicked=(&self:WidgetTree.Button_Confirm.HandleOnButtonDoubleClicked)
    onButtonBaseHovered=(&self:WidgetTree.Button_Confirm.HandleOnButtonHovered)
    onButtonBaseUnhovered=(&self:WidgetTree.Button_Confirm.HandleOnButtonUnhovered)
    onSelectedChangedBase=(&self:WidgetTree.Button_Confirm.HandleOnSelectedChanged)
   }
   slot {class=HorizontalBoxSlot, properties {padding=<ltrb 0 0 -4 0>, size {sizeRule=Fill, value=1}}}
  }
  Button_Decline {
   widgetClassPath=&SocialConfirmationButton
   bInherited=true
   bIsVariable=true
   parent=HorizontalBox_0
   widget=Button_Decline
   properties {
    displayText=NO
    navigation=&self:WidgetTree.Button_Decline.WidgetNavigation_0
    onButtonBaseClicked=(&self:WidgetTree.Button_Decline.HandleOnButtonClicked)
    onButtonBaseDoubleClicked=(&self:WidgetTree.Button_Decline.HandleOnButtonDoubleClicked)
    onButtonBaseHovered=(&self:WidgetTree.Button_Decline.HandleOnButtonHovered)
    onButtonBaseUnhovered=(&self:WidgetTree.Button_Decline.HandleOnButtonUnhovered)
    onSelectedChangedBase=(&self:WidgetTree.Button_Decline.HandleOnSelectedChanged)
   }
   slot {class=HorizontalBoxSlot, properties {padding=<ltrb -4 0 0 0>, size {sizeRule=Fill, value=1}}}
  }
  PartyPortraits {widgetClassPath=/Script/FortniteUI.FortJoinablePartyPortraitsDisplay, bInherited=true, widget=null}
  Text_PartyPresence {widgetClassPath=&CommonRichTextBlock, bInherited=true, widget=null}
  EntryBox_PartyMembers {widgetClassPath=&DynamicEntryBox, bInherited=true, widget=null}
  ScrollBox_PartyMembers {widgetClassPath=&ScrollBox, bInherited=true, widget=null}
  SidebarBackButton {widgetClassPath=&CommonButtonBase, bInherited=true, widget=null}
 ]
}
