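# Leave-party confirmation menu widget for the social panel, built on the native FortLeavePartyMenu class.
# A fixed-width bordered panel shows a LEAVE PARTY label above side-by-side YES and NO SocialConfirmationButtons, both inherited bindings from the parent class.
# The only graph logic plays the Opened animation on OnOpened, which fades the border in while it slides from the left and straightens from a shear.
# No confirm or decline handling appears in this Blueprint.
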WidgetBlueprint /FortniteCoreUI/UI/SocialPanel/LeavePartyConfirmationMenu
animations [Opened {
  binding_count=1
  display_label=""
  length_seconds=0.3
  object=&self:Opened
  sequence {bindings [Border_Confirmation {
     id=#1
     possessed_class=Border
     tracks [
      {
       type=MovieSceneFloatTrack
       display_name=Render Opacity
       sections [{
         type=MovieSceneFloatSection
         end_frame=6
         start_frame=0
         channels [None {type=MovieSceneScriptingFloatChannel, thinned_from=4, keys [[0,0],[3,1],[6,1]]}]
        }]
      }
      {
       type=MovieScene2DTransformTrack
       sections [{
         type=MovieScene2DTransformSection
         end_frame=6
         start_frame=0
         channels [
          Translation.X {type=MovieSceneScriptingFloatChannel, thinned_from=4, keys [[0,-256],[3,0],[6,0]]}
          Translation.Y {type=MovieSceneScriptingFloatChannel}
          Angle {type=MovieSceneScriptingFloatChannel}
          Scale.X {type=MovieSceneScriptingFloatChannel}
          Scale.Y {type=MovieSceneScriptingFloatChannel}
          Shear.X {type=MovieSceneScriptingFloatChannel, thinned_from=4, keys [[0,16],[3,0],[6,0]]}
          Shear.Y {type=MovieSceneScriptingFloatChannel}
         ]
        }]
      }
     ]
    }]}
 }]
graphs [EventGraph {nodes [
   Event OnOpened() {is=title K2Node_Event}
   > PlayAnimation(InAnimation=Get Opened() -> (Opened:Widget Animation Object Reference))
  ]}]
variables []
widgets {
 info {inheritedWidgetCount=2, parentClass=/Script/FortniteUI.FortLeavePartyMenu, rootWidgetClass=&Overlay, widgetCount=9}
 widgets [
  Overlay_0 {widgetClassPath=&Overlay, widget=Overlay_0}
  SizeBox_0 {
   widgetClassPath=&SizeBox
   parent=Overlay_0
   widget=SizeBox_0
   properties {bOverride_WidthOverride=true, widthOverride=520}
   slot {class=OverlaySlot}
  }
  VB_Content {widgetClassPath=&VerticalBox, parent=SizeBox_0, widget=VB_Content, slot {class=SizeBoxSlot}}
  Border_Confirmation {
   widgetClassPath=&Border
   parent=VB_Content
   widget=Border_Confirmation
   properties {
    navigation=&self:WidgetTree.Border_Confirmation.WidgetNavigation_0
    verticalAlignment=VAlign_Center
    visibility=SelfHitTestInvisible
    brushColor=<rgba 0 0.0168 0.0887 1>
    padding=<ltrb 0 8 0 8>
    renderTransformPivot=<0 0.5>
   }
   slot {class=VerticalBoxSlot}
  }
  VerticalBox_0 {
   widgetClassPath=&VerticalBox
   parent=Border_Confirmation
   widget=VerticalBox_0
   slot {class=BorderSlot, properties {verticalAlignment=VAlign_Center, padding=<ltrb 0 8 0 8>}}
  }
  Text_ConfirmationLabel {
   widgetClassPath=&CommonTextBlock
   parent=VerticalBox_0
   widget=Text_ConfirmationLabel
   properties {
    justification=Center
    renderOpacity=0.3
    scrollStyle=&TextScrollStyle-Base_C
    text=LEAVE PARTY
    font {
     bForceMonospaced=false
     bMaterialIsStencil=false
     fontMaterial=null
     fontObject=&BurbankSmall
     letterSpacing=0
     monospacedWidth=1
     size=24
     skewAmount=0
     typefaceFontName=NotoSans-Regular
     outlineSettings {
      bApplyOutlineToDropShadows=false
      bMiteredCorners=false
      bSeparateFillAlpha=false
      outlineMaterial=null
      outlineSize=0
      outlineColor=<rgba 0 0 0 1>
     }
    }
   }
   slot {class=VerticalBoxSlot, properties {horizontalAlignment=HAlign_Center, padding=<ltrb 0 8 0 -6>}}
  }
  HorizontalBox_0 {
   widgetClassPath=&HorizontalBox
   parent=VerticalBox_0
   widget=HorizontalBox_0
   properties {flowDirectionPreference=Culture}
   slot {class=VerticalBoxSlot, properties {padding=<ltrb 0 16 0 0>}}
  }
  Button_Confirm {
   widgetClassPath=/Game/UI/SocialPanel/SocialConfirmationButton
   bInherited=true
   bIsVariable=true
   parent=HorizontalBox_0
   widget=Button_Confirm
   properties {
    bIsAccept=true
    displayText=YES
    navigation=&self:WidgetTree.Button_Confirm.WidgetNavigation_0
    onButtonBaseClicked=(&self:WidgetTree.Button_Confirm.HandleOnButtonClicked)
    onButtonBaseDoubleClicked=(&self:WidgetTree.Button_Confirm.HandleOnButtonDoubleClicked)
    onButtonBaseHovered=(&self:WidgetTree.Button_Confirm.HandleOnButtonHovered)
    onButtonBaseUnhovered=(&self:WidgetTree.Button_Confirm.HandleOnButtonUnhovered)
    onSelectedChangedBase=(&self:WidgetTree.Button_Confirm.HandleOnSelectedChanged)
   }
   slot {class=HorizontalBoxSlot, properties {padding=<ltrb 0 0 -4 0>, size {sizeRule=Fill, value=1}}}
  }
  Button_Decline {
   widgetClassPath=&SocialConfirmationButton
   bInherited=true
   bIsVariable=true
   parent=HorizontalBox_0
   widget=Button_Decline
   properties {
    displayText=NO
    navigation=&self:WidgetTree.Button_Decline.WidgetNavigation_0
    onButtonBaseClicked=(&self:WidgetTree.Button_Decline.HandleOnButtonClicked)
    onButtonBaseDoubleClicked=(&self:WidgetTree.Button_Decline.HandleOnButtonDoubleClicked)
    onButtonBaseHovered=(&self:WidgetTree.Button_Decline.HandleOnButtonHovered)
    onButtonBaseUnhovered=(&self:WidgetTree.Button_Decline.HandleOnButtonUnhovered)
    onSelectedChangedBase=(&self:WidgetTree.Button_Decline.HandleOnSelectedChanged)
   }
   slot {class=HorizontalBoxSlot, properties {padding=<ltrb -4 0 0 0>, size {sizeRule=Fill, value=1}}}
  }
 ]
}
