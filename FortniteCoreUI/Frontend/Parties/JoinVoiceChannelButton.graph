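# Widget Blueprint for a party button, based on FortJoinVoiceChannelButton, labelled to switch voice chat to this device.
# Nested borders frame a horizontal box holding the inherited input action icon and the text label, with the inner border using a social panel material as background.
# Hover plays a short OnHover animation that shifts the outer border from blue toward yellow, and unhover plays it in reverse.
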WidgetBlueprint /FortniteCoreUI/Frontend/Parties/JoinVoiceChannelButton
animations [OnHover {
  binding_count=1
  length_seconds=0.1
  object=&self:OnHover
  sequence {bindings [Border_1 {
     id=#1
     possessed_class=Border
     tracks [{
       type=MovieSceneColorTrack
       display_name=Brush Color
       sections [{
         type=MovieSceneColorSection
         end_frame=2
         start_frame=0
         channels [
          Color.R {type=MovieSceneScriptingFloatChannel, keys [[0,0.2623],[2,0.9823]]}
          Color.G {type=MovieSceneScriptingFloatChannel, keys [[0,0.6514],[2,0.934]]}
          Color.B {type=MovieSceneScriptingFloatChannel, keys [[0,0.9823],[2,0.0656]]}
          Color.A {type=MovieSceneScriptingFloatChannel, keys [[0,1],[2,0.9]]}
         ]
        }]
      }]
    }]}
 }]
graphs [EventGraph {nodes [
   Event On Hovered()
   n1 = Get OnHover() -> (OnHover:Widget Animation Object Reference)
   > PlayAnimation(InAnimation=@n1)
   Event On Unhovered()
   > PlayAnimation(InAnimation=@n1, PlayMode=Reverse)
  ]}]
variables []
widgets {
 info {inheritedWidgetCount=1, parentClass=/Script/FortniteUI.FortJoinVoiceChannelButton, rootWidgetClass=&Border, widgetCount=5}
 widgets [
  Border_1 {widgetClassPath=&Border, widget=Border_1, properties {brushColor=<rgba 0.2623 0.6514 0.9823 1>}}
  Border_0 {
   widgetClassPath=&Border
   parent=Border_1
   widget=Border_0
   properties {
    horizontalAlignment=HAlign_Left
    verticalAlignment=VAlign_Center
    background {
     bIsDynamicallyLoaded=false
     drawAs=Image
     imageType=NoImage
     mirroring=NoMirror
     resourceName=None
     resourceObject=/FortniteCoreUI/UI/SocialPanel/Art/Materials/M_UI_SocialUserItemBorder_SearchFriend
     tiling=NoTile
     imageSize=<32 32>
     margin=<ltrb 0 0 0 0>
     outlineSettings {
      bUseBrushTransparency=false
      roundingType=HalfHeightRadius
      width=0
      color {colorUseRule=UseColor_Specified, specifiedColor=<rgba 0 0 0 0>}
      cornerRadii=<0 0 0 0>
     }
     tintColor {colorUseRule=UseColor_Specified, specifiedColor=<rgba 1 1 1 1>}
     uVRegion {bIsValid=false, max=<0 0>, min=<0 0>}
    }
    padding=<ltrb 16 8 16 8>
   }
   slot {class=BorderSlot}
  }
  HorizontalBox_0 {
   widgetClassPath=&HorizontalBox
   parent=Border_0
   widget=HorizontalBox_0
   slot {class=BorderSlot, properties {horizontalAlignment=HAlign_Left, verticalAlignment=VAlign_Center, padding=<ltrb 16 8 16 8>}}
  }
  InputActionWidget {
   widgetClassPath=&CommonActionWidget
   bInherited=true
   bIsVariable=true
   parent=HorizontalBox_0
   widget=InputActionWidget
   slot {class=HorizontalBoxSlot, properties {horizontalAlignment=HAlign_Left, padding=<ltrb 0 0 8 0>}}
  }
  CommonTextBlock_41 {
   widgetClassPath=&CommonTextBlock
   parent=HorizontalBox_0
   widget=CommonTextBlock_41
   properties {
    autoWrapText=true
    mobileFontSizeMultiplier=1.2
    scrollStyle=/FortUILibrary/Styles/Typography/Foundation/TextScrollStyle-NoFade.TextScrollStyle-NoFade_C
    text=SWITCH VOICE CHAT TO THIS DEVICE
    font {
     bForceMonospaced=false
     bMaterialIsStencil=false
     fontMaterial=null
     fontObject=&BurbankSmall
     letterSpacing=0
     monospacedWidth=1
     size=16
     skewAmount=0
     typefaceFontName=Bold
     outlineSettings {
      bApplyOutlineToDropShadows=false
      bMiteredCorners=false
      bSeparateFillAlpha=false
      outlineMaterial=null
      outlineSize=0
      outlineColor=<rgba 0 0 0 1>
     }
    }
    shadowOffset=<0 0>
   }
   slot {class=HorizontalBoxSlot, properties {horizontalAlignment=HAlign_Left, verticalAlignment=VAlign_Center}}
  }
 ]
}
