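# Social panel list entry for the local user, showing avatar, name, crown, voice channel and event RSVP info, with a settings menu anchor.
# Built on FortSocialEventRSVPLocalUserEntry, which supplies several of the bound widgets.
# Online status changes tint the avatar's indicator dot, mapping Chat to Online and ExtendedAway and DoNotDisturb to Away.
# Hover plays an animation that nudges the content and rotates, scales and brightens the settings icon, reversing faster on unhover.
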WidgetBlueprint /FortniteCoreUI/UI/SocialPanel/ListEntries/LocalUserEntry
animations [Hovered {
  binding_count=2
  display_label=""
  length_seconds=0.2
  object=&self:Hovered
  sequence {bindings [
    SettingsIcon {
     id=#1
     possessed_class=Image
     tracks [
      {
       type=MovieScene2DTransformTrack
       sections [{
         type=MovieScene2DTransformSection
         end_frame=4
         start_frame=0
         channels [
          Translation.X {type=MovieSceneScriptingFloatChannel}
          Translation.Y {type=MovieSceneScriptingFloatChannel}
          Angle {type=MovieSceneScriptingFloatChannel, keys [[0,0],[4,60]]}
          Scale.X {type=MovieSceneScriptingFloatChannel, keys [[0,1],[4,1.25]]}
          Scale.Y {type=MovieSceneScriptingFloatChannel, keys [[0,1],[4,1.25]]}
          Shear.X {type=MovieSceneScriptingFloatChannel}
          Shear.Y {type=MovieSceneScriptingFloatChannel}
         ]
        }]
      }
      {
       type=MovieSceneFloatTrack
       display_name=Render Opacity
       sections [{type=MovieSceneFloatSection, end_frame=4, start_frame=0, channels [None {type=MovieSceneScriptingFloatChannel, keys [[0,0.3],[4,0.5]]}]}]
      }
     ]
    }
    HorizontalBox_1 {
     id=#2
     possessed_class=HorizontalBox
     tracks [{
       type=MovieScene2DTransformTrack
       sections [{
         type=MovieScene2DTransformSection
         end_frame=4
         start_frame=0
         channels [
          Translation.X {type=MovieSceneScriptingFloatChannel, keys [[0,0],[2,-8]]}
          Translation.Y {type=MovieSceneScriptingFloatChannel}
          Angle {type=MovieSceneScriptingFloatChannel}
          Scale.X {type=MovieSceneScriptingFloatChannel}
          Scale.Y {type=MovieSceneScriptingFloatChannel}
          Shear.X {type=MovieSceneScriptingFloatChannel}
          Shear.Y {type=MovieSceneScriptingFloatChannel}
         ]
        }]
      }]
    }
   ]}
 }]
graphs [
 EventGraph {nodes [
   n1 = Event OnOnlineStatusChanged!() -> (OutputDelegate, then, OnlineStatus:ELocalUserOnlineStatus Enum=Online) {is=title K2Node_Event}
   > HandleOnlineIndicatorView!(execute, self, InLocalUserOnlineStatus:ELocalUserOnlineStatus Enum=@n1.OnlineStatus) -> (then)
   Event On Hovered()
   > PlayAnimation(InAnimation=Get Hovered() -> (Hovered:Widget Animation Object Reference))
   Event On Unhovered()
   > PlayAnimation(InAnimation=Get Hovered(), PlayMode=Reverse, PlaybackSpeed=2)
  ]}
 HandleOnlineIndicatorView {nodes [
   n1 = HandleOnlineIndicatorView!() -> (then, InLocalUserOnlineStatus:ELocalUserOnlineStatus Enum=Online) {is=title K2Node_FunctionEntry}
   n2 = Get Online() -> (Online:Linear Color Structure)
   n3 = Get Away() -> (Away:Linear Color Structure)
   > SetBrushColor(
    self=Get Border_OnlineStatusIndicator() -> (Border_OnlineStatusIndicator:Border Object Reference)
    InBrushColor=Select!(
     Online:Linear Color Structure=@n2
     Offline:Linear Color Structure=Get Offline() -> (Offline:Linear Color Structure)
     Away:Linear Color Structure=@n3
     ExtendedAway:Linear Color Structure=@n3
     DoNotDisturb:Linear Color Structure=@n3
     Chat:Linear Color Structure=@n2
     Index:ELocalUserOnlineStatus Enum=@n1
    ) -> (ReturnValue:Linear Color Structure)
   )
  ]}
]
variables [
 Away {type=LinearColor, default=<rgba 0.9647 0.5457 0.1221 1>}
 Offline {type=LinearColor, default=<rgba 0.1413 0.3278 0.5149 1>}
 Online {type=LinearColor, default=<rgba 0.2159 1 0.1329 1>}
]
widgets {
 info {
  inheritedWidgetCount=7
  parentClass=/Script/FortniteUI.FortSocialEventRSVPLocalUserEntry
  rootWidgetClass=&SizeBox
  widgetCount=29
 }
 widgets [
  SizeBox_0 {widgetClassPath=&SizeBox, widget=SizeBox_0, properties {bOverride_MinDesiredHeight=true, heightOverride=72, minDesiredHeight=72}}
  LocalUserEntryBorder {
   widgetClassPath=&Border
   bIsVariable=true
   parent=SizeBox_0
   widget=LocalUserEntryBorder
   properties {
    visibility=SelfHitTestInvisible
    background {
     bIsDynamicallyLoaded=false
     drawAs=Image
     imageType=NoImage
     mirroring=NoMirror
     resourceName=None
     resourceObject=null
     tiling=NoTile
     imageSize=<32 32>
     margin=<ltrb 0 0 0 0>
     outlineSettings {
      bUseBrushTransparency=false
      roundingType=HalfHeightRadius
      width=0
      color {colorUseRule=UseColor_Specified, specifiedColor=<rgba 0 0 0 0>}
      cornerRadii=<0 0 0 0>
     }
     tintColor {colorUseRule=UseColor_Specified, specifiedColor=<rgba 0 0.0044 0.0319 1>}
     uVRegion {bIsValid=false, max=<0 0>, min=<0 0>}
    }
    padding=<ltrb 3 3 3 3>
   }
   slot {class=SizeBoxSlot, properties {padding=<ltrb 8 0 8 0>}}
  }
  HorizontalBox_3 {
   widgetClassPath=&HorizontalBox
   parent=LocalUserEntryBorder
   widget=HorizontalBox_3
   slot {class=BorderSlot, properties {padding=<ltrb 3 3 3 3>}}
  }
  HorizontalBox_1 {
   widgetClassPath=&HorizontalBox
   parent=HorizontalBox_3
   widget=HorizontalBox_1
   slot {class=HorizontalBoxSlot, properties {verticalAlignment=VAlign_Center, size {sizeRule=Fill, value=1}}}
  }
  ScaleBox_0 {
   widgetClassPath=&ScaleBox
   parent=HorizontalBox_1
   widget=ScaleBox_0
   properties {stretch=ScaleToFitY}
   slot {class=HorizontalBoxSlot, properties {horizontalAlignment=HAlign_Center, verticalAlignment=VAlign_Center, padding=<ltrb 10 10 10 10>}}
  }
  Overlay_2 {widgetClassPath=&Overlay, parent=ScaleBox_0, widget=Overlay_2, slot {class=ScaleBoxSlot}}
  SocialAvatarIcon {
   widgetClassPath=/Script/FortniteUI.FortSocialAvatarIcon
   bInherited=true
   bIsVariable=true
   parent=Overlay_2
   widget=SocialAvatarIcon
   properties {
    avatarMaterial=/Game/UI/SocialPanel/Art/Materials/UI_UserPortrait_Buffet
    materialTextureParamName=SmallPreview
    brush {
     bIsDynamicallyLoaded=false
     drawAs=Image
     imageType=NoImage
     mirroring=NoMirror
     resourceName=None
     resourceObject=&UI_UserPortrait_Buffet
     tiling=NoTile
     imageSize=<72 72>
     margin=<ltrb 0 0 0 0>
     outlineSettings {
      bUseBrushTransparency=false
      roundingType=HalfHeightRadius
      width=0
      color {colorUseRule=UseColor_Specified, specifiedColor=<rgba 0 0 0 0>}
      cornerRadii=<0 0 0 0>
     }
     tintColor {colorUseRule=UseColor_Specified, specifiedColor=<rgba 1 1 1 1>}
     uVRegion {bIsValid=false, max=<0 0>, min=<0 0>}
    }
   }
   slot {class=OverlaySlot}
  }
  Border_OnlineStatusIndicator {
   widgetClassPath=&Border
   bIsVariable=true
   parent=Overlay_2
   widget=Border_OnlineStatusIndicator
   properties {
    background {
     bIsDynamicallyLoaded=false
     drawAs=Image
     imageType=NoImage
     mirroring=NoMirror
     resourceName=None
     resourceObject=/LauncherSocial/circle
     tiling=NoTile
     imageSize=<24 24>
     margin=<ltrb 0 0 0 0>
     outlineSettings {
      bUseBrushTransparency=false
      roundingType=HalfHeightRadius
      width=0
      color {colorUseRule=UseColor_Specified, specifiedColor=<rgba 0 0 0 0>}
      cornerRadii=<0 0 0 0>
     }
     tintColor {colorUseRule=UseColor_Specified, specifiedColor=<rgba 1 1 1 1>}
     uVRegion {bIsValid=false, max=<0 0>, min=<0 0>}
    }
    brushColor=<rgba 0.1946 0.9823 0.1812 1>
    padding=<ltrb 0 0 10 10>
   }
   slot {class=OverlaySlot, properties {horizontalAlignment=HAlign_Right, verticalAlignment=VAlign_Bottom}}
  }
  Spacer_0 {
   widgetClassPath=&Spacer
   parent=Border_OnlineStatusIndicator
   widget=Spacer_0
   properties {size=<8 8>}
   slot {class=BorderSlot, properties {padding=<ltrb 0 0 10 10>}}
  }
  FortMobileImage_enforzer {
   widgetClassPath=/Script/FortniteUI.FortMobileImage
   bIsVariable=true
   parent=HorizontalBox_1
   widget=FortMobileImage_enforzer
   properties {
    mobileSizeMultiplier=1.34
    brush {
     bIsDynamicallyLoaded=false
     drawAs=Image
     imageType=NoImage
     mirroring=NoMirror
     resourceName=None
     resourceObject=null
     tiling=NoTile
     imageSize=<15 72>
     margin=<ltrb 0 0 0 0>
     outlineSettings {
      bUseBrushTransparency=false
      roundingType=HalfHeightRadius
      width=0
      color {colorUseRule=UseColor_Specified, specifiedColor=<rgba 0 0 0 0>}
      cornerRadii=<0 0 0 0>
     }
     tintColor {colorUseRule=UseColor_Specified, specifiedColor=<rgba 1 1 1 0>}
     uVRegion {bIsValid=false, max=<0 0>, min=<0 0>}
    }
   }
   slot {class=HorizontalBoxSlot}
  }
  UserTextInfoVB {
   widgetClassPath=&VerticalBox
   parent=HorizontalBox_1
   widget=UserTextInfoVB
   slot {class=HorizontalBoxSlot, properties {verticalAlignment=VAlign_Center, padding=<ltrb 10 10 10 10>, size {sizeRule=Fill, value=1}}}
  }
  SizeBox_2 {
   widgetClassPath=&SizeBox
   parent=UserTextInfoVB
   widget=SizeBox_2
   properties {bOverride_MinDesiredHeight=true, heightOverride=24, minDesiredHeight=24}
   slot {class=VerticalBoxSlot, properties {horizontalAlignment=HAlign_Left, size {sizeRule=Fill, value=1}}}
  }
  HorizontalBox_4 {widgetClassPath=&HorizontalBox, parent=SizeBox_2, widget=HorizontalBox_4, slot {class=SizeBoxSlot}}
  Overlay_1 {
   widgetClassPath=&Overlay
   parent=HorizontalBox_4
   widget=Overlay_1
   slot {
    class=HorizontalBoxSlot
    properties {horizontalAlignment=HAlign_Center, verticalAlignment=VAlign_Center, padding=<ltrb 0 -4 0 -4>, size {sizeRule=Fill, value=1}}
   }
  }
  Text_UserName {
   widgetClassPath=&CommonTextBlock
   bInherited=true
   bIsVariable=true
   parent=Overlay_1
   widget=Text_UserName
   properties {
    mobileFontSizeMultiplier=1.34
    scrollStyle=/FortUILibrary/Styles/Typography/Foundation/TextScrollStyle-NoFade.TextScrollStyle-NoFade_C
    text=LOCAL USER NAME
    colorAndOpacity {colorUseRule=UseColor_Specified, specifiedColor=<rgba 1 1 1 0.5>}
    font {
     bForceMonospaced=false
     bMaterialIsStencil=false
     fontMaterial=null
     fontObject=&BurbankSmall
     letterSpacing=0
     monospacedWidth=1
     size=18
     skewAmount=0
     typefaceFontName=Black
     outlineSettings {
      bApplyOutlineToDropShadows=false
      bMiteredCorners=false
      bSeparateFillAlpha=false
      outlineMaterial=null
      outlineSize=0
      outlineColor=<rgba 0 0 0 1>
     }
    }
   }
   slot {class=OverlaySlot}
  }
  Image_Crown {
   widgetClassPath=&Image
   bInherited=true
   bIsVariable=true
   parent=Overlay_1
   widget=Image_Crown
   properties {brush {
     bIsDynamicallyLoaded=false
     drawAs=Image
     imageType=NoImage
     mirroring=NoMirror
     resourceName=None
     resourceObject=/Game/UI/Foundation/Banner/M-CrownShine
     tiling=NoTile
     imageSize=<15.5 11>
     margin=<ltrb 0 0 0 0>
     outlineSettings {
      bUseBrushTransparency=false
      roundingType=HalfHeightRadius
      width=0
      color {colorUseRule=UseColor_Specified, specifiedColor=<rgba 0 0 0 0>}
      cornerRadii=<0 0 0 0>
     }
     tintColor {colorUseRule=UseColor_Specified, specifiedColor=<rgba 1 1 1 1>}
     uVRegion {bIsValid=false, max=<0 0>, min=<0 0>}
    }}
   slot {class=OverlaySlot, properties {padding=<ltrb 0 -4 0 0>}}
  }
  PartyLeaderSpacer {
   widgetClassPath=&Spacer
   parent=HorizontalBox_4
   widget=PartyLeaderSpacer
   properties {size=<8 8>}
   slot {class=HorizontalBoxSlot}
  }
  HBox_VoiceStatus {
   widgetClassPath=&HorizontalBox
   parent=UserTextInfoVB
   widget=HBox_VoiceStatus
   slot {class=VerticalBoxSlot, properties {horizontalAlignment=HAlign_Left, padding=<ltrb 0 10 0 2>}}
  }
  SizeBox_1 {
   widgetClassPath=&SizeBox
   parent=HBox_VoiceStatus
   widget=SizeBox_1
   properties {bOverride_HeightOverride=true, bOverride_WidthOverride=true, heightOverride=32, minDesiredHeight=24, minDesiredWidth=24, widthOverride=32}
   slot {class=HorizontalBoxSlot, properties {padding=<ltrb 0 -8 4 -8>}}
  }
  Image_MicIcon {
   widgetClassPath=&Image
   bIsVariable=true
   parent=SizeBox_1
   widget=Image_MicIcon
   properties {brush {
     bIsDynamicallyLoaded=false
     drawAs=Image
     imageType=NoImage
     mirroring=NoMirror
     resourceName=None
     resourceObject=/Game/UI/SocialPanel/Art/Textures/T_UI_Icon_Mic_64
     tiling=NoTile
     imageSize=<32 32>
     margin=<ltrb 0 0 0 0>
     outlineSettings {
      bUseBrushTransparency=false
      roundingType=HalfHeightRadius
      width=0
      color {colorUseRule=UseColor_Specified, specifiedColor=<rgba 0 0 0 0>}
      cornerRadii=<0 0 0 0>
     }
     tintColor {colorUseRule=UseColor_Specified, specifiedColor=<rgba 1 1 1 0.3>}
     uVRegion {bIsValid=false, max=<0 0>, min=<0 0>}
    }}
   slot {class=SizeBoxSlot, properties {horizontalAlignment=HAlign_Center, verticalAlignment=VAlign_Center}}
  }
  Text_VoiceChannel {
   widgetClassPath=&CommonTextBlock
   bInherited=true
   bIsVariable=true
   parent=HBox_VoiceStatus
   widget=Text_VoiceChannel
   properties {
    mobileFontSizeMultiplier=1.2
    scrollStyle=&TextScrollStyle-NoFade_C
    text="{VChannel}"
    textTransformPolicy=ToUpper
    colorAndOpacity {colorUseRule=UseColor_Specified, specifiedColor=<rgba 1 1 1 0.3>}
    font {
     bForceMonospaced=false
     bMaterialIsStencil=false
     fontMaterial=null
     fontObject=&BurbankSmall
     letterSpacing=0
     monospacedWidth=1
     size=15
     skewAmount=0
     typefaceFontName=Black
     outlineSettings {
      bApplyOutlineToDropShadows=false
      bMiteredCorners=false
      bSeparateFillAlpha=false
      outlineMaterial=null
      outlineSize=0
      outlineColor=<rgba 0 0 0 1>
     }
    }
   }
   slot {class=HorizontalBoxSlot, properties {horizontalAlignment=HAlign_Center, verticalAlignment=VAlign_Center, padding=<ltrb 0 -4 0 -4>}}
  }
  BuffetInfo {
   widgetClassPath=&/FortSocialEventRSVPWidget
   bInherited=true
   bIsVariable=true
   parent=UserTextInfoVB
   widget=BuffetInfo
   slot {class=VerticalBoxSlot, properties {padding=<ltrb 0 6 0 4>}}
  }
  SizeBox_3 {
   widgetClassPath=&SizeBox
   parent=HorizontalBox_3
   widget=SizeBox_3
   properties {bOverride_MinDesiredWidth=true, minDesiredWidth=104}
   slot {class=HorizontalBoxSlot, properties {horizontalAlignment=HAlign_Right}}
  }
  Overlay_0 {widgetClassPath=&Overlay, parent=SizeBox_3, widget=Overlay_0, slot {class=SizeBoxSlot}}
  HorizontalBox_0 {
   widgetClassPath=&HorizontalBox
   parent=Overlay_0
   widget=HorizontalBox_0
   slot {class=OverlaySlot, properties {horizontalAlignment=HAlign_Center, verticalAlignment=VAlign_Center, padding=<ltrb 8 0 8 0>}}
  }
  InputActionWidget {
   widgetClassPath=&CommonActionWidget
   bInherited=true
   parent=HorizontalBox_0
   widget=InputActionWidget
   properties {visibility=SelfHitTestInvisible, inputActions [{dataTable=&InputActionDataTable, rowName=Select}]}
   slot {class=HorizontalBoxSlot, properties {horizontalAlignment=HAlign_Center, verticalAlignment=VAlign_Center}}
  }
  SettingsIconSB {
   widgetClassPath=&SizeBox
   parent=HorizontalBox_0
   widget=SettingsIconSB
   properties {
    bOverride_MinDesiredHeight=true
    bOverride_MinDesiredWidth=true
    heightOverride=72
    minDesiredHeight=72
    minDesiredWidth=72
    widthOverride=72
   }
   slot {class=HorizontalBoxSlot, properties {horizontalAlignment=HAlign_Left}}
  }
  SettingsIcon {
   widgetClassPath=&FortMobileImage
   bIsVariable=true
   parent=SettingsIconSB
   widget=SettingsIcon
   properties {
    mobileSizeMultiplier=1.1
    renderOpacity=0.3
    visibility=SelfHitTestInvisible
    brush {
     bIsDynamicallyLoaded=false
     drawAs=Image
     imageType=NoImage
     mirroring=NoMirror
     resourceName=None
     resourceObject=/Game/UI/SocialPanel/Art/Textures/T_UI_Icon_LocalUserSettings
     tiling=NoTile
     imageSize=<48 48>
     margin=<ltrb 0 0 0 0>
     outlineSettings {
      bUseBrushTransparency=false
      roundingType=HalfHeightRadius
      width=0
      color {colorUseRule=UseColor_Specified, specifiedColor=<rgba 0 0 0 0>}
      cornerRadii=<0 0 0 0>
     }
     tintColor {colorUseRule=UseColor_Specified, specifiedColor=<rgba 1 1 1 1>}
     uVRegion {bIsValid=false, max=<0 0>, min=<0 0>}
    }
   }
   slot {class=SizeBoxSlot, properties {horizontalAlignment=HAlign_Center, verticalAlignment=VAlign_Center}}
  }
  MenuAnchor_LocalUserSettingMenu {
   widgetClassPath=/Script/UMG.MenuAnchor
   bInherited=true
   bIsVariable=true
   parent=Overlay_0
   widget=MenuAnchor_LocalUserSettingMenu
   properties {
    menuClass=/FortniteCoreUI/UI/SocialPanel/LocalUserSettingsMenu.LocalUserSettingsMenu_C
    placement=MenuPlacement_MenuRight
    renderTransform {angle=0, scale=<1 1>, shear=<0 0>, translation=<8 -4>}
   }
   slot {class=OverlaySlot, properties {horizontalAlignment=HAlign_Right}}
  }
 ]
}
