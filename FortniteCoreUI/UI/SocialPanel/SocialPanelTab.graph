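# Social panel tab button derived from FortSocialTab, showing an icon, count, name and collapsed "!" marker over a material-driven background.
# Hover events play the Hovered animation forward or reversed, only when unselected and not using touch.
# Selection plays Selected, while deselection plays Unselected reversed, animating background material shape, color, content opacity and scale.
# An OnFriendsOnlineUpdated animation pops the count, but this graph never plays it.
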WidgetBlueprint /FortniteCoreUI/UI/SocialPanel/SocialPanelTab
animations [
 Selected {
  binding_count=5
  display_label=""
  length_seconds=0.4
  object=&self:Selected
  sequence {bindings [
    ButtonBase {
     id=#1
     possessed_class=Image
     tracks [
      {
       type=MovieSceneWidgetMaterialTrack
       sections [{
         type=MovieSceneParameterSection
         channels [
          TopAngle {type=MovieSceneScriptingFloatChannel, keys [[0,0],[8,2.5]]}
          BottomAngle {type=MovieSceneScriptingFloatChannel, keys [[0,0],[8,-1.5]]}
          BottomOffset {type=MovieSceneScriptingFloatChannel, keys [[0,0.075],[8,0.025]]}
          TopOffset {type=MovieSceneScriptingFloatChannel, keys [[0,0.075],[8,0.0375]]}
         ]
        }]
      }
      {
       type=MovieScene2DTransformTrack
       sections [{
         type=MovieScene2DTransformSection
         channels [
          Translation.X {type=MovieSceneScriptingFloatChannel}
          Translation.Y {type=MovieSceneScriptingFloatChannel}
          Angle {type=MovieSceneScriptingFloatChannel}
          Scale.X {type=MovieSceneScriptingFloatChannel, keys [[0,1],[4,1.075],[8,1]]}
          Scale.Y {type=MovieSceneScriptingFloatChannel, keys [[0,1],[4,1.125],[8,1]]}
          Shear.X {type=MovieSceneScriptingFloatChannel}
          Shear.Y {type=MovieSceneScriptingFloatChannel}
         ]
        }]
      }
      {
       type=MovieSceneColorTrack
       display_name=Color and Opacity
       sections [{
         type=MovieSceneColorSection
         channels [
          Color.R {type=MovieSceneScriptingFloatChannel, keys [[0,0],[8,0.0273]]}
          Color.G {type=MovieSceneScriptingFloatChannel, keys [[0,0.1559],[8,0.5029]]}
          Color.B {type=MovieSceneScriptingFloatChannel, keys [[0,0.4452],[8,0.7454]]}
          Color.A {type=MovieSceneScriptingFloatChannel, const [1,1]}
         ]
        }]
      }
     ]
    }
    Image_TabIcon {
     id=#2
     possessed_class=Image
     tracks [
      {
       type=MovieSceneFloatTrack
       display_name=Render Opacity
       sections [{type=MovieSceneFloatSection, end_frame=8, start_frame=0, channels [None {type=MovieSceneScriptingFloatChannel, keys [[0,0.5],[4,1],[8,1]]}]}]
      }
      {
       type=MovieScene2DTransformTrack
       sections [{
         type=MovieScene2DTransformSection
         end_frame=8
         start_frame=0
         channels [
          Translation.X {type=MovieSceneScriptingFloatChannel}
          Translation.Y {type=MovieSceneScriptingFloatChannel, keys [[0,2],[4,0],[8,0]]}
          Angle {type=MovieSceneScriptingFloatChannel}
          Scale.X {type=MovieSceneScriptingFloatChannel, keys [[0,0.875],[4,1],[8,1]]}
          Scale.Y {type=MovieSceneScriptingFloatChannel, keys [[0,0.875],[4,1],[8,1]]}
          Shear.X {type=MovieSceneScriptingFloatChannel}
          Shear.Y {type=MovieSceneScriptingFloatChannel}
         ]
        }]
      }
     ]
    }
    Text_TabName {
     id=#3
     possessed_class=CommonTextBlock
     tracks [
      {
       type=MovieScene2DTransformTrack
       sections [{
         type=MovieScene2DTransformSection
         end_frame=8
         start_frame=0
         channels [
          Translation.X {type=MovieSceneScriptingFloatChannel}
          Translation.Y {type=MovieSceneScriptingFloatChannel, keys [[0,0],[4,-6],[8,0]]}
          Angle {type=MovieSceneScriptingFloatChannel}
          Scale.X {type=MovieSceneScriptingFloatChannel}
          Scale.Y {type=MovieSceneScriptingFloatChannel}
          Shear.X {type=MovieSceneScriptingFloatChannel}
          Shear.Y {type=MovieSceneScriptingFloatChannel}
         ]
        }]
      }
      {
       type=MovieSceneFloatTrack
       display_name=Render Opacity
       sections [{type=MovieSceneFloatSection, end_frame=8, start_frame=0, channels [None {type=MovieSceneScriptingFloatChannel, keys [[0,0.5],[4,1],[8,1]]}]}]
      }
     ]
    }
    LeftsideHB {
     id=#4
     possessed_class=HorizontalBox
     tracks [{
       type=MovieScene2DTransformTrack
       sections [{
         type=MovieScene2DTransformSection
         end_frame=8
         start_frame=0
         channels [
          Translation.X {type=MovieSceneScriptingFloatChannel}
          Translation.Y {type=MovieSceneScriptingFloatChannel, keys [[0,0],[4,-8],[8,0]]}
          Angle {type=MovieSceneScriptingFloatChannel}
          Scale.X {type=MovieSceneScriptingFloatChannel, keys [[0,1],[4,1.125],[8,1]]}
          Scale.Y {type=MovieSceneScriptingFloatChannel, keys [[0,1],[4,1.125],[8,1]]}
          Shear.X {type=MovieSceneScriptingFloatChannel}
          Shear.Y {type=MovieSceneScriptingFloatChannel}
         ]
        }]
      }]
    }
    Text_Number {
     id=#5
     possessed_class=CommonTextBlock
     tracks [{
       type=MovieSceneFloatTrack
       display_name=Render Opacity
       sections [{type=MovieSceneFloatSection, end_frame=8, start_frame=0, channels [None {type=MovieSceneScriptingFloatChannel, keys [[0,0.5],[4,1],[8,1]]}]}]
      }]
    }
   ]}
 }
 Hovered {
  binding_count=5
  display_label=""
  length_seconds=0.4
  object=&self:Hovered
  sequence {bindings [
    Image_TabIcon {
     id=#6
     possessed_class=Image
     tracks [{
       type=MovieSceneFloatTrack
       display_name=Render Opacity
       sections [{type=MovieSceneFloatSection, end_frame=4, start_frame=0, channels [None {type=MovieSceneScriptingFloatChannel, keys [[0,0.5],[4,0.8]]}]}]
      }]
    }
    Text_Number {
     id=#7
     possessed_class=CommonTextBlock
     tracks [{
       type=MovieSceneFloatTrack
       display_name=Render Opacity
       sections [{type=MovieSceneFloatSection, end_frame=4, start_frame=0, channels [None {type=MovieSceneScriptingFloatChannel, keys [[0,0.5],[4,0.8]]}]}]
      }]
    }
    Text_TabName {
     id=#8
     possessed_class=CommonTextBlock
     tracks [{
       type=MovieSceneFloatTrack
       display_name=Render Opacity
       sections [{type=MovieSceneFloatSection, end_frame=4, start_frame=0, channels [None {type=MovieSceneScriptingFloatChannel, keys [[0,0.5],[4,0.8]]}]}]
      }]
    }
    LeftsideHB {
     id=#9
     possessed_class=HorizontalBox
     tracks [{
       type=MovieScene2DTransformTrack
       sections [{
         type=MovieScene2DTransformSection
         end_frame=4
         start_frame=0
         channels [
          Translation.X {type=MovieSceneScriptingFloatChannel}
          Translation.Y {type=MovieSceneScriptingFloatChannel, keys [[0,0],[4,-2]]}
          Angle {type=MovieSceneScriptingFloatChannel}
          Scale.X {type=MovieSceneScriptingFloatChannel, keys [[0,1],[4,1.0625]]}
          Scale.Y {type=MovieSceneScriptingFloatChannel, keys [[0,1],[4,1.0625]]}
          Shear.X {type=MovieSceneScriptingFloatChannel}
          Shear.Y {type=MovieSceneScriptingFloatChannel}
         ]
        }]
      }]
    }
    ButtonBase {
     id=#10
     possessed_class=Image
     tracks [{
       type=MovieSceneColorTrack
       display_name=Color and Opacity
       sections [{
         type=MovieSceneColorSection
         channels [
          Color.R {type=MovieSceneScriptingFloatChannel, keys [[0,0],[8,0.0048]]}
          Color.G {type=MovieSceneScriptingFloatChannel, keys [[0,0.1559],[8,0.2462]]}
          Color.B {type=MovieSceneScriptingFloatChannel, keys [[0,0.4452],[8,0.5395]]}
          Color.A {type=MovieSceneScriptingFloatChannel, const [1,1]}
         ]
        }]
      }]
    }
   ]}
 }
 OnFriendsOnlineUpdated {
  binding_count=1
  display_label=""
  length_seconds=0.3
  object=&self:OnFriendsOnlineUpdated
  sequence {bindings [Text_Number {
     id=#11
     possessed_class=CommonTextBlock
     tracks [{
       type=MovieScene2DTransformTrack
       sections [{
         type=MovieScene2DTransformSection
         end_frame=6
         start_frame=0
         channels [
          Translation.X {type=MovieSceneScriptingFloatChannel}
          Translation.Y {type=MovieSceneScriptingFloatChannel}
          Angle {type=MovieSceneScriptingFloatChannel}
          Scale.X {type=MovieSceneScriptingFloatChannel, keys [[0,0.3856],[2,1.2],[3.1457,0.9],[4.373,0.9694],[6,1]]}
          Scale.Y {type=MovieSceneScriptingFloatChannel, keys [[0,0.6856],[2,1.15],[3.1457,1.1],[4.373,1.064],[6,1]]}
          Shear.X {type=MovieSceneScriptingFloatChannel}
          Shear.Y {type=MovieSceneScriptingFloatChannel}
         ]
        }]
      }]
    }]}
 }
 Unselected {
  binding_count=4
  display_label=""
  length_seconds=0.4
  object=&self:Unselected
  sequence {bindings [
    Image_TabIcon {
     id=#2
     possessed_class=Image
     tracks [
      {
       type=MovieSceneFloatTrack
       display_name=Render Opacity
       sections [{type=MovieSceneFloatSection, end_frame=8, start_frame=0, channels [None {type=MovieSceneScriptingFloatChannel, keys [[0,0.5],[4,1],[8,1]]}]}]
      }
      {
       type=MovieScene2DTransformTrack
       sections [{
         type=MovieScene2DTransformSection
         end_frame=8
         start_frame=0
         channels [
          Translation.X {type=MovieSceneScriptingFloatChannel}
          Translation.Y {type=MovieSceneScriptingFloatChannel}
          Angle {type=MovieSceneScriptingFloatChannel}
          Scale.X {type=MovieSceneScriptingFloatChannel, keys [[0,0.875],[8,1]]}
          Scale.Y {type=MovieSceneScriptingFloatChannel, keys [[0,0.875],[8,1]]}
          Shear.X {type=MovieSceneScriptingFloatChannel}
          Shear.Y {type=MovieSceneScriptingFloatChannel}
         ]
        }]
      }
     ]
    }
    Text_TabName {
     id=#3
     possessed_class=CommonTextBlock
     tracks [{
       type=MovieSceneFloatTrack
       display_name=Render Opacity
       sections [{type=MovieSceneFloatSection, end_frame=8, start_frame=0, channels [None {type=MovieSceneScriptingFloatChannel, keys [[0,0.5],[4,1],[8,1]]}]}]
      }]
    }
    Text_Number {
     id=#5
     possessed_class=CommonTextBlock
     tracks [{
       type=MovieSceneFloatTrack
       display_name=Render Opacity
       sections [{type=MovieSceneFloatSection, end_frame=8, start_frame=0, channels [None {type=MovieSceneScriptingFloatChannel, keys [[0,0.5],[4,1],[8,1]]}]}]
      }]
    }
    ButtonBase {
     id=#12
     possessed_class=Image
     tracks [
      {
       type=MovieSceneWidgetMaterialTrack
       sections [{
         type=MovieSceneParameterSection
         channels [
          TopAngle {type=MovieSceneScriptingFloatChannel, keys [[0,0],[8,2.5]]}
          BottomAngle {type=MovieSceneScriptingFloatChannel, keys [[0,0],[8,-1.5]]}
          BottomOffset {type=MovieSceneScriptingFloatChannel, keys [[0,0.075],[8,0.025]]}
          TopOffset {type=MovieSceneScriptingFloatChannel, keys [[0,0.075],[8,0.0375]]}
         ]
        }]
      }
      {
       type=MovieSceneColorTrack
       display_name=Color and Opacity
       sections [{
         type=MovieSceneColorSection
         channels [
          Color.R {type=MovieSceneScriptingFloatChannel, keys [[0,0],[8,0.0273]]}
          Color.G {type=MovieSceneScriptingFloatChannel, keys [[0,0.1559],[8,0.5029]]}
          Color.B {type=MovieSceneScriptingFloatChannel, keys [[0,0.4452],[8,0.7454]]}
          Color.A {type=MovieSceneScriptingFloatChannel, const [1,1]}
         ]
        }]
      }
     ]
    }
   ]}
 }
]
graphs [EventGraph {nodes [
   Event On Hovered()
   n1 = GetOwningPlayer()
   n2 = AND Boolean(
    A=NOT Boolean(A=GetSelected!(self:Common Button Base Object Reference) -> (ReturnValue:bool))
    B=NOT Boolean(A=IsUsingTouch(PlayerController=@n1))
   )
   > Branch(Condition=@n2)
   n3 = Get Hovered() -> (Hovered:Widget Animation Object Reference)
   > PlayAnimation(InAnimation=@n3)
   Event On Unhovered()
   > Branch(Condition=@n2)
   > PlayAnimation(InAnimation=@n3, PlayMode=Reverse)
   Event On Selected()
   > PlayAnimation(InAnimation=Get Selected() -> (Selected:Widget Animation Object Reference), PlaybackSpeed=1.5)
   Event On Deselected()
   > PlayAnimation(InAnimation=Get Unselected() -> (Unselected:Widget Animation Object Reference), PlayMode=Reverse, PlaybackSpeed=1.5)
  ]}]
variables []
widgets {
 info {inheritedWidgetCount=5, parentClass=/Script/FortniteUI.FortSocialTab, rootWidgetClass=&Overlay, widgetCount=12}
 widgets [
  Overlay_0 {widgetClassPath=&Overlay, widget=Overlay_0}
  ButtonBase {
   widgetClassPath=&Image
   parent=Overlay_0
   widget=ButtonBase
   properties {
    visibility=SelfHitTestInvisible
    brush {
     bIsDynamicallyLoaded=false
     drawAs=Image
     imageType=NoImage
     mirroring=NoMirror
     resourceName=None
     resourceObject=/Game/UI/SocialPanel/Art/Materials/M_UI_SocialTab
     tiling=NoTile
     imageSize=<32 32>
     margin=<ltrb 0.5 0.5 0.5 0.5>
     outlineSettings {
      bUseBrushTransparency=false
      roundingType=HalfHeightRadius
      width=0
      color {colorUseRule=UseColor_Specified, specifiedColor=<rgba 0 0 0 0>}
      cornerRadii=<0 0 0 0>
     }
     tintColor {colorUseRule=UseColor_Specified, specifiedColor=<rgba 1 1 1 1>}
     uVRegion {bIsValid=false, max=<0 0>, min=<0 0>}
    }
    colorAndOpacity=<rgba 0 0.1559 0.4508 1>
   }
   slot {class=OverlaySlot, properties {horizontalAlignment=HAlign_Fill, verticalAlignment=VAlign_Fill}}
  }
  HorizontalBox_0 {
   widgetClassPath=&HorizontalBox
   parent=Overlay_0
   widget=HorizontalBox_0
   slot {class=OverlaySlot, properties {horizontalAlignment=HAlign_Fill, verticalAlignment=VAlign_Fill, padding=<ltrb 4 10 4 10>}}
  }
  LeftsideVB {
   widgetClassPath=&VerticalBox
   parent=HorizontalBox_0
   widget=LeftsideVB
   slot {class=HorizontalBoxSlot, properties {size {sizeRule=Fill, value=1}}}
  }
  LeftsideHB {
   widgetClassPath=&HorizontalBox
   parent=LeftsideVB
   widget=LeftsideHB
   slot {class=VerticalBoxSlot, properties {horizontalAlignment=HAlign_Center, verticalAlignment=VAlign_Center, padding=<ltrb 0 5 0 0>}}
  }
  Image_TabIcon {
   widgetClassPath=/Script/FortniteUI.FortMobileImage
   bInherited=true
   bIsVariable=true
   parent=LeftsideHB
   widget=Image_TabIcon
   properties {
    mobileSizeMultiplier=1.1
    renderOpacity=0.5
    brush {
     bIsDynamicallyLoaded=false
     drawAs=Image
     imageType=NoImage
     mirroring=NoMirror
     resourceName=None
     resourceObject=/Game/UI/SocialPanel/Art/Textures/T_UI_Icon_AddFriends_64
     tiling=NoTile
     imageSize=<44 44>
     margin=<ltrb 0 0 0 0>
     outlineSettings {
      bUseBrushTransparency=false
      roundingType=HalfHeightRadius
      width=0
      color {colorUseRule=UseColor_Specified, specifiedColor=<rgba 0 0 0 0>}
      cornerRadii=<0 0 0 0>
     }
     tintColor {colorUseRule=UseColor_Specified, specifiedColor=<rgba 1 1 1 1>}
     uVRegion {bIsValid=false, max=<0 0>, min=<0 0>}
    }
    renderTransform {angle=0, scale=<1 1>, shear=<0 0>, translation=<0 2>}
   }
   slot {class=HorizontalBoxSlot, properties {horizontalAlignment=HAlign_Center, verticalAlignment=VAlign_Center}}
  }
  Spacer_0 {widgetClassPath=&Spacer, parent=LeftsideHB, widget=Spacer_0, properties {size=<4 4>}, slot {class=HorizontalBoxSlot}}
  Text_Number {
   widgetClassPath=&CommonTextBlock
   bInherited=true
   bIsVariable=true
   parent=LeftsideHB
   widget=Text_Number
   properties {
    justification=Center
    mobileFontSizeMultiplier=1.2
    renderOpacity=0.5
    text=32
    font {
     bForceMonospaced=false
     bMaterialIsStencil=false
     fontMaterial=null
     fontObject=&BurbankSmall
     letterSpacing=0
     monospacedWidth=1
     size=20
     skewAmount=0
     typefaceFontName=Black
     outlineSettings {
      bApplyOutlineToDropShadows=false
      bMiteredCorners=false
      bSeparateFillAlpha=false
      outlineMaterial=null
      outlineSize=0
      outlineColor=<rgba 0 0 0 1>
     }
    }
    renderTransformPivot=<0.5 1>
   }
   slot {class=HorizontalBoxSlot, properties {horizontalAlignment=HAlign_Center, verticalAlignment=VAlign_Center}}
  }
  Overlay_1 {
   widgetClassPath=&Overlay
   parent=LeftsideVB
   widget=Overlay_1
   slot {class=VerticalBoxSlot, properties {padding=<ltrb 4 0 4 0>, size {sizeRule=Fill, value=1}}}
  }
  Text_TabName {
   widgetClassPath=&CommonTextBlock
   bInherited=true
   bIsVariable=true
   parent=Overlay_1
   widget=Text_TabName
   properties {
    justification=Center
    lineHeightPercentage=0.8
    mobileFontSizeMultiplier=1.25
    renderOpacity=0.5
    text=MY FRIENDS
    wrapTextAt=150
    font {
     bForceMonospaced=false
     bMaterialIsStencil=false
     fontMaterial=null
     fontObject=&BurbankSmall
     letterSpacing=0
     monospacedWidth=1
     size=14
     skewAmount=0
     typefaceFontName=Black
     outlineSettings {
      bApplyOutlineToDropShadows=false
      bMiteredCorners=false
      bSeparateFillAlpha=false
      outlineMaterial=null
      outlineSize=0
      outlineColor=<rgba 0 0 0 1>
     }
    }
   }
   slot {class=OverlaySlot, properties {horizontalAlignment=HAlign_Center, verticalAlignment=VAlign_Center}}
  }
  Text_Bang {
   widgetClassPath=&CommonTextBlock
   bInherited=true
   parent=HorizontalBox_0
   widget=Text_Bang
   properties {
    justification=Center
    text="!"
    visibility=Collapsed
    colorAndOpacity {colorUseRule=UseColor_Specified, specifiedColor=<rgba 0.9473 0.9911 0.0666 1>}
    font {
     bForceMonospaced=false
     bMaterialIsStencil=false
     fontMaterial=null
     fontObject=&BurbankSmall
     letterSpacing=0
     monospacedWidth=1
     size=40
     skewAmount=0
     typefaceFontName=Black
     outlineSettings {
      bApplyOutlineToDropShadows=false
      bMiteredCorners=false
      bSeparateFillAlpha=false
      outlineMaterial=null
      outlineSize=0
      outlineColor=<rgba 0 0 0 1>
     }
    }
    renderTransform {angle=0, scale=<1 1>, shear=<-8 0>, translation=<0 0>}
   }
   slot {class=HorizontalBoxSlot, properties {horizontalAlignment=HAlign_Right, verticalAlignment=VAlign_Center, padding=<ltrb 0 0 2 -4>}}
  }
  InputActionWidget {widgetClassPath=&CommonActionWidget, bInherited=true, widget=null}
 ]
}
