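# Social panel button widget, built on CommonButtonLegacy, that displays an "Add In WeGame" prompt for adding a friend through WeGame.
# A border using the WeGame add-friend material wraps a horizontal box holding the centered prompt text.
# The event graph only drives hover feedback: a short scale-up animation plays on hover and reverses at double speed on unhover.
# No click handling is defined here.
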WidgetBlueprint /FortniteCoreUI/UI/SocialPanel/SocialPanelWeGameFriendButton
animations [OnHovered {
  binding_count=1
  display_label=""
  length_seconds=0.1
  object=&self:OnHovered
  sequence {bindings [SocialPanelWeGameFriendButton_C_61 {
     id=#1
     possessed_class=SocialPanelWeGameFriendButton_C
     tracks [{
       type=MovieScene2DTransformTrack
       sections [{
         type=MovieScene2DTransformSection
         end_frame=2
         start_frame=-1
         channels [
          Translation.X {type=MovieSceneScriptingFloatChannel}
          Translation.Y {type=MovieSceneScriptingFloatChannel}
          Angle {type=MovieSceneScriptingFloatChannel}
          Scale.X {type=MovieSceneScriptingFloatChannel, keys [[-0.0003,1],[1.9997,1.0063],[2,1.1]]}
          Scale.Y {type=MovieSceneScriptingFloatChannel, keys [[-0.0003,1],[1.9997,1.0063],[2,1.1]]}
          Shear.X {type=MovieSceneScriptingFloatChannel}
          Shear.Y {type=MovieSceneScriptingFloatChannel}
         ]
        }]
      }]
    }]}
 }]
graphs [EventGraph {nodes [
   Event On Hovered()
   > PlayAnimation(InAnimation=Get OnHovered() -> (OnHovered:Widget Animation Object Reference))
   Event On Unhovered()
   > PlayAnimation(InAnimation=Get OnHovered(), PlayMode=Reverse, PlaybackSpeed=2)
  ]}]
variables []
widgets {
 info {inheritedWidgetCount=1, parentClass=&CommonButtonLegacy, rootWidgetClass=&Border, widgetCount=4}
 widgets [
  EntryBorder {
   widgetClassPath=&Border
   widget=EntryBorder
   properties {
    visibility=SelfHitTestInvisible
    background {
     bIsDynamicallyLoaded=false
     drawAs=Image
     imageType=NoImage
     mirroring=NoMirror
     resourceName=None
     resourceObject=/Game/UI/SocialPanel/Art/Materials/M_UI_SocialWeGameAddFriend
     tiling=NoTile
     imageSize=<32 32>
     margin=<ltrb 0 0 0 0>
     outlineSettings {
      bUseBrushTransparency=false
      roundingType=HalfHeightRadius
      width=0
      color {colorUseRule=UseColor_Specified, specifiedColor=<rgba 0 0 0 0>}
      cornerRadii=<0 0 0 0>
     }
     tintColor {colorUseRule=UseColor_Specified, specifiedColor=<rgba 1 1 1 1>}
     uVRegion {bIsValid=false, max=<0 0>, min=<0 0>}
    }
    padding=<ltrb 20 0 20 0>
   }
  }
  HorizontalBox_2 {
   widgetClassPath=&HorizontalBox
   parent=EntryBorder
   widget=HorizontalBox_2
   slot {class=BorderSlot, properties {padding=<ltrb 20 0 20 0>}}
  }
  Text_LinkPrompt {
   widgetClassPath=&CommonTextBlock
   bIsVariable=true
   parent=HorizontalBox_2
   widget=Text_LinkPrompt
   properties {
    justification=Center
    mobileFontSizeMultiplier=1.2
    text=Add In WeGame
    colorAndOpacity {colorUseRule=UseColor_Specified, specifiedColor=<rgba 0 0 0 1>}
    font {
     bForceMonospaced=false
     bMaterialIsStencil=false
     fontMaterial=null
     fontObject=&BurbankSmall
     letterSpacing=0
     monospacedWidth=1
     size=24
     skewAmount=0
     typefaceFontName=Black
     outlineSettings {
      bApplyOutlineToDropShadows=false
      bMiteredCorners=false
      bSeparateFillAlpha=false
      outlineMaterial=null
      outlineSize=0
      outlineColor=<rgba 0 0 0 1>
     }
    }
   }
   slot {class=HorizontalBoxSlot, properties {size {sizeRule=Fill, value=1}}}
  }
  InputActionWidget {widgetClassPath=&CommonActionWidget, bInherited=true, widget=null}
 ]
}
