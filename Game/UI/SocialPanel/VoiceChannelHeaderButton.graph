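# Social panel voice channel header button, a FortVoiceChannelHeaderButton widget showing a channel's title, status and join or expand controls.
# OnUpdateVisuals turns active, available, listening and restriction flags into opacity, visibility, warning color and switcher states; the simultaneous-connection runtime option selects between On/Off and Join/arrow layouts.
# Hover scales the button and slides out the join overlay for joinable, non-active channels.
# Expansion changes animate the arrow and play a sound.
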WidgetBlueprint /Game/UI/SocialPanel/VoiceChannelHeaderButton
animations [
 ScaleUp {
  binding_count=1
  length_seconds=0.25
  object=&self:ScaleUp
  sequence {bindings [VoiceChannelHeaderButton_C_58 {
     id=#1
     possessed_class=VoiceChannelHeaderButton_C
     tracks [{
       type=MovieScene2DTransformTrack
       sections [{
         type=MovieScene2DTransformSection
         end_frame=5
         start_frame=0
         channels [
          Translation.X {type=MovieSceneScriptingFloatChannel}
          Translation.Y {type=MovieSceneScriptingFloatChannel}
          Angle {type=MovieSceneScriptingFloatChannel}
          Scale.X {type=MovieSceneScriptingFloatChannel, keys [[0,1],[5,1.01]]}
          Scale.Y {type=MovieSceneScriptingFloatChannel, keys [[0,1],[5,1.01]]}
          Shear.X {type=MovieSceneScriptingFloatChannel}
          Shear.Y {type=MovieSceneScriptingFloatChannel}
         ]
        }]
      }]
    }]}
 }
 JoinSlideOut {
  binding_count=1
  length_seconds=0.15
  object=&self:JoinSlideOut
  sequence {bindings [Overlay_1 {
     id=#2
     possessed_class=Overlay
     tracks [{
       type=MovieScene2DTransformTrack
       sections [{
         type=MovieScene2DTransformSection
         end_frame=3
         start_frame=0
         channels [
          Translation.X {type=MovieSceneScriptingFloatChannel, const [0,2]}
          Translation.Y {type=MovieSceneScriptingFloatChannel, const [0,2]}
          Angle {type=MovieSceneScriptingFloatChannel, const [0,2]}
          Scale.X {type=MovieSceneScriptingFloatChannel, keys [[0,0],[3,1]]}
          Scale.Y {type=MovieSceneScriptingFloatChannel, const [1,2]}
          Shear.X {type=MovieSceneScriptingFloatChannel, const [0,2]}
          Shear.Y {type=MovieSceneScriptingFloatChannel, const [0,2]}
         ]
        }]
      }]
    }]}
 }
 OnExpansion {
  binding_count=1
  length_seconds=0.2
  object=&self:OnExpansion
  sequence {bindings [ExpandDirectionTick {
     id=#3
     possessed_class=FortMobileImage
     tracks [{
       type=MovieScene2DTransformTrack
       sections [{
         type=MovieScene2DTransformSection
         end_frame=4
         start_frame=0
         channels [
          Translation.X {type=MovieSceneScriptingFloatChannel, const [0,1]}
          Translation.Y {type=MovieSceneScriptingFloatChannel, const [0,1]}
          Angle {type=MovieSceneScriptingFloatChannel, keys [[0,0],[4,90]]}
          Scale.X {type=MovieSceneScriptingFloatChannel, keys [[0,1],[3,1.5],[4,1]]}
          Scale.Y {type=MovieSceneScriptingFloatChannel, keys [[0,1],[3,1.5],[4,1]]}
          Shear.X {type=MovieSceneScriptingFloatChannel, const [0,1]}
          Shear.Y {type=MovieSceneScriptingFloatChannel, const [0,1]}
         ]
        }]
      }]
    }]}
 }
]
graphs [EventGraph {nodes [
   n1 = Event On Item Expansion Changed!() -> (OutputDelegate, then, bIsExpanded:bool) {is=title K2Node_Event}
   > PlayAnimation(
    InAnimation=Get OnExpansion() -> (OnExpansion:Widget Animation Object Reference)
    PlayMode=Select(Option 0:EUMGSequencePlayMode Enum=Reverse, Option 1:EUMGSequencePlayMode Enum, Index=@n1.bIsExpanded) -> (ReturnValue:EUMGSequencePlayMode Enum)
    PlaybackSpeed=1.4
   )
   > PlaySound2D(Sound=/Game/Sounds/Fort_UI_Sounds/FrontEnd/New_FrontEnd/Fortnite_UI_MainMenu_Tab_Hover_02_Cue)
   n2 = Event OnUpdateVisuals!() -> (
    OutputDelegate
    then
    bIsActiveAndListening:bool
    bIsActiveOrAvailable:bool
    bIsAvailableAndNotActive:bool
    bIsListening:bool
    bShowingRestriction:bool
   ) {is=title K2Node_Event}
   > SetRenderOpacity(
    self=Get Text_ChannelTitle() -> (Text_ChannelTitle:Common Rich Text Block Object Reference)
    InOpacity=SelectFloat(A=1, B=0.5, bPickA=@n2.bIsActiveAndListening)
   )
   > SetRenderOpacity(InOpacity=SelectFloat(A=1, B=0.5, bPickA=@n2.bIsActiveAndListening))
   > SetVisibility(
    self=Get Switch_Transmitting() -> (Switch_Transmitting:Widget Switcher Object Reference)
    InVisibility=Select(Option 0:ESlateVisibility Enum=Collapsed, Option 1:ESlateVisibility Enum=SelfHitTestInvisible, Index=@n2.bIsActiveAndListening) -> (ReturnValue:ESlateVisibility Enum)
   )
   > SetVisibility(
    self=Get HBox_Join() -> (HBox_Join:Horizontal Box Object Reference)
    InVisibility=Select(Option 0=Collapsed, Option 1=SelfHitTestInvisible, Index=@n2.bIsActiveOrAvailable)
   )
   > SetRenderScale!(
    execute
    self:Widget Object Reference=Get Overlay_1() -> (Overlay_1:Overlay Object Reference)
    Scale:Vector 2D Structure=Select(Option 0:Vector 2D Structure=(X=0,Y=1), Option 1:Vector 2D Structure=(X=1,Y=1), Index=@n2.bIsActiveOrAvailable) -> (ReturnValue:Vector 2D Structure)
   ) -> (then)
   n3 = GetRuntimeOptions()
   > SetActiveWidgetIndex(
    self=Get SimultaneousChannelSwitcher() -> (SimultaneousChannelSwitcher:Widget Switcher Object Reference)
    Index=SelectInt(
     A=0
     B=1
     bPickA=Get bAllowSimultaneousChannelConnections!(self:Fort Runtime Options Object Reference=@n3) -> (bAllowSimultaneousChannelConnections:bool)
    )
   )
   > SetVisibility(
    self=Get JoinText() -> (JoinText:Common Rich Text Block Object Reference)
    InVisibility=Select(Option 0:ESlateVisibility Enum=Collapsed, Option 1:ESlateVisibility Enum=Visible, Index=@n2.bIsAvailableAndNotActive) -> (ReturnValue:ESlateVisibility Enum)
   )
   > SetActiveWidgetIndex(self=Get OnOffSwitcher() -> (OnOffSwitcher:Widget Switcher Object Reference), Index=SelectInt(A=0, B=1, bPickA=@n2.bIsListening))
   > SetVisibility(
    self=Get ExpandDirectionTick() -> (ExpandDirectionTick:Fort Mobile Image Object Reference)
    InVisibility=Select(Option 0=Collapsed, Option 1=Visible, Index=@n2.bIsActiveAndListening)
   )
   n4 = Get VoiceStatusActiveColor() -> (VoiceStatusActiveColor:Linear Color Structure)
   n5 = Get VoiceStatusWarningColor() -> (VoiceStatusWarningColor:Linear Color Structure)
   > SetColorAndOpacity(
    self=Get Text_ChannelInfo() -> (Text_ChannelInfo:Common Text Object Reference)
    InColorAndOpacity=Make Slate Color(SpecifiedColor=SelectColor(A=@n5, B=@n4, bPickA=@n2.bShowingRestriction))
   )
   > SetVisibility(
    self=Get Image_VoiceChatWarning() -> (Image_VoiceChatWarning:Image Object Reference)
    InVisibility=Select(Option 0=Collapsed, Option 1=Visible, Index=@n2.bShowingRestriction)
   )
   Event On Hovered()
   > PlayAnimationForward(InAnimation=Get ScaleUp() -> (ScaleUp:Widget Animation Object Reference))
   > n6 = Branch(Condition=IsActiveChannel!(self:Fort Voice Channel Header Button Object Reference) -> (ReturnValue:bool))
   Branch(execute=@n6.else, Condition=CanJoinChannel!(self:Fort Voice Channel Header Button Object Reference) -> (ReturnValue:bool))
   > PlayAnimationForward(InAnimation=Get JoinSlideOut() -> (JoinSlideOut:Widget Animation Object Reference))
   Event On Unhovered()
   > PlayAnimationReverse(InAnimation=Get ScaleUp())
   > n7 = Branch(Condition=IsActiveChannel())
   Branch(execute=@n7.else, Condition=CanJoinChannel())
   > PlayAnimationReverse(InAnimation=Get JoinSlideOut())
   K2Node_Event_6 = OnListItemObjectSet_1!() -> (OutputDelegate, then, ListItemObject:Object Reference) {is=title K2Node_CustomEvent}
   > Branch(Condition=IsListItemExpanded!(UserListEntry:User List Entry Interface) -> (ReturnValue:bool))
   n8 = Get OnExpansion()
   > PlayAnimation(InAnimation=@n8, StartAtTime=GetEndTime(self=@n8))
   Event On Enabled() {is=title K2Node_Event}
   > SetRenderScale(self=Get Overlay_1(), Scale=(X=1,Y=1))
   Event On Disabled() {is=title K2Node_Event}
   > SetRenderScale(self=Get Overlay_1(), Scale=(X=0,Y=1))
   Event PreConstruct()
   Event Construct()
   Event Tick()
  ]}]
variables [
 VoiceStatusActiveColor {type=LinearColor, default=<rgba 0.3231 0.8308 1 1>}
 VoiceStatusWarningColor {type=LinearColor, default=<rgba 1 0.8963 0.1046 1>}
]
widgets {
 info {
  inheritedWidgetCount=7
  parentClass=/Script/FortniteUI.FortVoiceChannelHeaderButton
  rootWidgetClass=&HorizontalBox
  widgetCount=27
 }
 widgets [
  HorizontalBox_1 {widgetClassPath=&HorizontalBox, widget=HorizontalBox_1}
  VerticalBox_0 {
   widgetClassPath=&VerticalBox
   parent=HorizontalBox_1
   widget=VerticalBox_0
   slot {class=HorizontalBoxSlot, properties {verticalAlignment=VAlign_Center, padding=<ltrb 15 0 15 0>, size {sizeRule=Fill, value=1}}}
  }
  HorizontalBox_0 {widgetClassPath=&HorizontalBox, parent=VerticalBox_0, widget=HorizontalBox_0, slot {class=VerticalBoxSlot}}
  Text_ChannelTitle {
   widgetClassPath=&CommonRichTextBlock
   bInherited=true
   bIsVariable=true
   parent=HorizontalBox_0
   widget=Text_ChannelTitle
   properties {
    bOverrideDefaultStyle=true
    text=Party Channel
    textTransformPolicy=ToUpper
    defaultTextStyleOverride {
     overflowPolicy=Clip
     transformPolicy=None
     colorAndOpacity {colorUseRule=UseColor_Specified, specifiedColor=<rgba 1 1 1 1>}
     font {
      bForceMonospaced=false
      bMaterialIsStencil=false
      fontMaterial=null
      fontObject=&BurbankSmall
      letterSpacing=0
      monospacedWidth=1
      size=19
      skewAmount=0
      typefaceFontName=Black
      outlineSettings {
       bApplyOutlineToDropShadows=false
       bMiteredCorners=false
       bSeparateFillAlpha=false
       outlineMaterial=null
       outlineSize=0
       outlineColor=<rgba 0 0 0 1>
      }
     }
     highlightColor {colorUseRule=UseColor_Specified, specifiedColor=<rgba 0 0 0 1>}
     highlightShape {
      bIsDynamicallyLoaded=false
      drawAs=Image
      imageType=NoImage
      mirroring=NoMirror
      resourceName=None
      resourceObject=null
      tiling=NoTile
      imageSize=<32 32>
      margin=<ltrb 0 0 0 0>
      outlineSettings {
       bUseBrushTransparency=false
       roundingType=HalfHeightRadius
       width=0
       color {colorUseRule=UseColor_Specified, specifiedColor=<rgba 0 0 0 0>}
       cornerRadii=<0 0 0 0>
      }
      tintColor {colorUseRule=UseColor_Specified, specifiedColor=<rgba 1 1 1 1>}
      uVRegion {bIsValid=false, max=<0 0>, min=<0 0>}
     }
     selectedBackgroundColor {colorUseRule=UseColor_Foreground, specifiedColor=<rgba 1 0 1 1>}
     shadowColorAndOpacity=<rgba 0 0 0 1>
     shadowOffset=<0 0>
     strikeBrush {
      bIsDynamicallyLoaded=false
      drawAs=Image
      imageType=NoImage
      mirroring=NoMirror
      resourceName=None
      resourceObject=null
      tiling=NoTile
      imageSize=<32 32>
      margin=<ltrb 0 0 0 0>
      outlineSettings {
       bUseBrushTransparency=false
       roundingType=HalfHeightRadius
       width=0
       color {colorUseRule=UseColor_Specified, specifiedColor=<rgba 0 0 0 0>}
       cornerRadii=<0 0 0 0>
      }
      tintColor {colorUseRule=UseColor_Specified, specifiedColor=<rgba 1 1 1 1>}
      uVRegion {bIsValid=false, max=<0 0>, min=<0 0>}
     }
     underlineBrush {
      bIsDynamicallyLoaded=false
      drawAs=Image
      imageType=NoImage
      mirroring=NoMirror
      resourceName=None
      resourceObject=null
      tiling=NoTile
      imageSize=<32 32>
      margin=<ltrb 0 0 0 0>
      outlineSettings {
       bUseBrushTransparency=false
       roundingType=HalfHeightRadius
       width=0
       color {colorUseRule=UseColor_Specified, specifiedColor=<rgba 0 0 0 0>}
       cornerRadii=<0 0 0 0>
      }
      tintColor {colorUseRule=UseColor_Specified, specifiedColor=<rgba 1 1 1 1>}
      uVRegion {bIsValid=false, max=<0 0>, min=<0 0>}
     }
    }
   }
   slot {class=HorizontalBoxSlot, properties {size {sizeRule=Fill, value=1}}}
  }
  Switch_Transmitting {
   widgetClassPath=&WidgetSwitcher
   bIsVariable=true
   parent=HorizontalBox_0
   widget=Switch_Transmitting
   slot {class=HorizontalBoxSlot}
  }
  Transmitting {
   widgetClassPath=&Image
   bIsVariable=true
   parent=Switch_Transmitting
   widget=Transmitting
   properties {
    visibility=HitTestInvisible
    brush {
     bIsDynamicallyLoaded=false
     drawAs=Image
     imageType=NoImage
     mirroring=NoMirror
     resourceName=None
     resourceObject=/Game/Athena/HUD/Art/BacchusHudIcons/Talk_Normal
     tiling=NoTile
     imageSize=<32 32>
     margin=<ltrb 0 0 0 0>
     outlineSettings {
      bUseBrushTransparency=false
      roundingType=HalfHeightRadius
      width=0
      color {colorUseRule=UseColor_Specified, specifiedColor=<rgba 0 0 0 0>}
      cornerRadii=<0 0 0 0>
     }
     tintColor {colorUseRule=UseColor_Specified, specifiedColor=<rgba 1 1 1 1>}
     uVRegion {bIsValid=false, max=<0 0>, min=<0 0>}
    }
   }
   slot {class=WidgetSwitcherSlot}
  }
  NotTransmitting {
   widgetClassPath=&Image
   bIsVariable=true
   parent=Switch_Transmitting
   widget=NotTransmitting
   properties {
    visibility=HitTestInvisible
    brush {
     bIsDynamicallyLoaded=false
     drawAs=Image
     imageType=NoImage
     mirroring=NoMirror
     resourceName=None
     resourceObject=/Game/Athena/HUD/Art/BacchusHudIcons/Mute_Normal
     tiling=NoTile
     imageSize=<32 32>
     margin=<ltrb 0 0 0 0>
     outlineSettings {
      bUseBrushTransparency=false
      roundingType=HalfHeightRadius
      width=0
      color {colorUseRule=UseColor_Specified, specifiedColor=<rgba 0 0 0 0>}
      cornerRadii=<0 0 0 0>
     }
     tintColor {colorUseRule=UseColor_Specified, specifiedColor=<rgba 1 1 1 1>}
     uVRegion {bIsValid=false, max=<0 0>, min=<0 0>}
    }
   }
   slot {class=WidgetSwitcherSlot}
  }
  SizeBox_0 {
   widgetClassPath=&SizeBox
   parent=HorizontalBox_0
   widget=SizeBox_0
   properties {bOverride_MinDesiredWidth=true, minDesiredWidth=72}
   slot {class=HorizontalBoxSlot, properties {horizontalAlignment=HAlign_Right}}
  }
  Overlay_1 {
   widgetClassPath=&Overlay
   bIsVariable=true
   parent=SizeBox_0
   widget=Overlay_1
   properties {renderTransformPivot=<1 0.5>}
   slot {class=SizeBoxSlot, properties {horizontalAlignment=HAlign_Right}}
  }
  SelectionBackplate {
   widgetClassPath=&Image
   bIsVariable=true
   parent=Overlay_1
   widget=SelectionBackplate
   properties {brush {
     bIsDynamicallyLoaded=false
     drawAs=Image
     imageType=NoImage
     mirroring=NoMirror
     resourceName=None
     resourceObject=/Game/UI/SocialPanel/Art/Materials/M_UI_VoiceHeader
     tiling=NoTile
     imageSize=<32 32>
     margin=<ltrb 0 0 0 0>
     outlineSettings {
      bUseBrushTransparency=false
      roundingType=HalfHeightRadius
      width=0
      color {colorUseRule=UseColor_Specified, specifiedColor=<rgba 0 0 0 0>}
      cornerRadii=<0 0 0 0>
     }
     tintColor {colorUseRule=UseColor_Specified, specifiedColor=<rgba 1 1 1 1>}
     uVRegion {bIsValid=false, max=<0 0>, min=<0 0>}
    }}
   slot {class=OverlaySlot, properties {horizontalAlignment=HAlign_Fill, verticalAlignment=VAlign_Fill}}
  }
  HorizontalBox_2 {
   widgetClassPath=&HorizontalBox
   parent=Overlay_1
   widget=HorizontalBox_2
   slot {class=OverlaySlot, properties {horizontalAlignment=HAlign_Center, verticalAlignment=VAlign_Center}}
  }
  HBox_Join {
   widgetClassPath=&HorizontalBox
   bIsVariable=true
   parent=HorizontalBox_2
   widget=HBox_Join
   slot {class=HorizontalBoxSlot, properties {horizontalAlignment=HAlign_Right, verticalAlignment=VAlign_Center, padding=<ltrb 0 0 10 0>}}
  }
  InputActionWidget {
   widgetClassPath=&CommonActionWidget
   bInherited=true
   bIsVariable=true
   parent=HBox_Join
   widget=InputActionWidget
   properties {inputActions [{dataTable=&InputActionDataTable, rowName=GenericForward}]}
   slot {class=HorizontalBoxSlot}
  }
  SimultaneousChannelSwitcher {
   widgetClassPath=&WidgetSwitcher
   bIsVariable=true
   parent=HBox_Join
   widget=SimultaneousChannelSwitcher
   slot {class=HorizontalBoxSlot, properties {verticalAlignment=VAlign_Center, padding=<ltrb 10 0 0 0>}}
  }
  OnOffSwitcher {
   widgetClassPath=&WidgetSwitcher
   bIsVariable=true
   parent=SimultaneousChannelSwitcher
   widget=OnOffSwitcher
   slot {class=WidgetSwitcherSlot, properties {horizontalAlignment=HAlign_Right}}
  }
  OnText {
   widgetClassPath=&CommonRichTextBlock
   bIsVariable=true
   parent=OnOffSwitcher
   widget=OnText
   properties {
    bOverrideDefaultStyle=true
    defaultTextStyleOverrideClass=/FortUILibrary/Styles/Typography/Foundation/TextStyle-BurbankSmall-XS-Cyan.TextStyle-BurbankSmall-XS-Cyan_C
    text=ON
    textTransformPolicy=ToUpper
    defaultTextStyleOverride {
     overflowPolicy=Clip
     transformPolicy=None
     colorAndOpacity {colorUseRule=UseColor_Specified, specifiedColor=<rgba 1 1 1 1>}
     font {
      bForceMonospaced=false
      bMaterialIsStencil=false
      fontMaterial=null
      fontObject=&BurbankSmall
      letterSpacing=0
      monospacedWidth=1
      size=18
      skewAmount=0
      typefaceFontName=Bold
      outlineSettings {
       bApplyOutlineToDropShadows=false
       bMiteredCorners=false
       bSeparateFillAlpha=false
       outlineMaterial=null
       outlineSize=0
       outlineColor=<rgba 0 0 0 1>
      }
     }
     highlightColor {colorUseRule=UseColor_Specified, specifiedColor=<rgba 0 0 0 1>}
     highlightShape {
      bIsDynamicallyLoaded=false
      drawAs=Image
      imageType=NoImage
      mirroring=NoMirror
      resourceName=None
      resourceObject=null
      tiling=NoTile
      imageSize=<32 32>
      margin=<ltrb 0 0 0 0>
      outlineSettings {
       bUseBrushTransparency=false
       roundingType=HalfHeightRadius
       width=0
       color {colorUseRule=UseColor_Specified, specifiedColor=<rgba 0 0 0 0>}
       cornerRadii=<0 0 0 0>
      }
      tintColor {colorUseRule=UseColor_Specified, specifiedColor=<rgba 1 1 1 1>}
      uVRegion {bIsValid=false, max=<0 0>, min=<0 0>}
     }
     selectedBackgroundColor {colorUseRule=UseColor_Foreground, specifiedColor=<rgba 1 0 1 1>}
     shadowColorAndOpacity=<rgba 0 0 0 1>
     shadowOffset=<0 0>
     strikeBrush {
      bIsDynamicallyLoaded=false
      drawAs=Image
      imageType=NoImage
      mirroring=NoMirror
      resourceName=None
      resourceObject=null
      tiling=NoTile
      imageSize=<32 32>
      margin=<ltrb 0 0 0 0>
      outlineSettings {
       bUseBrushTransparency=false
       roundingType=HalfHeightRadius
       width=0
       color {colorUseRule=UseColor_Specified, specifiedColor=<rgba 0 0 0 0>}
       cornerRadii=<0 0 0 0>
      }
      tintColor {colorUseRule=UseColor_Specified, specifiedColor=<rgba 1 1 1 1>}
      uVRegion {bIsValid=false, max=<0 0>, min=<0 0>}
     }
     underlineBrush {
      bIsDynamicallyLoaded=false
      drawAs=Image
      imageType=NoImage
      mirroring=NoMirror
      resourceName=None
      resourceObject=null
      tiling=NoTile
      imageSize=<32 32>
      margin=<ltrb 0 0 0 0>
      outlineSettings {
       bUseBrushTransparency=false
       roundingType=HalfHeightRadius
       width=0
       color {colorUseRule=UseColor_Specified, specifiedColor=<rgba 0 0 0 0>}
       cornerRadii=<0 0 0 0>
      }
      tintColor {colorUseRule=UseColor_Specified, specifiedColor=<rgba 1 1 1 1>}
      uVRegion {bIsValid=false, max=<0 0>, min=<0 0>}
     }
    }
   }
   slot {class=WidgetSwitcherSlot, properties {horizontalAlignment=HAlign_Right}}
  }
  OffText {
   widgetClassPath=&CommonRichTextBlock
   bIsVariable=true
   parent=OnOffSwitcher
   widget=OffText
   properties {
    bOverrideDefaultStyle=true
    defaultTextStyleOverrideClass=/FortUILibrary/Styles/Typography/Foundation/TextStyle-BurbankSmall-XS-Gray.TextStyle-BurbankSmall-XS-Gray_C
    text=OFF
    textTransformPolicy=ToUpper
    defaultTextStyleOverride {
     overflowPolicy=Clip
     transformPolicy=None
     colorAndOpacity {colorUseRule=UseColor_Specified, specifiedColor=<rgba 1 1 1 1>}
     font {
      bForceMonospaced=false
      bMaterialIsStencil=false
      fontMaterial=null
      fontObject=&BurbankSmall
      letterSpacing=0
      monospacedWidth=1
      size=18
      skewAmount=0
      typefaceFontName=Bold
      outlineSettings {
       bApplyOutlineToDropShadows=false
       bMiteredCorners=false
       bSeparateFillAlpha=false
       outlineMaterial=null
       outlineSize=0
       outlineColor=<rgba 0 0 0 1>
      }
     }
     highlightColor {colorUseRule=UseColor_Specified, specifiedColor=<rgba 0 0 0 1>}
     highlightShape {
      bIsDynamicallyLoaded=false
      drawAs=Image
      imageType=NoImage
      mirroring=NoMirror
      resourceName=None
      resourceObject=null
      tiling=NoTile
      imageSize=<32 32>
      margin=<ltrb 0 0 0 0>
      outlineSettings {
       bUseBrushTransparency=false
       roundingType=HalfHeightRadius
       width=0
       color {colorUseRule=UseColor_Specified, specifiedColor=<rgba 0 0 0 0>}
       cornerRadii=<0 0 0 0>
      }
      tintColor {colorUseRule=UseColor_Specified, specifiedColor=<rgba 1 1 1 1>}
      uVRegion {bIsValid=false, max=<0 0>, min=<0 0>}
     }
     selectedBackgroundColor {colorUseRule=UseColor_Foreground, specifiedColor=<rgba 1 0 1 1>}
     shadowColorAndOpacity=<rgba 0 0 0 1>
     shadowOffset=<0 0>
     strikeBrush {
      bIsDynamicallyLoaded=false
      drawAs=Image
      imageType=NoImage
      mirroring=NoMirror
      resourceName=None
      resourceObject=null
      tiling=NoTile
      imageSize=<32 32>
      margin=<ltrb 0 0 0 0>
      outlineSettings {
       bUseBrushTransparency=false
       roundingType=HalfHeightRadius
       width=0
       color {colorUseRule=UseColor_Specified, specifiedColor=<rgba 0 0 0 0>}
       cornerRadii=<0 0 0 0>
      }
      tintColor {colorUseRule=UseColor_Specified, specifiedColor=<rgba 1 1 1 1>}
      uVRegion {bIsValid=false, max=<0 0>, min=<0 0>}
     }
     underlineBrush {
      bIsDynamicallyLoaded=false
      drawAs=Image
      imageType=NoImage
      mirroring=NoMirror
      resourceName=None
      resourceObject=null
      tiling=NoTile
      imageSize=<32 32>
      margin=<ltrb 0 0 0 0>
      outlineSettings {
       bUseBrushTransparency=false
       roundingType=HalfHeightRadius
       width=0
       color {colorUseRule=UseColor_Specified, specifiedColor=<rgba 0 0 0 0>}
       cornerRadii=<0 0 0 0>
      }
      tintColor {colorUseRule=UseColor_Specified, specifiedColor=<rgba 1 1 1 1>}
      uVRegion {bIsValid=false, max=<0 0>, min=<0 0>}
     }
    }
   }
   slot {class=WidgetSwitcherSlot, properties {horizontalAlignment=HAlign_Right}}
  }
  SingleChannelSwitcher {widgetClassPath=&HorizontalBox, parent=SimultaneousChannelSwitcher, widget=SingleChannelSwitcher, slot {class=WidgetSwitcherSlot}}
  JoinText {
   widgetClassPath=&CommonRichTextBlock
   bIsVariable=true
   parent=SingleChannelSwitcher
   widget=JoinText
   properties {
    bOverrideDefaultStyle=true
    text=Join
    textTransformPolicy=ToUpper
    defaultTextStyleOverride {
     overflowPolicy=Clip
     transformPolicy=None
     colorAndOpacity {colorUseRule=UseColor_Specified, specifiedColor=<rgba 1 1 1 1>}
     font {
      bForceMonospaced=false
      bMaterialIsStencil=false
      fontMaterial=null
      fontObject=&BurbankSmall
      letterSpacing=0
      monospacedWidth=1
      size=18
      skewAmount=0
      typefaceFontName=Bold
      outlineSettings {
       bApplyOutlineToDropShadows=false
       bMiteredCorners=false
       bSeparateFillAlpha=false
       outlineMaterial=null
       outlineSize=0
       outlineColor=<rgba 0 0 0 1>
      }
     }
     highlightColor {colorUseRule=UseColor_Specified, specifiedColor=<rgba 0 0 0 1>}
     highlightShape {
      bIsDynamicallyLoaded=false
      drawAs=Image
      imageType=NoImage
      mirroring=NoMirror
      resourceName=None
      resourceObject=null
      tiling=NoTile
      imageSize=<32 32>
      margin=<ltrb 0 0 0 0>
      outlineSettings {
       bUseBrushTransparency=false
       roundingType=HalfHeightRadius
       width=0
       color {colorUseRule=UseColor_Specified, specifiedColor=<rgba 0 0 0 0>}
       cornerRadii=<0 0 0 0>
      }
      tintColor {colorUseRule=UseColor_Specified, specifiedColor=<rgba 1 1 1 1>}
      uVRegion {bIsValid=false, max=<0 0>, min=<0 0>}
     }
     selectedBackgroundColor {colorUseRule=UseColor_Foreground, specifiedColor=<rgba 1 0 1 1>}
     shadowColorAndOpacity=<rgba 0 0 0 1>
     shadowOffset=<0 0>
     strikeBrush {
      bIsDynamicallyLoaded=false
      drawAs=Image
      imageType=NoImage
      mirroring=NoMirror
      resourceName=None
      resourceObject=null
      tiling=NoTile
      imageSize=<32 32>
      margin=<ltrb 0 0 0 0>
      outlineSettings {
       bUseBrushTransparency=false
       roundingType=HalfHeightRadius
       width=0
       color {colorUseRule=UseColor_Specified, specifiedColor=<rgba 0 0 0 0>}
       cornerRadii=<0 0 0 0>
      }
      tintColor {colorUseRule=UseColor_Specified, specifiedColor=<rgba 1 1 1 1>}
      uVRegion {bIsValid=false, max=<0 0>, min=<0 0>}
     }
     underlineBrush {
      bIsDynamicallyLoaded=false
      drawAs=Image
      imageType=NoImage
      mirroring=NoMirror
      resourceName=None
      resourceObject=null
      tiling=NoTile
      imageSize=<32 32>
      margin=<ltrb 0 0 0 0>
      outlineSettings {
       bUseBrushTransparency=false
       roundingType=HalfHeightRadius
       width=0
       color {colorUseRule=UseColor_Specified, specifiedColor=<rgba 0 0 0 0>}
       cornerRadii=<0 0 0 0>
      }
      tintColor {colorUseRule=UseColor_Specified, specifiedColor=<rgba 1 1 1 1>}
      uVRegion {bIsValid=false, max=<0 0>, min=<0 0>}
     }
    }
   }
   slot {class=HorizontalBoxSlot}
  }
  ExpandDirectionTick {
   widgetClassPath=/Script/FortniteUI.FortMobileImage
   bIsVariable=true
   parent=SingleChannelSwitcher
   widget=ExpandDirectionTick
   properties {
    mobileSizeMultiplier=1.2
    renderOpacity=0.3
    brush {
     bIsDynamicallyLoaded=false
     drawAs=Image
     imageType=NoImage
     mirroring=NoMirror
     resourceName=None
     resourceObject=/Game/UI/SocialPanel/Art/Textures/T_UI_Icon_TreeArrow_64
     tiling=NoTile
     imageSize=<28 28>
     margin=<ltrb 0 0 0 0>
     outlineSettings {
      bUseBrushTransparency=false
      roundingType=HalfHeightRadius
      width=0
      color {colorUseRule=UseColor_Specified, specifiedColor=<rgba 0 0 0 0>}
      cornerRadii=<0 0 0 0>
     }
     tintColor {colorUseRule=UseColor_Specified, specifiedColor=<rgba 1 1 1 1>}
     uVRegion {bIsValid=false, max=<0 0>, min=<0 0>}
    }
    renderTransform {angle=90, scale=<1 1>, shear=<0 0>, translation=<0 0>}
   }
   slot {class=HorizontalBoxSlot}
  }
  ChannelInfoPanel {
   widgetClassPath=&HorizontalBox
   parent=VerticalBox_0
   widget=ChannelInfoPanel
   slot {class=VerticalBoxSlot, properties {padding=<ltrb 0 -5 0 0>}}
  }
  Image_VoiceChatWarning {
   widgetClassPath=&Image
   bIsVariable=true
   parent=ChannelInfoPanel
   widget=Image_VoiceChatWarning
   properties {
    visibility=Collapsed
    brush {
     bIsDynamicallyLoaded=false
     drawAs=Image
     imageType=NoImage
     mirroring=NoMirror
     resourceName=None
     resourceObject=/Game/UI/SocialPanel/Art/Textures/T_UI_Icon_MicLocked
     tiling=NoTile
     imageSize=<25 25>
     margin=<ltrb 0 0 0 0>
     outlineSettings {
      bUseBrushTransparency=false
      roundingType=HalfHeightRadius
      width=0
      color {colorUseRule=UseColor_Specified, specifiedColor=<rgba 0 0 0 0>}
      cornerRadii=<0 0 0 0>
     }
     tintColor {colorUseRule=UseColor_Specified, specifiedColor=<rgba 1 1 1 1>}
     uVRegion {bIsValid=false, max=<0 0>, min=<0 0>}
    }
    colorAndOpacity=<rgba 1 0.8963 0.1046 1>
   }
   slot {class=HorizontalBoxSlot, properties {horizontalAlignment=HAlign_Left, verticalAlignment=VAlign_Center}}
  }
  Text_ChannelInfo {
   widgetClassPath=&CommonTextBlock
   bInherited=true
   parent=ChannelInfoPanel
   widget=Text_ChannelInfo
   properties {
    autoWrapText=true
    text=You are in this channel
    textTransformPolicy=ToUpper
    colorAndOpacity {colorUseRule=UseColor_Specified, specifiedColor=<rgba 0.3231 0.8308 1 1>}
    font {
     bForceMonospaced=false
     bMaterialIsStencil=false
     fontMaterial=null
     fontObject=&BurbankSmall
     letterSpacing=0
     monospacedWidth=1
     size=15
     skewAmount=0
     typefaceFontName=Bold
     outlineSettings {
      bApplyOutlineToDropShadows=false
      bMiteredCorners=false
      bSeparateFillAlpha=false
      outlineMaterial=null
      outlineSize=0
      outlineColor=<rgba 0 0 0 1>
     }
    }
    shadowOffset=<0 0>
   }
   slot {class=HorizontalBoxSlot, properties {size {sizeRule=Fill, value=1}}}
  }
  JoinLeaveButton {widgetClassPath=/Script/FortniteUI.FortMainMenuButton, bInherited=true, widget=null}
  ReportButton {widgetClassPath=&FortMainMenuButton, bInherited=true, widget=null}
  Text_ChannelTitle_Hover {widgetClassPath=&CommonRichTextBlock, bInherited=true, widget=null}
  Text_ChannelInfo_Hover {widgetClassPath=&CommonTextBlock, bInherited=true, widget=null}
 ]
}
