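# Social panel view widget with Add Friends, Party Up and Voice Chat tabs below the local player's entry.
# Switcher_UserLists holds each tab's tree lists and plays SwitcherTransition on change; nested switchers cover search and voice-chat states.
# FortSocialPanelView_UserLists is the native parent; the graph forwards audio-moderation button clicks to it and applies the text it supplies.
# Construct and design-time PreConstruct set the mic-disabled image's material to muted.
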WidgetBlueprint /FortniteCoreUI/UI/SocialPanel/SocialPanelView_UserLists
animations [
 Open {
  binding_count=1
  display_label=""
  length_seconds=0.15
  object=&self:Open
  sequence {bindings [SocialPanelView_UserLists_C_72 {
     id=#1
     possessed_class=SocialPanelView_UserLists_C
     tracks [
      {
       type=MovieScene2DTransformTrack
       sections [{
         type=MovieScene2DTransformSection
         end_frame=3
         start_frame=0
         channels [
          Translation.X {type=MovieSceneScriptingFloatChannel, keys [[0,-80],[3,0]]}
          Translation.Y {type=MovieSceneScriptingFloatChannel, const [0,2]}
          Angle {type=MovieSceneScriptingFloatChannel, const [0,1]}
          Scale.X {type=MovieSceneScriptingFloatChannel, keys [[0,1],[2,1.0055],[3,1]]}
          Scale.Y {type=MovieSceneScriptingFloatChannel, const [1,1]}
          Shear.X {type=MovieSceneScriptingFloatChannel, const [0,1]}
          Shear.Y {type=MovieSceneScriptingFloatChannel, const [0,1]}
         ]
        }]
      }
      {
       type=MovieSceneFloatTrack
       display_name=Render Opacity
       sections [{type=MovieSceneFloatSection, end_frame=3, start_frame=0, channels [None {type=MovieSceneScriptingFloatChannel, keys [[0,0],[3,1]]}]}]
      }
     ]
    }]}
 }
 Open_0 {
  binding_count=1
  display_label=""
  length_seconds=0.2
  object=&self:Open_0
  sequence {bindings [SocialPanelView_UserLists_C_72 {
     id=#1
     possessed_class=SocialPanelView_UserLists_C
     tracks [
      {
       type=MovieScene2DTransformTrack
       sections [{
         type=MovieScene2DTransformSection
         end_frame=4
         start_frame=0
         channels [
          Translation.X {type=MovieSceneScriptingFloatChannel}
          Translation.Y {type=MovieSceneScriptingFloatChannel}
          Angle {type=MovieSceneScriptingFloatChannel, keys [[0,22.7572],[3.1633,-1.471],[3.9843,0]]}
          Scale.X {type=MovieSceneScriptingFloatChannel, keys [[0,0.4169],[3.1633,1.0811],[3.9843,1]]}
          Scale.Y {type=MovieSceneScriptingFloatChannel, const [1,2]}
          Shear.X {type=MovieSceneScriptingFloatChannel, keys [[0,8.7787],[3.1633,0.3397],[3.9843,0],[4,0]]}
          Shear.Y {type=MovieSceneScriptingFloatChannel, const [0,2]}
         ]
        }]
      }
      {
       type=MovieSceneFloatTrack
       display_name=Render Opacity
       sections [{
         type=MovieSceneFloatSection
         end_frame=4
         start_frame=0
         channels [None {type=MovieSceneScriptingFloatChannel, keys [[0,0],[3.5263,1.0001],[4,1]]}]
        }]
      }
     ]
    }]}
 }
 MenuAnchorItemOpen {
  binding_count=0
  display_label=""
  length_seconds=0.15
  object=&self:MenuAnchorItemOpen
  sequence {bindings []}
 }
 NoFriendsArrowBounce {
  binding_count=1
  display_label=""
  length_seconds=1.6
  object=&self:NoFriendsArrowBounce
  sequence {bindings [NoFriendsArrow {
     id=#2
     possessed_class=Image
     tracks [{
       type=MovieScene2DTransformTrack
       sections [{
         type=MovieScene2DTransformSection
         end_frame=32
         start_frame=0
         channels [
          Translation.X {type=MovieSceneScriptingFloatChannel, keys [[0,32],[16,16],[32,32]]}
          Translation.Y {type=MovieSceneScriptingFloatChannel, keys [[0,-32],[16,-16],[32.0003,-32]]}
          Angle {type=MovieSceneScriptingFloatChannel}
          Scale.X {type=MovieSceneScriptingFloatChannel, keys [[0,1],[16,0.75],[32,1]]}
          Scale.Y {type=MovieSceneScriptingFloatChannel, keys [[0,1],[16,0.75],[32,1]]}
          Shear.X {type=MovieSceneScriptingFloatChannel}
          Shear.Y {type=MovieSceneScriptingFloatChannel}
         ]
        }]
      }]
    }]}
 }
 SwitcherTransition {
  binding_count=6
  display_label=""
  length_seconds=0.3
  object=&self:SwitcherTransition
  sequence {bindings [
    MyFriendsGhostBorder {
     id=#3
     possessed_class=Border
     tracks [
      {
       type=MovieSceneFloatTrack
       display_name=Render Opacity
       sections [{type=MovieSceneFloatSection, channels [None {type=MovieSceneScriptingFloatChannel, keys [[0,0],[6,1]]}]}]
      }
      {
       type=MovieScene2DTransformTrack
       sections [{
         type=MovieScene2DTransformSection
         channels [
          Translation.X {type=MovieSceneScriptingFloatChannel, keys [[0,0],[6,352]]}
          Translation.Y {type=MovieSceneScriptingFloatChannel, keys [[0,0],[6,32]]}
          Angle {type=MovieSceneScriptingFloatChannel}
          Scale.X {type=MovieSceneScriptingFloatChannel, keys [[0,1],[6,0.25]]}
          Scale.Y {type=MovieSceneScriptingFloatChannel}
          Shear.X {type=MovieSceneScriptingFloatChannel}
          Shear.Y {type=MovieSceneScriptingFloatChannel, keys [[0,0],[6,16]]}
         ]
        }]
      }
     ]
    }
    VoiceChatGhostBorder {
     id=#4
     possessed_class=Border
     tracks [
      {
       type=MovieSceneFloatTrack
       display_name=Render Opacity
       sections [{type=MovieSceneFloatSection, channels [None {type=MovieSceneScriptingFloatChannel, keys [[0,0],[6,1]]}]}]
      }
      {
       type=MovieScene2DTransformTrack
       sections [{
         type=MovieScene2DTransformSection
         channels [
          Translation.X {type=MovieSceneScriptingFloatChannel, keys [[0,0],[6,352]]}
          Translation.Y {type=MovieSceneScriptingFloatChannel, keys [[0,0],[6,32]]}
          Angle {type=MovieSceneScriptingFloatChannel}
          Scale.X {type=MovieSceneScriptingFloatChannel, keys [[0,1],[6,0.25]]}
          Scale.Y {type=MovieSceneScriptingFloatChannel}
          Shear.X {type=MovieSceneScriptingFloatChannel}
          Shear.Y {type=MovieSceneScriptingFloatChannel, keys [[0,0],[6,16]]}
         ]
        }]
      }
     ]
    }
    AddFriendsGhostBorder {
     id=#5
     possessed_class=Border
     tracks [
      {
       type=MovieSceneFloatTrack
       display_name=Render Opacity
       sections [{type=MovieSceneFloatSection, channels [None {type=MovieSceneScriptingFloatChannel, keys [[0,0],[6,1]]}]}]
      }
      {
       type=MovieScene2DTransformTrack
       sections [{
         type=MovieScene2DTransformSection
         channels [
          Translation.X {type=MovieSceneScriptingFloatChannel, keys [[0,0],[6,352]]}
          Translation.Y {type=MovieSceneScriptingFloatChannel, keys [[0,0],[6,32]]}
          Angle {type=MovieSceneScriptingFloatChannel}
          Scale.X {type=MovieSceneScriptingFloatChannel, keys [[0,1],[6,0.25]]}
          Scale.Y {type=MovieSceneScriptingFloatChannel}
          Shear.X {type=MovieSceneScriptingFloatChannel}
          Shear.Y {type=MovieSceneScriptingFloatChannel, keys [[0,0],[6,16]]}
         ]
        }]
      }
     ]
    }
    VBox_AddFriends {
     id=#6
     possessed_class=VerticalBox
     tracks [
      {
       type=MovieScene2DTransformTrack
       sections [{
         type=MovieScene2DTransformSection
         channels [
          Translation.X {type=MovieSceneScriptingFloatChannel, keys [[0,-352],[6,0]]}
          Translation.Y {type=MovieSceneScriptingFloatChannel, keys [[0,32],[6,0]]}
          Angle {type=MovieSceneScriptingFloatChannel}
          Scale.X {type=MovieSceneScriptingFloatChannel, keys [[0,0.25],[6,1]]}
          Scale.Y {type=MovieSceneScriptingFloatChannel}
          Shear.X {type=MovieSceneScriptingFloatChannel, const [0,1]}
          Shear.Y {type=MovieSceneScriptingFloatChannel, keys [[0,-16],[6,0]]}
         ]
        }]
      }
      {
       type=MovieSceneFloatTrack
       display_name=Render Opacity
       sections [{type=MovieSceneFloatSection, channels [None {type=MovieSceneScriptingFloatChannel, keys [[0,0],[6,1]]}]}]
      }
     ]
    }
    UserTree_FriendsLists {
     id=#7
     possessed_class=SocialUserTreeView
     tracks [
      {
       type=MovieScene2DTransformTrack
       sections [{
         type=MovieScene2DTransformSection
         channels [
          Translation.X {type=MovieSceneScriptingFloatChannel, keys [[0,-352],[6,0]]}
          Translation.Y {type=MovieSceneScriptingFloatChannel, keys [[0,32],[6,0]]}
          Angle {type=MovieSceneScriptingFloatChannel}
          Scale.X {type=MovieSceneScriptingFloatChannel, keys [[0,0.25],[6,1]]}
          Scale.Y {type=MovieSceneScriptingFloatChannel}
          Shear.X {type=MovieSceneScriptingFloatChannel, const [0,1]}
          Shear.Y {type=MovieSceneScriptingFloatChannel, keys [[0,-16],[6,0]]}
         ]
        }]
      }
      {
       type=MovieSceneFloatTrack
       display_name=Render Opacity
       sections [{type=MovieSceneFloatSection, channels [None {type=MovieSceneScriptingFloatChannel, keys [[0,0],[6,1]]}]}]
      }
     ]
    }
    UserTree_VoiceChat {
     id=#8
     display_name=VoiceChatOvr
     possessed_class=FortSocialUserTreeView
     tracks [
      {
       type=MovieScene2DTransformTrack
       sections [{
         type=MovieScene2DTransformSection
         channels [
          Translation.X {type=MovieSceneScriptingFloatChannel, keys [[0,-352],[6,0]]}
          Translation.Y {type=MovieSceneScriptingFloatChannel, keys [[0,32],[6,0]]}
          Angle {type=MovieSceneScriptingFloatChannel}
          Scale.X {type=MovieSceneScriptingFloatChannel, keys [[0,0.25],[6,1]]}
          Scale.Y {type=MovieSceneScriptingFloatChannel}
          Shear.X {type=MovieSceneScriptingFloatChannel, const [0,1]}
          Shear.Y {type=MovieSceneScriptingFloatChannel, keys [[0,-16],[6,0]]}
         ]
        }]
      }
      {
       type=MovieSceneFloatTrack
       display_name=Render Opacity
       sections [{type=MovieSceneFloatSection, channels [None {type=MovieSceneScriptingFloatChannel, keys [[0,0],[6,1]]}]}]
      }
     ]
    }
   ]}
 }
 WeGameButtonHovered {
  binding_count=0
  length_seconds=5
  object=&self:WeGameButtonHovered
  sequence {bindings []}
 }
]
graphs [EventGraph {nodes [
   "On Active Widget Changed (Switcher_UserLists)"!() -> (OutputDelegate, then, ActiveWidget:Widget Object Reference, ActiveWidgetIndex:int=0) {is=title K2Node_ComponentBoundEvent}
   > PlayAnimation(InAnimation=Get SwitcherTransition() -> (SwitcherTransition:Widget Animation Object Reference))
   "On Clicked (WBP_UIKit_Button_TurnOn)"!() -> (OutputDelegate, then, Button:Common Button Base Object Reference) {is=title K2Node_ComponentBoundEvent}
   > HandleAcceptAudioModerationFromTutorial(self:Fort Social Panel View User Lists Object Reference)
   "On Clicked (WBP_UIKit_Button_LearnMore)"!() -> (OutputDelegate, then, Button:Common Button Base Object Reference) {is=title K2Node_ComponentBoundEvent}
   > HandleLearnMoreAudioModerationFromTutorial(self:Fort Social Panel View User Lists Object Reference)
   n1 = Event OnSetVoiceChannelButtonText!() -> (OutputDelegate, then, ButtonText:"Text (by ref)") {is=title K2Node_Event}
   > n2 = Cast To LeavePlatformPartyButton!(execute, Object=Get Button_LeavePlatformVoiceChat() -> (Button_LeavePlatformVoiceChat:Common Button Legacy Object Reference)) -> (then, CastFailed, AsLeave Platform Party Button:Leave Platform Party Button Object Reference) {is=title K2Node_DynamicCast}
   > SetText!(execute, self:Leave Platform Party Button Object Reference=@n2, New Text:Text=@n1.ButtonText) -> (then)
   n3 = Event Construct()
   n4 = GetDynamicMaterial(execute=[@n3.then, @n6.then], self=Get Image_MicDisabled() -> (Image_MicDisabled:Image Object Reference))
   > SetScalarParameterValue(self=@n4, ParameterName=IsSelfMuted, Value=1)
   n5 = Event PreConstruct()
   > n6 = Branch(Condition=@n5.IsDesignTime)
   n7 = Event UpdateAudioModerationAcceptPanelText!() -> (OutputDelegate, then, Title:"Text (by ref)", Description:"Text (by ref)") {is=title K2Node_Event}
   > SetText!(
    execute
    self:Rich Text Block Object Reference=Get VoiceChat_AcceptTitle() -> (VoiceChat_AcceptTitle:Common Rich Text Block Object Reference)
    InText:"Text (by ref)"=@n7.Title
   ) -> (then)
   > SetText(self=Get VoiceChat_AcceptDescription() -> (VoiceChat_AcceptDescription:Common Rich Text Block Object Reference), InText=@n7.Description)
  ]}]
variables []
widgets {
 info {inheritedWidgetCount=27, parentClass=/Script/FortniteUI.FortSocialPanelView_UserLists, rootWidgetClass=&Border, widgetCount=84}
 widgets [
  Border_18 {
   widgetClassPath=&Border
   widget=Border_18
   properties {
    background {
     bIsDynamicallyLoaded=false
     drawAs=NoDrawType
     imageType=NoImage
     mirroring=NoMirror
     resourceName=None
     resourceObject=null
     tiling=NoTile
     imageSize=<32 32>
     margin=<ltrb 0 0 0 0>
     outlineSettings {
      bUseBrushTransparency=false
      roundingType=HalfHeightRadius
      width=0
      color {colorUseRule=UseColor_Specified, specifiedColor=<rgba 0 0 0 0>}
      cornerRadii=<0 0 0 0>
     }
     tintColor {colorUseRule=UseColor_Specified, specifiedColor=<rgba 1 1 1 1>}
     uVRegion {bIsValid=false, max=<0 0>, min=<0 0>}
    }
    padding=<ltrb 0 0 0 0>
   }
  }
  VerticalBox_1 {
   widgetClassPath=&VerticalBox
   parent=Border_18
   widget=VerticalBox_1
   slot {class=BorderSlot, properties {padding=<ltrb 0 0 0 0>}}
  }
  HorizontalBox {
   widgetClassPath=&HorizontalBox
   parent=VerticalBox_1
   widget=HorizontalBox
   slot {class=VerticalBoxSlot, properties {horizontalAlignment=HAlign_Center, verticalAlignment=VAlign_Center, padding=<ltrb 32 16 32 16>}}
  }
  Image_TitleIcon {
   widgetClassPath=/Script/FortniteUI.FortMobileImage
   bIsVariable=true
   parent=HorizontalBox
   widget=Image_TitleIcon
   properties {
    mobileSizeMultiplier=1.2
    visibility=SelfHitTestInvisible
    brush {
     bIsDynamicallyLoaded=false
     drawAs=Image
     imageType=NoImage
     mirroring=NoMirror
     resourceName=None
     resourceObject=/FortniteCoreUI/UI/SocialPanel/Art/Textures/T_UI_Icon_MyFriends_64
     tiling=NoTile
     imageSize=<32 32>
     margin=<ltrb 0 0 0 0>
     outlineSettings {
      bUseBrushTransparency=false
      roundingType=HalfHeightRadius
      width=0
      color {colorUseRule=UseColor_Specified, specifiedColor=<rgba 0 0 0 0>}
      cornerRadii=<0 0 0 0>
     }
     tintColor {colorUseRule=UseColor_Specified, specifiedColor=<rgba 0.0331 0.402 1 1>}
     uVRegion {bIsValid=false, max=<0 0>, min=<0 0>}
    }
   }
   slot {class=HorizontalBoxSlot, properties {verticalAlignment=VAlign_Center, padding=<ltrb 0 0 8 0>}}
  }
  Text_LowPowerMode {
   widgetClassPath=&CommonTextBlock
   bIsVariable=true
   parent=HorizontalBox
   widget=Text_LowPowerMode
   properties {
    justification=Center
    mobileFontSizeMultiplier=1.34
    text=Social
    textTransformPolicy=ToUpper
    colorAndOpacity {colorUseRule=UseColor_Specified, specifiedColor=<rgba 0.0331 0.402 1 1>}
    font {
     bForceMonospaced=false
     bMaterialIsStencil=false
     fontMaterial=null
     fontObject=&BurbankSmall
     letterSpacing=0
     monospacedWidth=1
     size=24
     skewAmount=0
     typefaceFontName=Black
     outlineSettings {
      bApplyOutlineToDropShadows=false
      bMiteredCorners=false
      bSeparateFillAlpha=false
      outlineMaterial=null
      outlineSize=0
      outlineColor=<rgba 0 0 0 1>
     }
    }
    shadowOffset=<0 0>
   }
   slot {class=HorizontalBoxSlot, properties {size {sizeRule=Fill, value=1}}}
  }
  UserEntry_LocalPlayer {
   widgetClassPath=/FortniteCoreUI/UI/SocialPanel/ListEntries/LocalUserEntry
   bInherited=true
   bIsVariable=true
   parent=VerticalBox_1
   widget=UserEntry_LocalPlayer
   properties {
    onButtonBaseClicked=(&self:WidgetTree.UserEntry_LocalPlayer.HandleOnButtonClicked)
    onButtonBaseDoubleClicked=(&self:WidgetTree.UserEntry_LocalPlayer.HandleOnButtonDoubleClicked)
    onButtonBaseHovered=(&self:WidgetTree.UserEntry_LocalPlayer.HandleOnButtonHovered)
    onButtonBaseUnhovered=(&self:WidgetTree.UserEntry_LocalPlayer.HandleOnButtonUnhovered)
    onSelectedChangedBase=(&self:WidgetTree.UserEntry_LocalPlayer.HandleOnSelectedChanged)
    triggeringInputAction {dataTable=&InputActionDataTable, rowName=GenericPreview}
   }
   slot {class=VerticalBoxSlot, properties {padding=<ltrb 0 0 0 4>}}
  }
  TabToggleSB {
   widgetClassPath=&SizeBox
   parent=VerticalBox_1
   widget=TabToggleSB
   properties {bOverride_MinDesiredHeight=true, heightOverride=96, minDesiredHeight=96}
   slot {class=VerticalBoxSlot, properties {padding=<ltrb 0 0 0 8>}}
  }
  Overlay_3 {widgetClassPath=&Overlay, parent=TabToggleSB, widget=Overlay_3, slot {class=SizeBoxSlot}}
  HorizontalBox_0 {
   widgetClassPath=&HorizontalBox
   parent=Overlay_3
   widget=HorizontalBox_0
   slot {class=OverlaySlot, properties {horizontalAlignment=HAlign_Fill, verticalAlignment=VAlign_Fill, padding=<ltrb 6 0 6 0>}}
  }
  FortMobileImage_127 {
   widgetClassPath=&FortMobileImage
   bIsVariable=true
   parent=HorizontalBox_0
   widget=FortMobileImage_127
   properties {
    mobileSizeMultiplier=1.2
    brush {
     bIsDynamicallyLoaded=false
     drawAs=Image
     imageType=NoImage
     mirroring=NoMirror
     resourceName=None
     resourceObject=null
     tiling=NoTile
     imageSize=<0.1 96>
     margin=<ltrb 0 0 0 0>
     outlineSettings {
      bUseBrushTransparency=false
      roundingType=HalfHeightRadius
      width=0
      color {colorUseRule=UseColor_Specified, specifiedColor=<rgba 0 0 0 0>}
      cornerRadii=<0 0 0 0>
     }
     tintColor {colorUseRule=UseColor_Specified, specifiedColor=<rgba 1 1 1 0>}
     uVRegion {bIsValid=false, max=<0 0>, min=<0 0>}
    }
   }
   slot {class=HorizontalBoxSlot, properties {horizontalAlignment=HAlign_Left, verticalAlignment=VAlign_Top}}
  }
  SocialTab_AddFriends {
   widgetClassPath=&/SocialPanelTab
   bInherited=true
   bIsVariable=true
   parent=HorizontalBox_0
   widget=SocialTab_AddFriends
   properties {
    clickMethod=DownAndUp
    onButtonBaseClicked=(&self:WidgetTree.SocialTab_AddFriends.HandleOnButtonClicked)
    onButtonBaseDoubleClicked=(&self:WidgetTree.SocialTab_AddFriends.HandleOnButtonDoubleClicked)
    onButtonBaseHovered=(&self:WidgetTree.SocialTab_AddFriends.HandleOnButtonHovered)
    onButtonBaseUnhovered=(&self:WidgetTree.SocialTab_AddFriends.HandleOnButtonUnhovered)
    onSelectedChangedBase=(&self:WidgetTree.SocialTab_AddFriends.HandleOnSelectedChanged)
    pressMethod=DownAndUp
    tabName=ADD FRIENDS
    inviteTintColor {colorUseRule=UseColor_Specified, specifiedColor=<rgba 0.9473 0.9911 0.0666 1>}
   }
   slot {class=HorizontalBoxSlot, properties {size {sizeRule=Fill, value=1}}}
  }
  SocialTab_Friends {
   widgetClassPath=&SocialPanelTab
   bInherited=true
   bIsVariable=true
   parent=HorizontalBox_0
   widget=SocialTab_Friends
   properties {
    clickMethod=DownAndUp
    onButtonBaseClicked=(&self:WidgetTree.SocialTab_Friends.HandleOnButtonClicked)
    onButtonBaseDoubleClicked=(&self:WidgetTree.SocialTab_Friends.HandleOnButtonDoubleClicked)
    onButtonBaseHovered=(&self:WidgetTree.SocialTab_Friends.HandleOnButtonHovered)
    onButtonBaseUnhovered=(&self:WidgetTree.SocialTab_Friends.HandleOnButtonUnhovered)
    onSelectedChangedBase=(&self:WidgetTree.SocialTab_Friends.HandleOnSelectedChanged)
    pressMethod=DownAndUp
    tabIcon=&T_UI_Icon_MyFriends_64
    tabName=PARTY UP
    inviteTintColor {colorUseRule=UseColor_Specified, specifiedColor=<rgba 0.9473 0.9911 0.0666 1>}
   }
   slot {class=HorizontalBoxSlot, properties {size {sizeRule=Fill, value=1}}}
  }
  SocialTab_VoiceChat {
   widgetClassPath=&SocialPanelTab
   bInherited=true
   bIsVariable=true
   parent=HorizontalBox_0
   widget=SocialTab_VoiceChat
   properties {
    clickMethod=DownAndUp
    onButtonBaseClicked=(&self:WidgetTree.SocialTab_VoiceChat.HandleOnButtonClicked)
    onButtonBaseDoubleClicked=(&self:WidgetTree.SocialTab_VoiceChat.HandleOnButtonDoubleClicked)
    onButtonBaseHovered=(&self:WidgetTree.SocialTab_VoiceChat.HandleOnButtonHovered)
    onButtonBaseUnhovered=(&self:WidgetTree.SocialTab_VoiceChat.HandleOnButtonUnhovered)
    onSelectedChangedBase=(&self:WidgetTree.SocialTab_VoiceChat.HandleOnSelectedChanged)
    pressMethod=DownAndUp
    tabIcon=/Game/UI/SocialPanel/Art/Textures/T_UI_Icon_Mic_64
    tabName=VOICE CHAT
    inviteTintColor {colorUseRule=UseColor_Specified, specifiedColor=<rgba 0.9473 0.9911 0.0666 1>}
   }
   slot {class=HorizontalBoxSlot, properties {size {sizeRule=Fill, value=1}}}
  }
  InputActionWidget_Left {
   widgetClassPath=&CommonActionWidget
   parent=Overlay_3
   widget=InputActionWidget_Left
   properties {visibility=SelfHitTestInvisible, inputActions [{dataTable=&InputActionDataTable, rowName=MainTabLeft}]}
   slot {class=OverlaySlot, properties {verticalAlignment=VAlign_Center, padding=<ltrb 12 0 0 20>}}
  }
  InputActionWidget_Right {
   widgetClassPath=&CommonActionWidget
   parent=Overlay_3
   widget=InputActionWidget_Right
   properties {visibility=SelfHitTestInvisible, inputActions [{dataTable=&InputActionDataTable, rowName=MainTabRight}]}
   slot {class=OverlaySlot, properties {horizontalAlignment=HAlign_Right, verticalAlignment=VAlign_Center, padding=<ltrb 0 0 12 20>}}
  }
  NavStopBorder {
   widgetClassPath=&Border
   parent=VerticalBox_1
   widget=NavStopBorder
   properties {
    clipping=ClipToBounds
    navigation=&self:WidgetTree.NavStopBorder.WidgetNavigation_0
    background {
     bIsDynamicallyLoaded=false
     drawAs=NoDrawType
     imageType=NoImage
     mirroring=NoMirror
     resourceName=None
     resourceObject=null
     tiling=NoTile
     imageSize=<32 32>
     margin=<ltrb 0 0 0 0>
     outlineSettings {
      bUseBrushTransparency=false
      roundingType=HalfHeightRadius
      width=0
      color {colorUseRule=UseColor_Specified, specifiedColor=<rgba 0 0 0 0>}
      cornerRadii=<0 0 0 0>
     }
     tintColor {colorUseRule=UseColor_Specified, specifiedColor=<rgba 1 1 1 1>}
     uVRegion {bIsValid=false, max=<0 0>, min=<0 0>}
    }
    padding=<ltrb 0 0 0 0>
   }
   slot {class=VerticalBoxSlot, properties {size {sizeRule=Fill, value=1}}}
  }
  Switcher_UserLists {
   widgetClassPath=&CommonWidgetSwitcherLegacy
   bInherited=true
   bIsVariable=true
   parent=NavStopBorder
   widget=Switcher_UserLists
   properties {bIsVolatile=true, bOutroPanelBelow=false, transitionCurveType=Linear, transitionDuration=0, transitionType=Horizontal}
   slot {class=BorderSlot, properties {padding=<ltrb 0 0 0 0>}}
  }
  AddFriendsOvr {widgetClassPath=&Overlay, parent=Switcher_UserLists, widget=AddFriendsOvr, slot {class=WidgetSwitcherSlot}}
  AddFriendsGhostBorder {
   widgetClassPath=&Border
   parent=AddFriendsOvr
   widget=AddFriendsGhostBorder
   properties {
    background {
     bIsDynamicallyLoaded=false
     drawAs=Image
     imageType=NoImage
     mirroring=NoMirror
     resourceName=None
     resourceObject=/FortniteCoreUI/UI/SocialPanel/Art/Materials/M_UI_SocialUserDarkPanel
     tiling=NoTile
     imageSize=<32 32>
     margin=<ltrb 0 0 0 0>
     outlineSettings {
      bUseBrushTransparency=false
      roundingType=HalfHeightRadius
      width=0
      color {colorUseRule=UseColor_Specified, specifiedColor=<rgba 0 0 0 0>}
      cornerRadii=<0 0 0 0>
     }
     tintColor {colorUseRule=UseColor_Specified, specifiedColor=<rgba 1 1 1 1>}
     uVRegion {bIsValid=false, max=<0 0>, min=<0 0>}
    }
    padding=<ltrb 0 0 0 0>
   }
   slot {class=OverlaySlot, properties {horizontalAlignment=HAlign_Fill, verticalAlignment=VAlign_Fill}}
  }
  VBox_AddFriends {
   widgetClassPath=&VerticalBox
   parent=AddFriendsOvr
   widget=VBox_AddFriends
   slot {class=OverlaySlot, properties {horizontalAlignment=HAlign_Fill, verticalAlignment=VAlign_Fill}}
  }
  Button_FriendSearch {
   widgetClassPath=&/SocialPanel_FriendSearchBar
   bInherited=true
   bIsVariable=true
   parent=VBox_AddFriends
   widget=Button_FriendSearch
   properties {
    bIsFocusable=true
    onButtonBaseClicked=(&self:WidgetTree.Button_FriendSearch.HandleOnButtonClicked)
    onButtonBaseDoubleClicked=(&self:WidgetTree.Button_FriendSearch.HandleOnButtonDoubleClicked)
    onButtonBaseHovered=(&self:WidgetTree.Button_FriendSearch.HandleOnButtonHovered)
    onButtonBaseUnhovered=(&self:WidgetTree.Button_FriendSearch.HandleOnButtonUnhovered)
    onSelectedChangedBase=(&self:WidgetTree.Button_FriendSearch.HandleOnSelectedChanged)
   }
   slot {class=VerticalBoxSlot}
  }
  SocialPanel_FriendLinkButton {
   widgetClassPath=&/SocialPanel_FriendLinkButton
   bIsVariable=true
   parent=VBox_AddFriends
   widget=SocialPanel_FriendLinkButton
   properties {
    clickMethod=MouseDown
    onButtonBaseClicked=(&self:WidgetTree.SocialPanel_FriendLinkButton.HandleOnButtonClicked)
    onButtonBaseDoubleClicked=(&self:WidgetTree.SocialPanel_FriendLinkButton.HandleOnButtonDoubleClicked)
    onButtonBaseHovered=(&self:WidgetTree.SocialPanel_FriendLinkButton.HandleOnButtonHovered)
    onButtonBaseUnhovered=(&self:WidgetTree.SocialPanel_FriendLinkButton.HandleOnButtonUnhovered)
    onSelectedChangedBase=(&self:WidgetTree.SocialPanel_FriendLinkButton.HandleOnSelectedChanged)
    pressMethod=ButtonPress
   }
   slot {class=VerticalBoxSlot, properties {padding=<ltrb 0 0 0 4>}}
  }
  Button_AddWeGameFriends {
   widgetClassPath=&/SocialPanelWeGameFriendButton
   bInherited=true
   bIsVariable=true
   parent=VBox_AddFriends
   widget=Button_AddWeGameFriends
   properties {
    onButtonBaseClicked=(&self:WidgetTree.Button_AddWeGameFriends.HandleOnButtonClicked)
    onButtonBaseDoubleClicked=(&self:WidgetTree.Button_AddWeGameFriends.HandleOnButtonDoubleClicked)
    onButtonBaseHovered=(&self:WidgetTree.Button_AddWeGameFriends.HandleOnButtonHovered)
    onButtonBaseUnhovered=(&self:WidgetTree.Button_AddWeGameFriends.HandleOnButtonUnhovered)
    onSelectedChangedBase=(&self:WidgetTree.Button_AddWeGameFriends.HandleOnSelectedChanged)
   }
   slot {class=VerticalBoxSlot, properties {padding=<ltrb 4 0 4 0>}}
  }
  Switcher_AddFriends {
   widgetClassPath=&CommonWidgetSwitcherLegacy
   bInherited=true
   bIsVariable=true
   parent=VBox_AddFriends
   widget=Switcher_AddFriends
   slot {class=VerticalBoxSlot, properties {size {sizeRule=Fill, value=1}}}
  }
  UserTree_AddFriends {
   widgetClassPath=/Script/FortniteUI.FortSocialUserTreeView
   bInherited=true
   bIsVariable=true
   parent=Switcher_AddFriends
   widget=UserTree_AddFriends
   properties {
    entryWidgetClass=&/UserListHeaderEntry.UserListHeaderEntry_C
    inviteEntryWidgetClass=&/InviteListEntry_FriendInvite_XS.InviteListEntry_FriendInvite_XS_C
    joinablePartyEntryWidgetClass=&/JoinablePartyEntry.JoinablePartyEntry_C
    joinablePartyEntryWidgetClass_2=&JoinablePartyEntry_C
    leavePartyEntryWidgetClass=&/LeavePartyListEntry.LeavePartyListEntry_C
    missedInviteEntryWidgetClass=&/MissedInvitesEntry.MissedInvitesEntry_C
    multiFriendActionEntryWidgetClass=/FortniteCoreUI/UI/Sidebar/ListEntries/MultiFriendActionListEntry.MultiFriendActionListEntry_C
    partyMemberEntryWidgetClass=/FortniteCoreUI/UI/SocialPanel/ListEntries/PartyMemberListEntry.PartyMemberListEntry_C
    recentPlayerEntryWidgetClass=/FortniteCoreUI/UI/SocialPanel/ListEntries/UserListEntry_XS.UserListEntry_XS_C
    referFriendEntryWidgetClass=&/ReferAFriendEntry.ReferAFriendEntry_C
    suggestedFriendEntryWidgetClass=&/SuggestedFriendListEntry.SuggestedFriendListEntry_C
    teamMemberEntryWidgetClass=/FortniteCoreUI/UI/SocialPanel/ListEntries/TeamMemberListEntry.TeamMemberListEntry_C
    userEntryWidgetClass=&UserListEntry_XS_C
    userSearchResultEntryWidgetClass=/FortniteCoreUI/UI/SocialPanel/ListEntries/UserSearchResultListEntry_XS.UserSearchResultListEntry_XS_C
    voiceChatMemberEntryWidgetClass=/FortniteCoreUI/UI/SocialPanel/ListEntries/VoiceChatMemberListEntry.VoiceChatMemberListEntry_C
   }
   slot {class=WidgetSwitcherSlot}
  }
  VerticalBox_0 {
   widgetClassPath=&VerticalBox
   parent=Switcher_AddFriends
   widget=VerticalBox_0
   slot {class=WidgetSwitcherSlot, properties {verticalAlignment=VAlign_Top, padding=<ltrb 16 16 16 16>}}
  }
  Text_SearchingForUsers {
   widgetClassPath=&CommonTextBlock
   bInherited=true
   bIsVariable=true
   parent=VerticalBox_0
   widget=Text_SearchingForUsers
   properties {
    autoWrapText=true
    justification=Center
    text=Searching for users...
    textTransformPolicy=ToUpper
    colorAndOpacity {colorUseRule=UseColor_Specified, specifiedColor=<rgba 0.0723 0.8714 1 1>}
    font {
     bForceMonospaced=false
     bMaterialIsStencil=false
     fontMaterial=null
     fontObject=&BurbankSmall
     letterSpacing=0
     monospacedWidth=1
     size=24
     skewAmount=0
     typefaceFontName=Bold
     outlineSettings {
      bApplyOutlineToDropShadows=false
      bMiteredCorners=false
      bSeparateFillAlpha=false
      outlineMaterial=null
      outlineSize=0
      outlineColor=<rgba 0 0 0 1>
     }
    }
    shadowOffset=<0 0>
   }
   slot {class=VerticalBoxSlot}
  }
  CircularThrobber_UserSearch {
   widgetClassPath=&Image
   bIsVariable=true
   parent=VerticalBox_0
   widget=CircularThrobber_UserSearch
   properties {brush {
     bIsDynamicallyLoaded=false
     drawAs=Image
     imageType=NoImage
     mirroring=NoMirror
     resourceName=None
     resourceObject=/FortUILibrary/UIKit/Throbber/MI_UIKit_Color_Throbber
     tiling=NoTile
     imageSize=<32 32>
     margin=<ltrb 0 0 0 0>
     outlineSettings {
      bUseBrushTransparency=false
      roundingType=HalfHeightRadius
      width=0
      color {colorUseRule=UseColor_Specified, specifiedColor=<rgba 0 0 0 0>}
      cornerRadii=<0 0 0 0>
     }
     tintColor {colorUseRule=UseColor_Specified, specifiedColor=<rgba 1 1 1 1>}
     uVRegion {bIsValid=false, max=<0 0>, min=<0 0>}
    }}
   slot {class=VerticalBoxSlot, properties {horizontalAlignment=HAlign_Center, verticalAlignment=VAlign_Center, padding=<ltrb 16 12 16 16>}}
  }
  Text_SearchFailed {
   widgetClassPath=&CommonTextBlock
   bInherited=true
   bIsVariable=true
   parent=Switcher_AddFriends
   widget=Text_SearchFailed
   properties {
    autoWrapText=true
    justification=Center
    text="Search Failed!"
    textTransformPolicy=ToUpper
    colorAndOpacity {colorUseRule=UseColor_Specified, specifiedColor=<rgba 1 0.1714 0.1329 1>}
    font {
     bForceMonospaced=false
     bMaterialIsStencil=false
     fontMaterial=null
     fontObject=&BurbankSmall
     letterSpacing=0
     monospacedWidth=1
     size=24
     skewAmount=0
     typefaceFontName=Bold
     outlineSettings {
      bApplyOutlineToDropShadows=false
      bMiteredCorners=false
      bSeparateFillAlpha=false
      outlineMaterial=null
      outlineSize=0
      outlineColor=<rgba 0 0 0 1>
     }
    }
    shadowOffset=<0 0>
   }
   slot {class=WidgetSwitcherSlot, properties {verticalAlignment=VAlign_Top, padding=<ltrb 16 16 16 16>}}
  }
  UserTree_SearchResults {
   widgetClassPath=&FortSocialUserTreeView
   bInherited=true
   bIsVariable=true
   parent=Switcher_AddFriends
   widget=UserTree_SearchResults
   properties {
    entryWidgetClass=&UserListHeaderEntry_C
    inviteEntryWidgetClass=&InviteListEntry_FriendInvite_XS_C
    joinablePartyEntryWidgetClass=&JoinablePartyEntry_C
    joinablePartyEntryWidgetClass_2=&JoinablePartyEntry_C
    leavePartyEntryWidgetClass=&LeavePartyListEntry_C
    missedInviteEntryWidgetClass=&MissedInvitesEntry_C
    partyMemberEntryWidgetClass=&PartyMemberListEntry_C
    teamMemberEntryWidgetClass=&TeamMemberListEntry_C
    userEntryWidgetClass=&UserListEntry_XS_C
    userSearchResultEntryWidgetClass=&UserSearchResultListEntry_XS_C
    voiceChatMemberEntryWidgetClass=&VoiceChatMemberListEntry_C
   }
   slot {class=WidgetSwitcherSlot}
  }
  MyFriendsOvr {widgetClassPath=&Overlay, parent=Switcher_UserLists, widget=MyFriendsOvr, slot {class=WidgetSwitcherSlot}}
  MyFriendsGhostBorder {
   widgetClassPath=&Border
   parent=MyFriendsOvr
   widget=MyFriendsGhostBorder
   properties {
    background {
     bIsDynamicallyLoaded=false
     drawAs=Image
     imageType=NoImage
     mirroring=NoMirror
     resourceName=None
     resourceObject=&M_UI_SocialUserDarkPanel
     tiling=NoTile
     imageSize=<32 32>
     margin=<ltrb 0 0 0 0>
     outlineSettings {
      bUseBrushTransparency=false
      roundingType=HalfHeightRadius
      width=0
      color {colorUseRule=UseColor_Specified, specifiedColor=<rgba 0 0 0 0>}
      cornerRadii=<0 0 0 0>
     }
     tintColor {colorUseRule=UseColor_Specified, specifiedColor=<rgba 1 1 1 1>}
     uVRegion {bIsValid=false, max=<0 0>, min=<0 0>}
    }
    padding=<ltrb 0 0 0 0>
   }
   slot {class=OverlaySlot, properties {horizontalAlignment=HAlign_Fill, verticalAlignment=VAlign_Fill}}
  }
  UserTree_FriendsLists {
   widgetClassPath=&FortSocialUserTreeView
   bInherited=true
   bIsVariable=true
   parent=MyFriendsOvr
   widget=UserTree_FriendsLists
   properties {
    entryWidgetClass=&UserListHeaderEntry_C
    inviteEntryWidgetClass=&InviteListEntry_FriendInvite_XS_C
    joinRequestEntryWidgetClass=/FortniteCoreUI/UI/SocialPanel/ListEntries/JoinRequestUserListEntry.JoinRequestUserListEntry_C
    joinablePartyEntryWidgetClass=&JoinablePartyEntry_C
    joinablePartyEntryWidgetClass_2=&JoinablePartyEntry_C
    leavePartyEntryWidgetClass=&LeavePartyListEntry_C
    lfgInviteEntryWidgetClass=/FortniteCoreUI/UI/SocialPanel/ListEntries/UserListEntry.UserListEntry_C
    missedInviteEntryWidgetClass=&MissedInvitesEntry_C
    partyMemberEntryWidgetClass=&PartyMemberListEntry_C
    referFriendEntryWidgetClass=&ReferAFriendEntry_C
    teamMemberEntryWidgetClass=&TeamMemberListEntry_C
    userEntryWidgetClass=&UserListEntry_C
   }
   slot {class=OverlaySlot, properties {horizontalAlignment=HAlign_Fill, verticalAlignment=VAlign_Fill}}
  }
  VoiceChatOvr {widgetClassPath=&Overlay, parent=Switcher_UserLists, widget=VoiceChatOvr, slot {class=WidgetSwitcherSlot}}
  VoiceChatGhostBorder {
   widgetClassPath=&Border
   parent=VoiceChatOvr
   widget=VoiceChatGhostBorder
   properties {
    background {
     bIsDynamicallyLoaded=false
     drawAs=Image
     imageType=NoImage
     mirroring=NoMirror
     resourceName=None
     resourceObject=&M_UI_SocialUserDarkPanel
     tiling=NoTile
     imageSize=<32 32>
     margin=<ltrb 0 0 0 0>
     outlineSettings {
      bUseBrushTransparency=false
      roundingType=HalfHeightRadius
      width=0
      color {colorUseRule=UseColor_Specified, specifiedColor=<rgba 0 0 0 0>}
      cornerRadii=<0 0 0 0>
     }
     tintColor {colorUseRule=UseColor_Specified, specifiedColor=<rgba 1 1 1 1>}
     uVRegion {bIsValid=false, max=<0 0>, min=<0 0>}
    }
    padding=<ltrb 0 0 0 0>
   }
   slot {class=OverlaySlot, properties {horizontalAlignment=HAlign_Fill, verticalAlignment=VAlign_Fill}}
  }
  Switcher_VoiceChatState {
   widgetClassPath=&CommonWidgetSwitcherLegacy
   bInherited=true
   bIsVariable=true
   parent=VoiceChatOvr
   widget=Switcher_VoiceChatState
   properties {activeWidgetIndex=5}
   slot {class=OverlaySlot, properties {horizontalAlignment=HAlign_Fill, verticalAlignment=VAlign_Fill}}
  }
  VBox_NormalView {widgetClassPath=&VerticalBox, parent=Switcher_VoiceChatState, widget=VBox_NormalView, slot {class=WidgetSwitcherSlot}}
  UserTree_VoiceChat {
   widgetClassPath=&FortSocialUserTreeView
   bInherited=true
   bIsVariable=true
   parent=VBox_NormalView
   widget=UserTree_VoiceChat
   properties {
    entryWidgetClass=&UserListHeaderEntry_C
    inviteEntryWidgetClass=&InviteListEntry_FriendInvite_XS_C
    joinablePartyEntryWidgetClass=&JoinablePartyEntry_C
    joinablePartyEntryWidgetClass_2=&JoinablePartyEntry_C
    leavePartyEntryWidgetClass=&LeavePartyListEntry_C
    missedInviteEntryWidgetClass=&MissedInvitesEntry_C
    npcEntryClass=/FortniteCoreUI/UI/Sidebar/ListEntries/SidebarVoiceChatNpcListEntry.SidebarVoiceChatNpcListEntry_C
    partyMemberEntryWidgetClass=&PartyMemberListEntry_C
    referFriendEntryWidgetClass=&ReferAFriendEntry_C
    teamMemberEntryWidgetClass=&TeamMemberListEntry_C
    userEntryWidgetClass=&UserListEntry_XS_C
    voiceChannelHeaderEntryWidgetClass=/Game/UI/SocialPanel/VoiceChannelHeaderButton.VoiceChannelHeaderButton_C
    voiceChannelHeaderOptionEntryWidgetClass=/Game/UI/SocialPanel/VoiceChannelHeaderOptions.VoiceChannelHeaderOptions_C
    voiceChatMemberEntryWidgetClass=&VoiceChatMemberListEntry_C
   }
   slot {class=VerticalBoxSlot, properties {size {sizeRule=Fill, value=1}}}
  }
  VoiceChannelSettings {
   widgetClassPath=&/VoiceChannelSettings
   bInherited=true
   bIsVariable=true
   parent=VBox_NormalView
   widget=VoiceChannelSettings
   slot {class=VerticalBoxSlot, properties {verticalAlignment=VAlign_Bottom}}
  }
  VBox_VoiceOnDifferentDevice {
   widgetClassPath=&VerticalBox
   parent=Switcher_VoiceChatState
   widget=VBox_VoiceOnDifferentDevice
   slot {class=WidgetSwitcherSlot, properties {horizontalAlignment=HAlign_Center, verticalAlignment=VAlign_Center}}
  }
  HorizontalBox_28 {widgetClassPath=&HorizontalBox, parent=VBox_VoiceOnDifferentDevice, widget=HorizontalBox_28, slot {class=VerticalBoxSlot}}
  Image_MicIcon {
   widgetClassPath=&FortMobileImage
   bIsVariable=true
   parent=HorizontalBox_28
   widget=Image_MicIcon
   properties {
    mobileSizeMultiplier=1.34
    visibility=SelfHitTestInvisible
    brush {
     bIsDynamicallyLoaded=false
     drawAs=Image
     imageType=NoImage
     mirroring=NoMirror
     resourceName=None
     resourceObject=&T_UI_Icon_Mic_64
     tiling=NoTile
     imageSize=<32 32>
     margin=<ltrb 0 0 0 0>
     outlineSettings {
      bUseBrushTransparency=false
      roundingType=HalfHeightRadius
      width=0
      color {colorUseRule=UseColor_Specified, specifiedColor=<rgba 0 0 0 0>}
      cornerRadii=<0 0 0 0>
     }
     tintColor {colorUseRule=UseColor_Specified, specifiedColor=<rgba 1 1 1 1>}
     uVRegion {bIsValid=false, max=<0 0>, min=<0 0>}
    }
   }
   slot {class=HorizontalBoxSlot, properties {horizontalAlignment=HAlign_Center, verticalAlignment=VAlign_Center}}
  }
  CommonTextBlock_83 {
   widgetClassPath=&CommonTextBlock
   parent=HorizontalBox_28
   widget=CommonTextBlock_83
   properties {
    mobileFontSizeMultiplier=1.2
    scrollStyle=/FortUILibrary/Styles/Typography/Foundation/TextScrollStyle-NoFade.TextScrollStyle-NoFade_C
    text=VOICE CHAT IS CURRENTLY ON A DIFFERENT DEVICE
    wrapTextAt=450
    colorAndOpacity {colorUseRule=UseColor_Specified, specifiedColor=<rgba 0.2623 0.6514 0.9823 1>}
    font {
     bForceMonospaced=false
     bMaterialIsStencil=false
     fontMaterial=null
     fontObject=&BurbankSmall
     letterSpacing=0
     monospacedWidth=1
     size=20
     skewAmount=0
     typefaceFontName=Bold
     outlineSettings {
      bApplyOutlineToDropShadows=false
      bMiteredCorners=false
      bSeparateFillAlpha=false
      outlineMaterial=null
      outlineSize=0
      outlineColor=<rgba 0 0 0 1>
     }
    }
    shadowOffset=<0 0>
   }
   slot {class=HorizontalBoxSlot, properties {padding=<ltrb 8 0 8 0>, size {sizeRule=Fill, value=1}}}
  }
  Button_JoinVoiceChat {
   widgetClassPath=/FortniteCoreUI/Frontend/Parties/JoinVoiceChannelButton
   bInherited=true
   bIsVariable=true
   parent=VBox_VoiceOnDifferentDevice
   widget=Button_JoinVoiceChat
   properties {
    onButtonBaseClicked=(&self:WidgetTree.Button_JoinVoiceChat.HandleOnButtonClicked)
    onButtonBaseDoubleClicked=(&self:WidgetTree.Button_JoinVoiceChat.HandleOnButtonDoubleClicked)
    onButtonBaseHovered=(&self:WidgetTree.Button_JoinVoiceChat.HandleOnButtonHovered)
    onButtonBaseUnhovered=(&self:WidgetTree.Button_JoinVoiceChat.HandleOnButtonUnhovered)
    onSelectedChangedBase=(&self:WidgetTree.Button_JoinVoiceChat.HandleOnSelectedChanged)
   }
   slot {class=VerticalBoxSlot, properties {padding=<ltrb 0 16 0 0>}}
  }
  VBox_VoiceAcceptAudioMod {
   widgetClassPath=&VerticalBox
   bInherited=true
   parent=Switcher_VoiceChatState
   widget=VBox_VoiceAcceptAudioMod
   slot {class=WidgetSwitcherSlot, properties {verticalAlignment=VAlign_Top}}
  }
  ScaleBox_80 {
   widgetClassPath=&ScaleBox
   parent=VBox_VoiceAcceptAudioMod
   widget=ScaleBox_80
   slot {class=VerticalBoxSlot, properties {verticalAlignment=VAlign_Top, padding=<ltrb 0 12 0 0>, size {sizeRule=Fill, value=1}}}
  }
  EmptyListImage {
   widgetClassPath=&Image
   bIsVariable=true
   parent=ScaleBox_80
   widget=EmptyListImage
   properties {
    visibility=SelfHitTestInvisible
    brush {
     bIsDynamicallyLoaded=false
     drawAs=Image
     imageType=NoImage
     mirroring=NoMirror
     resourceName=None
     resourceObject=/Game/UI/Sidebar/Materials/MI_UI_Icon_Headset_Mic
     tiling=NoTile
     imageSize=<110 110>
     margin=<ltrb 0 0 0 0>
     outlineSettings {
      bUseBrushTransparency=false
      roundingType=HalfHeightRadius
      width=0
      color {colorUseRule=UseColor_Specified, specifiedColor=<rgba 0 0 0 0>}
      cornerRadii=<0 0 0 0>
     }
     tintColor {colorUseRule=UseColor_Specified, specifiedColor=<rgba 0.0331 0.402 1 1>}
     uVRegion {bIsValid=false, max=<0 0>, min=<0 0>}
    }
   }
   slot {class=ScaleBoxSlot}
  }
  VoiceChat_AcceptTitle {
   widgetClassPath=&CommonRichTextBlock
   bIsVariable=true
   parent=VBox_VoiceAcceptAudioMod
   widget=VoiceChat_AcceptTitle
   properties {
    autoWrapText=true
    bOverrideDefaultStyle=true
    justification=Center
    text=Voice Reporting Update
    textTransformPolicy=ToUpper
    defaultTextStyleOverride {
     overflowPolicy=Clip
     transformPolicy=None
     colorAndOpacity {colorUseRule=UseColor_Specified, specifiedColor=<rgba 1 1 1 1>}
     font {
      bForceMonospaced=false
      bMaterialIsStencil=false
      fontMaterial=null
      fontObject=&BurbankSmall
      letterSpacing=0
      monospacedWidth=1
      size=21
      skewAmount=0
      typefaceFontName=Black
      outlineSettings {
       bApplyOutlineToDropShadows=false
       bMiteredCorners=false
       bSeparateFillAlpha=false
       outlineMaterial=null
       outlineSize=0
       outlineColor=<rgba 0 0 0 1>
      }
     }
     highlightColor {colorUseRule=UseColor_Specified, specifiedColor=<rgba 0 0 0 1>}
     highlightShape {
      bIsDynamicallyLoaded=false
      drawAs=Image
      imageType=NoImage
      mirroring=NoMirror
      resourceName=None
      resourceObject=null
      tiling=NoTile
      imageSize=<32 32>
      margin=<ltrb 0 0 0 0>
      outlineSettings {
       bUseBrushTransparency=false
       roundingType=HalfHeightRadius
       width=0
       color {colorUseRule=UseColor_Specified, specifiedColor=<rgba 0 0 0 0>}
       cornerRadii=<0 0 0 0>
      }
      tintColor {colorUseRule=UseColor_Specified, specifiedColor=<rgba 1 1 1 1>}
      uVRegion {bIsValid=false, max=<0 0>, min=<0 0>}
     }
     selectedBackgroundColor {colorUseRule=UseColor_Foreground, specifiedColor=<rgba 1 0 1 1>}
     shadowColorAndOpacity=<rgba 0 0 0 1>
     shadowOffset=<0 0>
     strikeBrush {
      bIsDynamicallyLoaded=false
      drawAs=Image
      imageType=NoImage
      mirroring=NoMirror
      resourceName=None
      resourceObject=null
      tiling=NoTile
      imageSize=<32 32>
      margin=<ltrb 0 0 0 0>
      outlineSettings {
       bUseBrushTransparency=false
       roundingType=HalfHeightRadius
       width=0
       color {colorUseRule=UseColor_Specified, specifiedColor=<rgba 0 0 0 0>}
       cornerRadii=<0 0 0 0>
      }
      tintColor {colorUseRule=UseColor_Specified, specifiedColor=<rgba 1 1 1 1>}
      uVRegion {bIsValid=false, max=<0 0>, min=<0 0>}
     }
     underlineBrush {
      bIsDynamicallyLoaded=false
      drawAs=Image
      imageType=NoImage
      mirroring=NoMirror
      resourceName=None
      resourceObject=null
      tiling=NoTile
      imageSize=<32 32>
      margin=<ltrb 0 0 0 0>
      outlineSettings {
       bUseBrushTransparency=false
       roundingType=HalfHeightRadius
       width=0
       color {colorUseRule=UseColor_Specified, specifiedColor=<rgba 0 0 0 0>}
       cornerRadii=<0 0 0 0>
      }
      tintColor {colorUseRule=UseColor_Specified, specifiedColor=<rgba 1 1 1 1>}
      uVRegion {bIsValid=false, max=<0 0>, min=<0 0>}
     }
    }
   }
   slot {class=VerticalBoxSlot, properties {verticalAlignment=VAlign_Top, padding=<ltrb 32 10 32 0>}}
  }
  VoiceChat_AcceptDescription {
   widgetClassPath=&CommonRichTextBlock
   bIsVariable=true
   parent=VBox_VoiceAcceptAudioMod
   widget=VoiceChat_AcceptDescription
   properties {
    autoWrapText=true
    bIsScrollingEnabled=false
    bOverrideDefaultStyle=true
    lineHeightPercentage=0.85
    text="We've updated voice reporting to help us review reported violations of our Community Rules.\r\n- When voice chat is turned on, this feature stores voice chat audio on each participant's device on a rolling five-minute basis, which will be sent to Epic Games only if a participant reports a violation.\r\n- Voice reporting is usually on, except in parties formed among friends where all players prefer voice reporting off.\r\n- You can see if voice reporting is currently on or off in the Sidebar.\r\nSee your voice chat settings for more details and options"
    defaultTextStyleOverride {
     overflowPolicy=Clip
     transformPolicy=None
     colorAndOpacity {colorUseRule=UseColor_Specified, specifiedColor=<rgba 0.5089 0.5583 0.7084 1>}
     font {
      bForceMonospaced=false
      bMaterialIsStencil=false
      fontMaterial=null
      fontObject=&BurbankSmall
      letterSpacing=0
      monospacedWidth=1
      size=13.5
      skewAmount=0
      typefaceFontName=Bold
      outlineSettings {
       bApplyOutlineToDropShadows=false
       bMiteredCorners=false
       bSeparateFillAlpha=false
       outlineMaterial=null
       outlineSize=0
       outlineColor=<rgba 0 0 0 1>
      }
     }
     highlightColor {colorUseRule=UseColor_Specified, specifiedColor=<rgba 0 0 0 1>}
     highlightShape {
      bIsDynamicallyLoaded=false
      drawAs=Image
      imageType=NoImage
      mirroring=NoMirror
      resourceName=None
      resourceObject=null
      tiling=NoTile
      imageSize=<32 32>
      margin=<ltrb 0 0 0 0>
      outlineSettings {
       bUseBrushTransparency=false
       roundingType=HalfHeightRadius
       width=0
       color {colorUseRule=UseColor_Specified, specifiedColor=<rgba 0 0 0 0>}
       cornerRadii=<0 0 0 0>
      }
      tintColor {colorUseRule=UseColor_Specified, specifiedColor=<rgba 1 1 1 1>}
      uVRegion {bIsValid=false, max=<0 0>, min=<0 0>}
     }
     selectedBackgroundColor {colorUseRule=UseColor_Foreground, specifiedColor=<rgba 1 0 1 1>}
     shadowColorAndOpacity=<rgba 0 0 0 1>
     shadowOffset=<0 0>
     strikeBrush {
      bIsDynamicallyLoaded=false
      drawAs=Image
      imageType=NoImage
      mirroring=NoMirror
      resourceName=None
      resourceObject=null
      tiling=NoTile
      imageSize=<32 32>
      margin=<ltrb 0 0 0 0>
      outlineSettings {
       bUseBrushTransparency=false
       roundingType=HalfHeightRadius
       width=0
       color {colorUseRule=UseColor_Specified, specifiedColor=<rgba 0 0 0 0>}
       cornerRadii=<0 0 0 0>
      }
      tintColor {colorUseRule=UseColor_Specified, specifiedColor=<rgba 1 1 1 1>}
      uVRegion {bIsValid=false, max=<0 0>, min=<0 0>}
     }
     underlineBrush {
      bIsDynamicallyLoaded=false
      drawAs=Image
      imageType=NoImage
      mirroring=NoMirror
      resourceName=None
      resourceObject=null
      tiling=NoTile
      imageSize=<32 32>
      margin=<ltrb 0 0 0 0>
      outlineSettings {
       bUseBrushTransparency=false
       roundingType=HalfHeightRadius
       width=0
       color {colorUseRule=UseColor_Specified, specifiedColor=<rgba 0 0 0 0>}
       cornerRadii=<0 0 0 0>
      }
      tintColor {colorUseRule=UseColor_Specified, specifiedColor=<rgba 1 1 1 1>}
      uVRegion {bIsValid=false, max=<0 0>, min=<0 0>}
     }
    }
   }
   slot {class=VerticalBoxSlot, properties {padding=<ltrb 24 10 24 16>}}
  }
  VerticalBox_133 {
   widgetClassPath=&VerticalBox
   parent=VBox_VoiceAcceptAudioMod
   widget=VerticalBox_133
   slot {class=VerticalBoxSlot, properties {verticalAlignment=VAlign_Bottom, padding=<ltrb 24 24 24 24>}}
  }
  WBP_UIKit_Button_TurnOn {
   widgetClassPath=&WBP_UIKit_Button_Regular
   bIsVariable=true
   parent=VerticalBox_133
   widget=WBP_UIKit_Button_TurnOn
   properties {
    bDisplayInputActionWhenNotInteractable=false
    bShouldUseFallbackDefaultInputAction=false
    buttonSize=Small
    text=Turn on voice chat
    triggeringInputAction {dataTable=&InputActionDataTable, rowName=GenericFaceButtonLeft}
   }
   slot {class=VerticalBoxSlot, properties {padding=<ltrb 0 0 0 12>}}
  }
  WBP_UIKit_Button_LearnMore {
   widgetClassPath=&WBP_UIKit_Button_Regular
   bIsVariable=true
   parent=VerticalBox_133
   widget=WBP_UIKit_Button_LearnMore
   properties {
    bDisplayInputActionWhenNotInteractable=false
    bShouldUseFallbackDefaultInputAction=false
    buttonSize=Small
    text=Learn More
    triggeringInputAction {dataTable=&InputActionDataTable, rowName=ToastAccept}
   }
   slot {class=VerticalBoxSlot}
  }
  VBox_VoicePlatformPartyWithSwitchButton {
   widgetClassPath=&VerticalBox
   bInherited=true
   parent=Switcher_VoiceChatState
   widget=VBox_VoicePlatformPartyWithSwitchButton
   slot {class=WidgetSwitcherSlot, properties {horizontalAlignment=HAlign_Center, verticalAlignment=VAlign_Center, padding=<ltrb 40 0 40 0>}}
  }
  HorizontalBox_1 {
   widgetClassPath=&HorizontalBox
   parent=VBox_VoicePlatformPartyWithSwitchButton
   widget=HorizontalBox_1
   properties {visibility=Collapsed}
   slot {class=VerticalBoxSlot, properties {verticalAlignment=VAlign_Center}}
  }
  Image_MicIcon_1 {
   widgetClassPath=&FortMobileImage
   bIsVariable=true
   parent=HorizontalBox_1
   widget=Image_MicIcon_1
   properties {
    mobileSizeMultiplier=1.34
    visibility=SelfHitTestInvisible
    brush {
     bIsDynamicallyLoaded=false
     drawAs=Image
     imageType=NoImage
     mirroring=NoMirror
     resourceName=None
     resourceObject=&T_UI_Icon_Mic_64
     tiling=NoTile
     imageSize=<32 32>
     margin=<ltrb 0 0 0 0>
     outlineSettings {
      bUseBrushTransparency=false
      roundingType=HalfHeightRadius
      width=0
      color {colorUseRule=UseColor_Specified, specifiedColor=<rgba 0 0 0 0>}
      cornerRadii=<0 0 0 0>
     }
     tintColor {colorUseRule=UseColor_Specified, specifiedColor=<rgba 0.2623 0.6514 0.9823 1>}
     uVRegion {bIsValid=false, max=<0 0>, min=<0 0>}
    }
   }
   slot {class=HorizontalBoxSlot, properties {horizontalAlignment=HAlign_Center, verticalAlignment=VAlign_Center}}
  }
  CommonTextBlock {
   widgetClassPath=&CommonTextBlock
   parent=HorizontalBox_1
   widget=CommonTextBlock
   properties {
    mobileFontSizeMultiplier=1.2
    scrollStyle=&TextScrollStyle-NoFade_C
    text=YOU ARE NOT IN THE PARTY VOICE CHANNEL
    colorAndOpacity {colorUseRule=UseColor_Specified, specifiedColor=<rgba 0.2623 0.6514 0.9823 1>}
    font {
     bForceMonospaced=false
     bMaterialIsStencil=false
     fontMaterial=null
     fontObject=&BurbankSmall
     letterSpacing=0
     monospacedWidth=1
     size=20
     skewAmount=0
     typefaceFontName=Bold
     outlineSettings {
      bApplyOutlineToDropShadows=false
      bMiteredCorners=false
      bSeparateFillAlpha=false
      outlineMaterial=null
      outlineSize=0
      outlineColor=<rgba 0 0 0 1>
     }
    }
    shadowOffset=<0 0>
   }
   slot {
    class=HorizontalBoxSlot
    properties {horizontalAlignment=HAlign_Left, verticalAlignment=VAlign_Center, padding=<ltrb 8 0 8 0>, size {sizeRule=Fill, value=1}}
   }
  }
  SizeBox_0 {
   widgetClassPath=&SizeBox
   parent=VBox_VoicePlatformPartyWithSwitchButton
   widget=SizeBox_0
   properties {bOverride_MaxDesiredWidth=true, maxDesiredWidth=700}
   slot {class=VerticalBoxSlot, properties {horizontalAlignment=HAlign_Center, verticalAlignment=VAlign_Center, padding=<ltrb 8 0 8 0>}}
  }
  Text_PartyVoiceMessage {
   widgetClassPath=&CommonTextBlock
   bInherited=true
   parent=SizeBox_0
   widget=Text_PartyVoiceMessage
   properties {
    autoWrapText=true
    justification=Center
    mobileFontSizeMultiplier=1.2
    text="Your party is trying to talk to you, but you are currently in  Xbox Live Party voice chat."
    visibility=Collapsed
    colorAndOpacity {colorUseRule=UseColor_Specified, specifiedColor=<rgba 0.2623 0.6514 0.9823 1>}
    font {
     bForceMonospaced=false
     bMaterialIsStencil=false
     fontMaterial=null
     fontObject=&BurbankSmall
     letterSpacing=0
     monospacedWidth=1
     size=20
     skewAmount=0
     typefaceFontName=Bold
     outlineSettings {
      bApplyOutlineToDropShadows=false
      bMiteredCorners=false
      bSeparateFillAlpha=false
      outlineMaterial=null
      outlineSize=0
      outlineColor=<rgba 0 0 0 1>
     }
    }
    shadowOffset=<0 0>
   }
   slot {class=SizeBoxSlot}
  }
  Button_LeavePlatformVoiceChat {
   widgetClassPath=/FortniteCoreUI/Frontend/Parties/LeavePlatformPartyButton
   bInherited=true
   bIsVariable=true
   parent=VBox_VoicePlatformPartyWithSwitchButton
   widget=Button_LeavePlatformVoiceChat
   properties {
    onButtonBaseClicked=(&self:WidgetTree.Button_LeavePlatformVoiceChat.HandleOnButtonClicked)
    onButtonBaseDoubleClicked=(&self:WidgetTree.Button_LeavePlatformVoiceChat.HandleOnButtonDoubleClicked)
    onButtonBaseHovered=(&self:WidgetTree.Button_LeavePlatformVoiceChat.HandleOnButtonHovered)
    onButtonBaseUnhovered=(&self:WidgetTree.Button_LeavePlatformVoiceChat.HandleOnButtonUnhovered)
    onSelectedChangedBase=(&self:WidgetTree.Button_LeavePlatformVoiceChat.HandleOnSelectedChanged)
   }
   slot {class=VerticalBoxSlot, properties {horizontalAlignment=HAlign_Center, padding=<ltrb 2 0 0 0>}}
  }
  VBox_VoicePlatformPartyInstructions {
   widgetClassPath=&VerticalBox
   bInherited=true
   parent=Switcher_VoiceChatState
   widget=VBox_VoicePlatformPartyInstructions
   slot {class=WidgetSwitcherSlot, properties {horizontalAlignment=HAlign_Left, padding=<ltrb 20 40 0 20>}}
  }
  HorizontalBox_2 {
   widgetClassPath=&HorizontalBox
   parent=VBox_VoicePlatformPartyInstructions
   widget=HorizontalBox_2
   properties {visibility=Collapsed}
   slot {class=VerticalBoxSlot, properties {padding=<ltrb 2 8 0 16>}}
  }
  Image_MicIcon_2 {
   widgetClassPath=&FortMobileImage
   bIsVariable=true
   parent=HorizontalBox_2
   widget=Image_MicIcon_2
   properties {
    mobileSizeMultiplier=1.3
    brush {
     bIsDynamicallyLoaded=false
     drawAs=Image
     imageType=NoImage
     mirroring=NoMirror
     resourceName=None
     resourceObject=&T_UI_Icon_Mic_64
     tiling=NoTile
     imageSize=<32 32>
     margin=<ltrb 0 0 0 0>
     outlineSettings {
      bUseBrushTransparency=false
      roundingType=HalfHeightRadius
      width=0
      color {colorUseRule=UseColor_Specified, specifiedColor=<rgba 0 0 0 0>}
      cornerRadii=<0 0 0 0>
     }
     tintColor {colorUseRule=UseColor_Specified, specifiedColor=<rgba 0.4452 1 1 1>}
     uVRegion {bIsValid=false, max=<0 0>, min=<0 0>}
    }
   }
   slot {class=HorizontalBoxSlot, properties {horizontalAlignment=HAlign_Center, verticalAlignment=VAlign_Center}}
  }
  CommonRichTextBlock_97 {
   widgetClassPath=&CommonRichTextBlock
   bIsVariable=true
   parent=HorizontalBox_2
   widget=CommonRichTextBlock_97
   properties {
    bOverrideDefaultStyle=true
    text=YOU ARE NOT IN THE PARTY VOICE CHANNEL
    wrapTextAt=350
    defaultTextStyleOverride {
     overflowPolicy=Clip
     transformPolicy=None
     colorAndOpacity {colorUseRule=UseColor_Specified, specifiedColor=<rgba 0.4452 1 1 1>}
     font {
      bForceMonospaced=false
      bMaterialIsStencil=false
      fontMaterial=null
      fontObject=&BurbankSmall
      letterSpacing=0
      monospacedWidth=1
      size=18
      skewAmount=0
      typefaceFontName=Black
      outlineSettings {
       bApplyOutlineToDropShadows=false
       bMiteredCorners=false
       bSeparateFillAlpha=false
       outlineMaterial=null
       outlineSize=0
       outlineColor=<rgba 0 0 0 1>
      }
     }
     highlightColor {colorUseRule=UseColor_Specified, specifiedColor=<rgba 0 0 0 1>}
     highlightShape {
      bIsDynamicallyLoaded=false
      drawAs=Image
      imageType=NoImage
      mirroring=NoMirror
      resourceName=None
      resourceObject=null
      tiling=NoTile
      imageSize=<32 32>
      margin=<ltrb 0 0 0 0>
      outlineSettings {
       bUseBrushTransparency=false
       roundingType=HalfHeightRadius
       width=0
       color {colorUseRule=UseColor_Specified, specifiedColor=<rgba 0 0 0 0>}
       cornerRadii=<0 0 0 0>
      }
      tintColor {colorUseRule=UseColor_Specified, specifiedColor=<rgba 1 1 1 1>}
      uVRegion {bIsValid=false, max=<0 0>, min=<0 0>}
     }
     selectedBackgroundColor {colorUseRule=UseColor_Foreground, specifiedColor=<rgba 1 0 1 1>}
     shadowColorAndOpacity=<rgba 0 0 0 1>
     shadowOffset=<0 0>
     strikeBrush {
      bIsDynamicallyLoaded=false
      drawAs=Image
      imageType=NoImage
      mirroring=NoMirror
      resourceName=None
      resourceObject=null
      tiling=NoTile
      imageSize=<32 32>
      margin=<ltrb 0 0 0 0>
      outlineSettings {
       bUseBrushTransparency=false
       roundingType=HalfHeightRadius
       width=0
       color {colorUseRule=UseColor_Specified, specifiedColor=<rgba 0 0 0 0>}
       cornerRadii=<0 0 0 0>
      }
      tintColor {colorUseRule=UseColor_Specified, specifiedColor=<rgba 1 1 1 1>}
      uVRegion {bIsValid=false, max=<0 0>, min=<0 0>}
     }
     underlineBrush {
      bIsDynamicallyLoaded=false
      drawAs=Image
      imageType=NoImage
      mirroring=NoMirror
      resourceName=None
      resourceObject=null
      tiling=NoTile
      imageSize=<32 32>
      margin=<ltrb 0 0 0 0>
      outlineSettings {
       bUseBrushTransparency=false
       roundingType=HalfHeightRadius
       width=0
       color {colorUseRule=UseColor_Specified, specifiedColor=<rgba 0 0 0 0>}
       cornerRadii=<0 0 0 0>
      }
      tintColor {colorUseRule=UseColor_Specified, specifiedColor=<rgba 1 1 1 1>}
      uVRegion {bIsValid=false, max=<0 0>, min=<0 0>}
     }
    }
   }
   slot {class=HorizontalBoxSlot, properties {verticalAlignment=VAlign_Center, padding=<ltrb 8 0 8 0>, size {sizeRule=Fill, value=1}}}
  }
  RichText_PlatformPartyChatMessageHeader {
   widgetClassPath=&CommonRichTextBlock
   bInherited=true
   bIsVariable=true
   parent=VBox_VoicePlatformPartyInstructions
   widget=RichText_PlatformPartyChatMessageHeader
   properties {
    autoWrapText=true
    bOverrideDefaultStyle=true
    text="Your Fortnite party members want to speak to you, but you are currently in Xbox Live Party Chat."
    visibility=Collapsed
    defaultTextStyleOverride {
     overflowPolicy=Clip
     transformPolicy=None
     colorAndOpacity {colorUseRule=UseColor_Specified, specifiedColor=<rgba 0.2542 0.4969 0.7913 1>}
     font {
      bForceMonospaced=false
      bMaterialIsStencil=false
      fontMaterial=null
      fontObject=&BurbankSmall
      letterSpacing=0
      monospacedWidth=1
      size=18
      skewAmount=0
      typefaceFontName=Bold
      outlineSettings {
       bApplyOutlineToDropShadows=false
       bMiteredCorners=false
       bSeparateFillAlpha=false
       outlineMaterial=null
       outlineSize=0
       outlineColor=<rgba 0 0 0 1>
      }
     }
     highlightColor {colorUseRule=UseColor_Specified, specifiedColor=<rgba 0 0 0 1>}
     highlightShape {
      bIsDynamicallyLoaded=false
      drawAs=Image
      imageType=NoImage
      mirroring=NoMirror
      resourceName=None
      resourceObject=null
      tiling=NoTile
      imageSize=<32 32>
      margin=<ltrb 0 0 0 0>
      outlineSettings {
       bUseBrushTransparency=false
       roundingType=HalfHeightRadius
       width=0
       color {colorUseRule=UseColor_Specified, specifiedColor=<rgba 0 0 0 0>}
       cornerRadii=<0 0 0 0>
      }
      tintColor {colorUseRule=UseColor_Specified, specifiedColor=<rgba 1 1 1 1>}
      uVRegion {bIsValid=false, max=<0 0>, min=<0 0>}
     }
     selectedBackgroundColor {colorUseRule=UseColor_Foreground, specifiedColor=<rgba 1 0 1 1>}
     shadowColorAndOpacity=<rgba 0 0 0 1>
     shadowOffset=<0 0>
     strikeBrush {
      bIsDynamicallyLoaded=false
      drawAs=Image
      imageType=NoImage
      mirroring=NoMirror
      resourceName=None
      resourceObject=null
      tiling=NoTile
      imageSize=<32 32>
      margin=<ltrb 0 0 0 0>
      outlineSettings {
       bUseBrushTransparency=false
       roundingType=HalfHeightRadius
       width=0
       color {colorUseRule=UseColor_Specified, specifiedColor=<rgba 0 0 0 0>}
       cornerRadii=<0 0 0 0>
      }
      tintColor {colorUseRule=UseColor_Specified, specifiedColor=<rgba 1 1 1 1>}
      uVRegion {bIsValid=false, max=<0 0>, min=<0 0>}
     }
     underlineBrush {
      bIsDynamicallyLoaded=false
      drawAs=Image
      imageType=NoImage
      mirroring=NoMirror
      resourceName=None
      resourceObject=null
      tiling=NoTile
      imageSize=<32 32>
      margin=<ltrb 0 0 0 0>
      outlineSettings {
       bUseBrushTransparency=false
       roundingType=HalfHeightRadius
       width=0
       color {colorUseRule=UseColor_Specified, specifiedColor=<rgba 0 0 0 0>}
       cornerRadii=<0 0 0 0>
      }
      tintColor {colorUseRule=UseColor_Specified, specifiedColor=<rgba 1 1 1 1>}
      uVRegion {bIsValid=false, max=<0 0>, min=<0 0>}
     }
    }
    renderTransform {angle=0, scale=<1 1>, shear=<-8 0>, translation=<0 0>}
   }
   slot {class=VerticalBoxSlot, properties {padding=<ltrb 8 0 8 16>}}
  }
  RichText_LeavePlatformPartyTitle {
   widgetClassPath=&CommonRichTextBlock
   bIsVariable=true
   parent=VBox_VoicePlatformPartyInstructions
   widget=RichText_LeavePlatformPartyTitle
   properties {
    autoWrapText=true
    bOverrideDefaultStyle=true
    text="In order to join the Fortnite Party Voice Channel, please do the following:"
    defaultTextStyleOverride {
     overflowPolicy=Clip
     transformPolicy=None
     colorAndOpacity {colorUseRule=UseColor_Specified, specifiedColor=<rgba 0.2542 0.4969 0.7913 1>}
     font {
      bForceMonospaced=false
      bMaterialIsStencil=false
      fontMaterial=null
      fontObject=&BurbankSmall
      letterSpacing=0
      monospacedWidth=1
      size=18
      skewAmount=0
      typefaceFontName=Bold
      outlineSettings {
       bApplyOutlineToDropShadows=false
       bMiteredCorners=false
       bSeparateFillAlpha=false
       outlineMaterial=null
       outlineSize=0
       outlineColor=<rgba 0 0 0 1>
      }
     }
     highlightColor {colorUseRule=UseColor_Specified, specifiedColor=<rgba 0 0 0 1>}
     highlightShape {
      bIsDynamicallyLoaded=false
      drawAs=Image
      imageType=NoImage
      mirroring=NoMirror
      resourceName=None
      resourceObject=null
      tiling=NoTile
      imageSize=<32 32>
      margin=<ltrb 0 0 0 0>
      outlineSettings {
       bUseBrushTransparency=false
       roundingType=HalfHeightRadius
       width=0
       color {colorUseRule=UseColor_Specified, specifiedColor=<rgba 0 0 0 0>}
       cornerRadii=<0 0 0 0>
      }
      tintColor {colorUseRule=UseColor_Specified, specifiedColor=<rgba 1 1 1 1>}
      uVRegion {bIsValid=false, max=<0 0>, min=<0 0>}
     }
     selectedBackgroundColor {colorUseRule=UseColor_Foreground, specifiedColor=<rgba 1 0 1 1>}
     shadowColorAndOpacity=<rgba 0 0 0 1>
     shadowOffset=<0 0>
     strikeBrush {
      bIsDynamicallyLoaded=false
      drawAs=Image
      imageType=NoImage
      mirroring=NoMirror
      resourceName=None
      resourceObject=null
      tiling=NoTile
      imageSize=<32 32>
      margin=<ltrb 0 0 0 0>
      outlineSettings {
       bUseBrushTransparency=false
       roundingType=HalfHeightRadius
       width=0
       color {colorUseRule=UseColor_Specified, specifiedColor=<rgba 0 0 0 0>}
       cornerRadii=<0 0 0 0>
      }
      tintColor {colorUseRule=UseColor_Specified, specifiedColor=<rgba 1 1 1 1>}
      uVRegion {bIsValid=false, max=<0 0>, min=<0 0>}
     }
     underlineBrush {
      bIsDynamicallyLoaded=false
      drawAs=Image
      imageType=NoImage
      mirroring=NoMirror
      resourceName=None
      resourceObject=null
      tiling=NoTile
      imageSize=<32 32>
      margin=<ltrb 0 0 0 0>
      outlineSettings {
       bUseBrushTransparency=false
       roundingType=HalfHeightRadius
       width=0
       color {colorUseRule=UseColor_Specified, specifiedColor=<rgba 0 0 0 0>}
       cornerRadii=<0 0 0 0>
      }
      tintColor {colorUseRule=UseColor_Specified, specifiedColor=<rgba 1 1 1 1>}
      uVRegion {bIsValid=false, max=<0 0>, min=<0 0>}
     }
    }
    renderTransform {angle=0, scale=<1 1>, shear=<-8 0>, translation=<0 0>}
   }
   slot {class=VerticalBoxSlot, properties {padding=<ltrb 8 0 8 16>}}
  }
  RichText_LeavePlatformPartyInstructions {
   widgetClassPath=&CommonRichTextBlock
   bInherited=true
   bIsVariable=true
   parent=VBox_VoicePlatformPartyInstructions
   widget=RichText_LeavePlatformPartyInstructions
   properties {
    autoWrapText=true
    bOverrideDefaultStyle=true
    visibility=Collapsed
    defaultTextStyleOverride {
     overflowPolicy=Clip
     transformPolicy=None
     colorAndOpacity {colorUseRule=UseColor_Specified, specifiedColor=<rgba 0.2542 0.4969 0.7913 1>}
     font {
      bForceMonospaced=false
      bMaterialIsStencil=false
      fontMaterial=null
      fontObject=&BurbankSmall
      letterSpacing=0
      monospacedWidth=1
      size=18
      skewAmount=0
      typefaceFontName=Bold
      outlineSettings {
       bApplyOutlineToDropShadows=false
       bMiteredCorners=false
       bSeparateFillAlpha=false
       outlineMaterial=null
       outlineSize=0
       outlineColor=<rgba 0 0 0 1>
      }
     }
     highlightColor {colorUseRule=UseColor_Specified, specifiedColor=<rgba 0 0 0 1>}
     highlightShape {
      bIsDynamicallyLoaded=false
      drawAs=Image
      imageType=NoImage
      mirroring=NoMirror
      resourceName=None
      resourceObject=null
      tiling=NoTile
      imageSize=<32 32>
      margin=<ltrb 0 0 0 0>
      outlineSettings {
       bUseBrushTransparency=false
       roundingType=HalfHeightRadius
       width=0
       color {colorUseRule=UseColor_Specified, specifiedColor=<rgba 0 0 0 0>}
       cornerRadii=<0 0 0 0>
      }
      tintColor {colorUseRule=UseColor_Specified, specifiedColor=<rgba 1 1 1 1>}
      uVRegion {bIsValid=false, max=<0 0>, min=<0 0>}
     }
     selectedBackgroundColor {colorUseRule=UseColor_Foreground, specifiedColor=<rgba 1 0 1 1>}
     shadowColorAndOpacity=<rgba 0 0 0 1>
     shadowOffset=<0 0>
     strikeBrush {
      bIsDynamicallyLoaded=false
      drawAs=Image
      imageType=NoImage
      mirroring=NoMirror
      resourceName=None
      resourceObject=null
      tiling=NoTile
      imageSize=<32 32>
      margin=<ltrb 0 0 0 0>
      outlineSettings {
       bUseBrushTransparency=false
       roundingType=HalfHeightRadius
       width=0
       color {colorUseRule=UseColor_Specified, specifiedColor=<rgba 0 0 0 0>}
       cornerRadii=<0 0 0 0>
      }
      tintColor {colorUseRule=UseColor_Specified, specifiedColor=<rgba 1 1 1 1>}
      uVRegion {bIsValid=false, max=<0 0>, min=<0 0>}
     }
     underlineBrush {
      bIsDynamicallyLoaded=false
      drawAs=Image
      imageType=NoImage
      mirroring=NoMirror
      resourceName=None
      resourceObject=null
      tiling=NoTile
      imageSize=<32 32>
      margin=<ltrb 0 0 0 0>
      outlineSettings {
       bUseBrushTransparency=false
       roundingType=HalfHeightRadius
       width=0
       color {colorUseRule=UseColor_Specified, specifiedColor=<rgba 0 0 0 0>}
       cornerRadii=<0 0 0 0>
      }
      tintColor {colorUseRule=UseColor_Specified, specifiedColor=<rgba 1 1 1 1>}
      uVRegion {bIsValid=false, max=<0 0>, min=<0 0>}
     }
    }
   }
   slot {class=VerticalBoxSlot}
  }
  VBox_VoiceDisabledParty {
   widgetClassPath=&VerticalBox
   parent=Switcher_VoiceChatState
   widget=VBox_VoiceDisabledParty
   slot {class=WidgetSwitcherSlot, properties {horizontalAlignment=HAlign_Center, verticalAlignment=VAlign_Center, padding=<ltrb 16 16 16 16>}}
  }
  Image_MicIcon_3 {
   widgetClassPath=&Image
   bIsVariable=true
   parent=VBox_VoiceDisabledParty
   widget=Image_MicIcon_3
   properties {
    visibility=SelfHitTestInvisible
    brush {
     bIsDynamicallyLoaded=false
     drawAs=Image
     imageType=NoImage
     mirroring=NoMirror
     resourceName=None
     resourceObject=&T_UI_Icon_Mic_64
     tiling=NoTile
     imageSize=<64 64>
     margin=<ltrb 0 0 0 0>
     outlineSettings {
      bUseBrushTransparency=false
      roundingType=HalfHeightRadius
      width=0
      color {colorUseRule=UseColor_Specified, specifiedColor=<rgba 0 0 0 0>}
      cornerRadii=<0 0 0 0>
     }
     tintColor {colorUseRule=UseColor_Specified, specifiedColor=<rgba 1 1 1 0.92>}
     uVRegion {bIsValid=false, max=<0 0>, min=<0 0>}
    }
   }
   slot {class=VerticalBoxSlot, properties {horizontalAlignment=HAlign_Center, verticalAlignment=VAlign_Center}}
  }
  RichText_VoiceChatDisabled {
   widgetClassPath=&CommonRichTextBlock
   bIsVariable=true
   parent=VBox_VoiceDisabledParty
   widget=RichText_VoiceChatDisabled
   properties {
    autoWrapText=true
    bOverrideDefaultStyle=true
    justification=Center
    text=You are not in the party voice channel.
    textTransformPolicy=ToUpper
    defaultTextStyleOverride {
     overflowPolicy=Clip
     transformPolicy=None
     colorAndOpacity {colorUseRule=UseColor_Specified, specifiedColor=<rgba 0.2542 0.4969 0.7913 1>}
     font {
      bForceMonospaced=false
      bMaterialIsStencil=false
      fontMaterial=null
      fontObject=&BurbankSmall
      letterSpacing=0
      monospacedWidth=1
      size=20
      skewAmount=0
      typefaceFontName=Bold
      outlineSettings {
       bApplyOutlineToDropShadows=false
       bMiteredCorners=false
       bSeparateFillAlpha=false
       outlineMaterial=null
       outlineSize=0
       outlineColor=<rgba 0 0 0 1>
      }
     }
     highlightColor {colorUseRule=UseColor_Specified, specifiedColor=<rgba 0 0 0 1>}
     highlightShape {
      bIsDynamicallyLoaded=false
      drawAs=Image
      imageType=NoImage
      mirroring=NoMirror
      resourceName=None
      resourceObject=null
      tiling=NoTile
      imageSize=<32 32>
      margin=<ltrb 0 0 0 0>
      outlineSettings {
       bUseBrushTransparency=false
       roundingType=HalfHeightRadius
       width=0
       color {colorUseRule=UseColor_Specified, specifiedColor=<rgba 0 0 0 0>}
       cornerRadii=<0 0 0 0>
      }
      tintColor {colorUseRule=UseColor_Specified, specifiedColor=<rgba 1 1 1 1>}
      uVRegion {bIsValid=false, max=<0 0>, min=<0 0>}
     }
     selectedBackgroundColor {colorUseRule=UseColor_Foreground, specifiedColor=<rgba 1 0 1 1>}
     shadowColorAndOpacity=<rgba 0 0 0 1>
     shadowOffset=<0 0>
     strikeBrush {
      bIsDynamicallyLoaded=false
      drawAs=Image
      imageType=NoImage
      mirroring=NoMirror
      resourceName=None
      resourceObject=null
      tiling=NoTile
      imageSize=<32 32>
      margin=<ltrb 0 0 0 0>
      outlineSettings {
       bUseBrushTransparency=false
       roundingType=HalfHeightRadius
       width=0
       color {colorUseRule=UseColor_Specified, specifiedColor=<rgba 0 0 0 0>}
       cornerRadii=<0 0 0 0>
      }
      tintColor {colorUseRule=UseColor_Specified, specifiedColor=<rgba 1 1 1 1>}
      uVRegion {bIsValid=false, max=<0 0>, min=<0 0>}
     }
     underlineBrush {
      bIsDynamicallyLoaded=false
      drawAs=Image
      imageType=NoImage
      mirroring=NoMirror
      resourceName=None
      resourceObject=null
      tiling=NoTile
      imageSize=<32 32>
      margin=<ltrb 0 0 0 0>
      outlineSettings {
       bUseBrushTransparency=false
       roundingType=HalfHeightRadius
       width=0
       color {colorUseRule=UseColor_Specified, specifiedColor=<rgba 0 0 0 0>}
       cornerRadii=<0 0 0 0>
      }
      tintColor {colorUseRule=UseColor_Specified, specifiedColor=<rgba 1 1 1 1>}
      uVRegion {bIsValid=false, max=<0 0>, min=<0 0>}
     }
    }
   }
   slot {class=VerticalBoxSlot, properties {verticalAlignment=VAlign_Center, padding=<ltrb 8 0 8 0>}}
  }
  RichText_VoiceChatDisabledWarningMessage {
   widgetClassPath=&CommonRichTextBlock
   bIsVariable=true
   parent=VBox_VoiceDisabledParty
   widget=RichText_VoiceChatDisabledWarningMessage
   properties {
    autoWrapText=true
    bOverrideDefaultStyle=true
    justification=Center
    text="Your party is trying to talk to you, but you currently have Voice Chat disabled."
    defaultTextStyleOverride {
     overflowPolicy=Clip
     transformPolicy=None
     colorAndOpacity {colorUseRule=UseColor_Specified, specifiedColor=<rgba 0.2542 0.4969 0.7913 1>}
     font {
      bForceMonospaced=false
      bMaterialIsStencil=false
      fontMaterial=null
      fontObject=&BurbankSmall
      letterSpacing=0
      monospacedWidth=1
      size=18
      skewAmount=0
      typefaceFontName=Bold
      outlineSettings {
       bApplyOutlineToDropShadows=false
       bMiteredCorners=false
       bSeparateFillAlpha=false
       outlineMaterial=null
       outlineSize=0
       outlineColor=<rgba 0 0 0 1>
      }
     }
     highlightColor {colorUseRule=UseColor_Specified, specifiedColor=<rgba 0 0 0 1>}
     highlightShape {
      bIsDynamicallyLoaded=false
      drawAs=Image
      imageType=NoImage
      mirroring=NoMirror
      resourceName=None
      resourceObject=null
      tiling=NoTile
      imageSize=<32 32>
      margin=<ltrb 0 0 0 0>
      outlineSettings {
       bUseBrushTransparency=false
       roundingType=HalfHeightRadius
       width=0
       color {colorUseRule=UseColor_Specified, specifiedColor=<rgba 0 0 0 0>}
       cornerRadii=<0 0 0 0>
      }
      tintColor {colorUseRule=UseColor_Specified, specifiedColor=<rgba 1 1 1 1>}
      uVRegion {bIsValid=false, max=<0 0>, min=<0 0>}
     }
     selectedBackgroundColor {colorUseRule=UseColor_Foreground, specifiedColor=<rgba 1 0 1 1>}
     shadowColorAndOpacity=<rgba 0 0 0 1>
     shadowOffset=<0 0>
     strikeBrush {
      bIsDynamicallyLoaded=false
      drawAs=Image
      imageType=NoImage
      mirroring=NoMirror
      resourceName=None
      resourceObject=null
      tiling=NoTile
      imageSize=<32 32>
      margin=<ltrb 0 0 0 0>
      outlineSettings {
       bUseBrushTransparency=false
       roundingType=HalfHeightRadius
       width=0
       color {colorUseRule=UseColor_Specified, specifiedColor=<rgba 0 0 0 0>}
       cornerRadii=<0 0 0 0>
      }
      tintColor {colorUseRule=UseColor_Specified, specifiedColor=<rgba 1 1 1 1>}
      uVRegion {bIsValid=false, max=<0 0>, min=<0 0>}
     }
     underlineBrush {
      bIsDynamicallyLoaded=false
      drawAs=Image
      imageType=NoImage
      mirroring=NoMirror
      resourceName=None
      resourceObject=null
      tiling=NoTile
      imageSize=<32 32>
      margin=<ltrb 0 0 0 0>
      outlineSettings {
       bUseBrushTransparency=false
       roundingType=HalfHeightRadius
       width=0
       color {colorUseRule=UseColor_Specified, specifiedColor=<rgba 0 0 0 0>}
       cornerRadii=<0 0 0 0>
      }
      tintColor {colorUseRule=UseColor_Specified, specifiedColor=<rgba 1 1 1 1>}
      uVRegion {bIsValid=false, max=<0 0>, min=<0 0>}
     }
    }
    renderTransform {angle=0, scale=<1 1>, shear=<-8 0>, translation=<0 0>}
   }
   slot {class=VerticalBoxSlot, properties {horizontalAlignment=HAlign_Center, verticalAlignment=VAlign_Center, padding=<ltrb 8 32 8 32>}}
  }
  Button_EnableVoiceChat {
   widgetClassPath=/Game/UI/Frontend/MainMenu/MainMenuButton
   bInherited=true
   bIsVariable=true
   parent=VBox_VoiceDisabledParty
   widget=Button_EnableVoiceChat
   properties {
    defaultText=Enable Voice Chat
    onButtonBaseClicked=(&self:WidgetTree.Button_EnableVoiceChat.HandleOnButtonClicked)
    onButtonBaseDoubleClicked=(&self:WidgetTree.Button_EnableVoiceChat.HandleOnButtonDoubleClicked)
    onButtonBaseHovered=(&self:WidgetTree.Button_EnableVoiceChat.HandleOnButtonHovered)
    onButtonBaseUnhovered=(&self:WidgetTree.Button_EnableVoiceChat.HandleOnButtonUnhovered)
    onSelectedChangedBase=(&self:WidgetTree.Button_EnableVoiceChat.HandleOnSelectedChanged)
    triggeringInputAction {dataTable=&InputActionDataTable, rowName=GenericAccept}
   }
   slot {class=VerticalBoxSlot, properties {horizontalAlignment=HAlign_Center, verticalAlignment=VAlign_Center}}
  }
  VBox_SocialBan {
   widgetClassPath=&Overlay
   parent=Switcher_VoiceChatState
   widget=VBox_SocialBan
   slot {class=WidgetSwitcherSlot, properties {verticalAlignment=VAlign_Center}}
  }
  bckgrnd {
   widgetClassPath=&Image
   bIsVariable=true
   parent=VBox_SocialBan
   widget=bckgrnd
   properties {brush {
     bIsDynamicallyLoaded=false
     drawAs=Image
     imageType=NoImage
     mirroring=NoMirror
     resourceName=None
     resourceObject=/Game/UI/SocialPanel/Art/Materials/M_UI_SocialUserItemBorder_UserListItem
     tiling=NoTile
     imageSize=<32 32>
     margin=<ltrb 0 0 0 0>
     outlineSettings {
      bUseBrushTransparency=false
      roundingType=HalfHeightRadius
      width=0
      color {colorUseRule=UseColor_Specified, specifiedColor=<rgba 0 0 0 0>}
      cornerRadii=<0 0 0 0>
     }
     tintColor {colorUseRule=UseColor_Foreground, specifiedColor=<rgba 0.015 0.0213 0.0417 1>}
     uVRegion {bIsValid=false, max=<0 0>, min=<0 0>}
    }}
   slot {class=OverlaySlot, properties {horizontalAlignment=HAlign_Fill, verticalAlignment=VAlign_Fill}}
  }
  VerticalBox_2 {
   widgetClassPath=&VerticalBox
   parent=VBox_SocialBan
   widget=VerticalBox_2
   slot {class=OverlaySlot, properties {horizontalAlignment=HAlign_Fill, verticalAlignment=VAlign_Center, padding=<ltrb 30 30 30 30>}}
  }
  RichText_VoiceChatDisabledHeader {
   widgetClassPath=&CommonRichTextBlock
   bIsVariable=true
   parent=VerticalBox_2
   widget=RichText_VoiceChatDisabledHeader
   properties {
    autoWrapText=true
    bOverrideDefaultStyle=true
    justification=Center
    text="You have an active social ban!"
    textTransformPolicy=ToUpper
    wrapTextAt=350
    defaultTextStyleOverride {
     overflowPolicy=Clip
     transformPolicy=None
     colorAndOpacity {colorUseRule=UseColor_Specified, specifiedColor=<rgba 0.2542 0.4969 0.7913 1>}
     font {
      bForceMonospaced=false
      bMaterialIsStencil=false
      fontMaterial=null
      fontObject=&BurbankSmall
      letterSpacing=0
      monospacedWidth=1
      size=20
      skewAmount=0
      typefaceFontName=Bold
      outlineSettings {
       bApplyOutlineToDropShadows=false
       bMiteredCorners=false
       bSeparateFillAlpha=false
       outlineMaterial=null
       outlineSize=0
       outlineColor=<rgba 0 0 0 1>
      }
     }
     highlightColor {colorUseRule=UseColor_Specified, specifiedColor=<rgba 0 0 0 1>}
     highlightShape {
      bIsDynamicallyLoaded=false
      drawAs=Image
      imageType=NoImage
      mirroring=NoMirror
      resourceName=None
      resourceObject=null
      tiling=NoTile
      imageSize=<32 32>
      margin=<ltrb 0 0 0 0>
      outlineSettings {
       bUseBrushTransparency=false
       roundingType=HalfHeightRadius
       width=0
       color {colorUseRule=UseColor_Specified, specifiedColor=<rgba 0 0 0 0>}
       cornerRadii=<0 0 0 0>
      }
      tintColor {colorUseRule=UseColor_Specified, specifiedColor=<rgba 1 1 1 1>}
      uVRegion {bIsValid=false, max=<0 0>, min=<0 0>}
     }
     selectedBackgroundColor {colorUseRule=UseColor_Foreground, specifiedColor=<rgba 1 0 1 1>}
     shadowColorAndOpacity=<rgba 0 0 0 1>
     shadowOffset=<0 0>
     strikeBrush {
      bIsDynamicallyLoaded=false
      drawAs=Image
      imageType=NoImage
      mirroring=NoMirror
      resourceName=None
      resourceObject=null
      tiling=NoTile
      imageSize=<32 32>
      margin=<ltrb 0 0 0 0>
      outlineSettings {
       bUseBrushTransparency=false
       roundingType=HalfHeightRadius
       width=0
       color {colorUseRule=UseColor_Specified, specifiedColor=<rgba 0 0 0 0>}
       cornerRadii=<0 0 0 0>
      }
      tintColor {colorUseRule=UseColor_Specified, specifiedColor=<rgba 1 1 1 1>}
      uVRegion {bIsValid=false, max=<0 0>, min=<0 0>}
     }
     underlineBrush {
      bIsDynamicallyLoaded=false
      drawAs=Image
      imageType=NoImage
      mirroring=NoMirror
      resourceName=None
      resourceObject=null
      tiling=NoTile
      imageSize=<32 32>
      margin=<ltrb 0 0 0 0>
      outlineSettings {
       bUseBrushTransparency=false
       roundingType=HalfHeightRadius
       width=0
       color {colorUseRule=UseColor_Specified, specifiedColor=<rgba 0 0 0 0>}
       cornerRadii=<0 0 0 0>
      }
      tintColor {colorUseRule=UseColor_Specified, specifiedColor=<rgba 1 1 1 1>}
      uVRegion {bIsValid=false, max=<0 0>, min=<0 0>}
     }
    }
   }
   slot {class=VerticalBoxSlot}
  }
  HorizontalBox_3 {
   widgetClassPath=&HorizontalBox
   parent=VerticalBox_2
   widget=HorizontalBox_3
   slot {class=VerticalBoxSlot, properties {horizontalAlignment=HAlign_Center, verticalAlignment=VAlign_Top}}
  }
  Image_MicDisabled {
   widgetClassPath=&Image
   bIsVariable=true
   parent=HorizontalBox_3
   widget=Image_MicDisabled
   properties {
    visibility=SelfHitTestInvisible
    brush {
     bIsDynamicallyLoaded=false
     drawAs=Image
     imageType=NoImage
     mirroring=NoMirror
     resourceName=None
     resourceObject=/Game/Athena/HUD/Art/BacchusHudIcons/Mute_Normal
     tiling=NoTile
     imageSize=<54 54>
     margin=<ltrb 0 0 0 0>
     outlineSettings {
      bUseBrushTransparency=false
      roundingType=HalfHeightRadius
      width=0
      color {colorUseRule=UseColor_Specified, specifiedColor=<rgba 0 0 0 0>}
      cornerRadii=<0 0 0 0>
     }
     tintColor {colorUseRule=UseColor_Specified, specifiedColor=<rgba 1 0.0823 0.0931 1>}
     uVRegion {bIsValid=false, max=<0 0>, min=<0 0>}
    }
   }
   slot {class=HorizontalBoxSlot, properties {padding=<ltrb 0 20 0 20>}}
  }
  RichText_SocialBanRemainingTime_VoiceTab {
   widgetClassPath=&CommonRichTextBlock
   bInherited=true
   bIsVariable=true
   parent=HorizontalBox_3
   widget=RichText_SocialBanRemainingTime_VoiceTab
   properties {
    bOverrideDefaultStyle=true
    justification=Center
    text=30 DAYS REMAINING
    textTransformPolicy=ToUpper
    defaultTextStyleOverride {
     overflowPolicy=Clip
     transformPolicy=None
     colorAndOpacity {colorUseRule=UseColor_Specified, specifiedColor=<rgba 1 1 1 1>}
     font {
      bForceMonospaced=false
      bMaterialIsStencil=false
      fontMaterial=null
      fontObject=&BurbankSmall
      letterSpacing=0
      monospacedWidth=1
      size=21
      skewAmount=0
      typefaceFontName=Bold
      outlineSettings {
       bApplyOutlineToDropShadows=false
       bMiteredCorners=false
       bSeparateFillAlpha=false
       outlineMaterial=null
       outlineSize=0
       outlineColor=<rgba 0 0 0 1>
      }
     }
     highlightColor {colorUseRule=UseColor_Specified, specifiedColor=<rgba 0 0 0 1>}
     highlightShape {
      bIsDynamicallyLoaded=false
      drawAs=Image
      imageType=NoImage
      mirroring=NoMirror
      resourceName=None
      resourceObject=null
      tiling=NoTile
      imageSize=<32 32>
      margin=<ltrb 0 0 0 0>
      outlineSettings {
       bUseBrushTransparency=false
       roundingType=HalfHeightRadius
       width=0
       color {colorUseRule=UseColor_Specified, specifiedColor=<rgba 0 0 0 0>}
       cornerRadii=<0 0 0 0>
      }
      tintColor {colorUseRule=UseColor_Specified, specifiedColor=<rgba 1 1 1 1>}
      uVRegion {bIsValid=false, max=<0 0>, min=<0 0>}
     }
     selectedBackgroundColor {colorUseRule=UseColor_Foreground, specifiedColor=<rgba 1 0 1 1>}
     shadowColorAndOpacity=<rgba 0 0 0 1>
     shadowOffset=<0 0>
     strikeBrush {
      bIsDynamicallyLoaded=false
      drawAs=Image
      imageType=NoImage
      mirroring=NoMirror
      resourceName=None
      resourceObject=null
      tiling=NoTile
      imageSize=<32 32>
      margin=<ltrb 0 0 0 0>
      outlineSettings {
       bUseBrushTransparency=false
       roundingType=HalfHeightRadius
       width=0
       color {colorUseRule=UseColor_Specified, specifiedColor=<rgba 0 0 0 0>}
       cornerRadii=<0 0 0 0>
      }
      tintColor {colorUseRule=UseColor_Specified, specifiedColor=<rgba 1 1 1 1>}
      uVRegion {bIsValid=false, max=<0 0>, min=<0 0>}
     }
     underlineBrush {
      bIsDynamicallyLoaded=false
      drawAs=Image
      imageType=NoImage
      mirroring=NoMirror
      resourceName=None
      resourceObject=null
      tiling=NoTile
      imageSize=<32 32>
      margin=<ltrb 0 0 0 0>
      outlineSettings {
       bUseBrushTransparency=false
       roundingType=HalfHeightRadius
       width=0
       color {colorUseRule=UseColor_Specified, specifiedColor=<rgba 0 0 0 0>}
       cornerRadii=<0 0 0 0>
      }
      tintColor {colorUseRule=UseColor_Specified, specifiedColor=<rgba 1 1 1 1>}
      uVRegion {bIsValid=false, max=<0 0>, min=<0 0>}
     }
    }
   }
   slot {class=HorizontalBoxSlot, properties {verticalAlignment=VAlign_Center, padding=<ltrb 10 0 0 0>}}
  }
  RichText_SocialBanDescription {
   widgetClassPath=&CommonRichTextBlock
   bInherited=true
   bIsVariable=true
   parent=VerticalBox_2
   widget=RichText_SocialBanDescription
   properties {
    bOverrideDefaultStyle=true
    justification=Center
    text=You cannot use voice chat while banned.
    wrapTextAt=350
    defaultTextStyleOverride {
     overflowPolicy=Clip
     transformPolicy=None
     colorAndOpacity {colorUseRule=UseColor_Specified, specifiedColor=<rgba 0.2542 0.4969 0.7913 1>}
     font {
      bForceMonospaced=false
      bMaterialIsStencil=false
      fontMaterial=null
      fontObject=&BurbankSmall
      letterSpacing=0
      monospacedWidth=1
      size=18
      skewAmount=0
      typefaceFontName=Bold
      outlineSettings {
       bApplyOutlineToDropShadows=false
       bMiteredCorners=false
       bSeparateFillAlpha=false
       outlineMaterial=null
       outlineSize=0
       outlineColor=<rgba 0 0 0 1>
      }
     }
     highlightColor {colorUseRule=UseColor_Specified, specifiedColor=<rgba 0 0 0 1>}
     highlightShape {
      bIsDynamicallyLoaded=false
      drawAs=Image
      imageType=NoImage
      mirroring=NoMirror
      resourceName=None
      resourceObject=null
      tiling=NoTile
      imageSize=<32 32>
      margin=<ltrb 0 0 0 0>
      outlineSettings {
       bUseBrushTransparency=false
       roundingType=HalfHeightRadius
       width=0
       color {colorUseRule=UseColor_Specified, specifiedColor=<rgba 0 0 0 0>}
       cornerRadii=<0 0 0 0>
      }
      tintColor {colorUseRule=UseColor_Specified, specifiedColor=<rgba 1 1 1 1>}
      uVRegion {bIsValid=false, max=<0 0>, min=<0 0>}
     }
     selectedBackgroundColor {colorUseRule=UseColor_Foreground, specifiedColor=<rgba 1 0 1 1>}
     shadowColorAndOpacity=<rgba 0 0 0 1>
     shadowOffset=<0 0>
     strikeBrush {
      bIsDynamicallyLoaded=false
      drawAs=Image
      imageType=NoImage
      mirroring=NoMirror
      resourceName=None
      resourceObject=null
      tiling=NoTile
      imageSize=<32 32>
      margin=<ltrb 0 0 0 0>
      outlineSettings {
       bUseBrushTransparency=false
       roundingType=HalfHeightRadius
       width=0
       color {colorUseRule=UseColor_Specified, specifiedColor=<rgba 0 0 0 0>}
       cornerRadii=<0 0 0 0>
      }
      tintColor {colorUseRule=UseColor_Specified, specifiedColor=<rgba 1 1 1 1>}
      uVRegion {bIsValid=false, max=<0 0>, min=<0 0>}
     }
     underlineBrush {
      bIsDynamicallyLoaded=false
      drawAs=Image
      imageType=NoImage
      mirroring=NoMirror
      resourceName=None
      resourceObject=null
      tiling=NoTile
      imageSize=<32 32>
      margin=<ltrb 0 0 0 0>
      outlineSettings {
       bUseBrushTransparency=false
       roundingType=HalfHeightRadius
       width=0
       color {colorUseRule=UseColor_Specified, specifiedColor=<rgba 0 0 0 0>}
       cornerRadii=<0 0 0 0>
      }
      tintColor {colorUseRule=UseColor_Specified, specifiedColor=<rgba 1 1 1 1>}
      uVRegion {bIsValid=false, max=<0 0>, min=<0 0>}
     }
    }
    renderTransform {angle=0, scale=<1 1>, shear=<-8 0>, translation=<0 0>}
   }
   slot {class=VerticalBoxSlot, properties {verticalAlignment=VAlign_Top}}
  }
  NoFriends {
   widgetClassPath=&VerticalBox
   parent=Switcher_UserLists
   widget=NoFriends
   slot {class=WidgetSwitcherSlot, properties {verticalAlignment=VAlign_Top}}
  }
  Overlay_0 {
   widgetClassPath=&Overlay
   parent=NoFriends
   widget=Overlay_0
   slot {class=VerticalBoxSlot, properties {horizontalAlignment=HAlign_Center, padding=<ltrb 0 32 0 32>}}
  }
  RightsideIcon {
   widgetClassPath=&Image
   parent=Overlay_0
   widget=RightsideIcon
   properties {
    visibility=SelfHitTestInvisible
    brush {
     bIsDynamicallyLoaded=false
     drawAs=Image
     imageType=NoImage
     mirroring=NoMirror
     resourceName=None
     resourceObject=/FortniteCoreUI/UI/SocialPanel/Art/Textures/T_UI_Icon_AddFriends_128
     tiling=NoTile
     imageSize=<128 128>
     margin=<ltrb 0 0 0 0>
     outlineSettings {
      bUseBrushTransparency=false
      roundingType=HalfHeightRadius
      width=0
      color {colorUseRule=UseColor_Specified, specifiedColor=<rgba 0 0 0 0>}
      cornerRadii=<0 0 0 0>
     }
     tintColor {colorUseRule=UseColor_Specified, specifiedColor=<rgba 1 1 1 1>}
     uVRegion {bIsValid=false, max=<0 0>, min=<0 0>}
    }
    colorAndOpacity=<rgba 1 1 1 0.3>
   }
   slot {class=OverlaySlot}
  }
  NoFriendsArrow {
   widgetClassPath=&Image
   parent=Overlay_0
   widget=NoFriendsArrow
   properties {
    visibility=SelfHitTestInvisible
    brush {
     bIsDynamicallyLoaded=false
     drawAs=Image
     imageType=NoImage
     mirroring=NoMirror
     resourceName=None
     resourceObject=/Game/UI/SocialPanel/Art/Textures/T_UI_Icon_TreeArrow_64
     tiling=NoTile
     imageSize=<64 64>
     margin=<ltrb 0 0 0 0>
     outlineSettings {
      bUseBrushTransparency=false
      roundingType=HalfHeightRadius
      width=0
      color {colorUseRule=UseColor_Specified, specifiedColor=<rgba 0 0 0 0>}
      cornerRadii=<0 0 0 0>
     }
     tintColor {colorUseRule=UseColor_Specified, specifiedColor=<rgba 0.0723 0.8714 1 1>}
     uVRegion {bIsValid=false, max=<0 0>, min=<0 0>}
    }
    renderTransform {angle=-45, scale=<1 1>, shear=<0 0>, translation=<32 -32>}
   }
   slot {class=OverlaySlot, properties {horizontalAlignment=HAlign_Right}}
  }
  RichTextBlock_0 {
   widgetClassPath=/Script/UMG.RichTextBlock
   bIsVariable=true
   parent=NoFriends
   widget=RichTextBlock_0
   properties {
    autoWrapText=true
    justification=Center
    text="NO FRIENDS YET?\r\n<B>ADD FRIENDS</>\r\nAND FILL OUT YOUR SQUAD."
    textStyleSet=&/SocialStyles
    wrapTextAt=448
    renderTransform {angle=0, scale=<1 1>, shear=<-8 0>, translation=<0 0>}
   }
   slot {class=VerticalBoxSlot, properties {horizontalAlignment=HAlign_Center, padding=<ltrb 0 0 0 32>}}
  }
 ]
}
