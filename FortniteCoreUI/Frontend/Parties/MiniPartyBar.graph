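# MiniPartyBar is a frontend party strip showing the local player's banner, a separator and three remote member banners.
# A transparent button over the local banner pushes a cached, lazily created TeamMemberManage modal unless it is already active.
# Construct collapses remote banners and separator when party features are disabled.
# Helper functions map ShowPartyBar feature state to bar visibility while parties are enabled, and build a banner tooltip for non-gamepad input.
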WidgetBlueprint /FortniteCoreUI/Frontend/Parties/MiniPartyBar
graphs [
 EventGraph {nodes [
   "On Button Clicked (ButtonLocalPlayer)"!() -> (OutputDelegate, then, Button:Common Button Legacy Object Reference) {is=title K2Node_ComponentBoundEvent}
   > n1 = Is Valid(InputObject=Get NewLocalPlayerWidget() -> (NewLocalPlayerWidget:Team Member Manage Object Reference))
   n2 = Create Widget!(
    execute:Exec=@n1.Is Not Valid
    Class:User Widget Class Reference=/Game/UI/Frontend/Parties/TeamMemberManage.TeamMemberManage_C
    OwningPlayer:Player Controller Object Reference=GetOwningPlayer()
   ) -> (then:Exec, ReturnValue:Team Member Manage Object Reference)
   > n3 = Set NewLocalPlayerWidget(NewLocalPlayerWidget:Team Member Manage Object Reference=@n2) -> (Output_Get:Team Member Manage Object Reference)
   SetRepresentingTeamMember!(execute=[@n4.else, @n3.then], self:Fort Team Member Manage Object Reference=Get NewLocalPlayerWidget(), LocalTeamMemberIndex:int=0) -> (then)
   > TryPushWidgetAsModalPanel!(execute, PushWidget:User Widget Object Reference=Get NewLocalPlayerWidget()) -> (then)
   n4 = Branch(
    execute=@n1.Is Valid
    Condition=IsActivated!(self:Common Activatable Widget Object Reference=Get NewLocalPlayerWidget()) -> (ReturnValue:bool)
   )
   Event Construct()
   > n5 = Branch(Condition=ArePartyFeaturesEnabled!(self:Fort UIData Configuration Context Object Reference=Get FortUIDataConfigurationContext!() -> (ReturnValue:Fort UIData Configuration Context Object Reference) {is=title K2Node_GetBlueprintContextEx}) -> (ReturnValue:bool))
   n6 = For Each Loop(
    Exec=@n5.else
    Array:Array of Widget Object References=Make Array!(
     "[0]":Widget Object Reference=Get RemotePlayer1() -> (RemotePlayer1:Team Member Banner Object Reference)
     "[1]":Widget Object Reference=Get RemotePlayer2() -> (RemotePlayer2:Team Member Banner Object Reference)
     "[2]":Widget Object Reference=Get RemotePlayer3() -> (RemotePlayer3:Team Member Banner Object Reference)
    ) -> (Array:Array of Widget Object References)
   ) -> (Array Element:Widget Object Reference)
   > SetVisibility(self=@n6, InVisibility=Collapsed)
   SetVisibility(execute=@n6.Completed, self=Get LineSeparator() -> (LineSeparator:Image Object Reference), InVisibility=Collapsed)
  ]}
 Get Banners Tooltip Widget {nodes [
   Get Banners Tooltip Widget() {is=title K2Node_FunctionEntry}
   n1 = Get CommonInputSubsystem!() -> (ReturnValue:Common Input Object Reference) {is=title K2Node_GetBlueprintContextEx}
   > n2 = Branch(Condition="Equal (Enum)"(A=GetCurrentInputType(self=@n1), B=Gamepad))
   n3 = Create Basic Tooltip!(
    execute=@n2.else
    Owning Player:Player Controller Object Reference=GetOwningPlayer()
    Body Text:Text=MakeLiteralText(Value="LOCTEXT(\"Fly your customized banner with pride!\")")
    Header Text:Text=MakeLiteralText(Value="LOCTEXT(\"HOMEBASE BANNER\")")
    Icon Brush:Slate Brush Asset Object Reference
    Header Style:Common Text Style Class Reference
    Body Style:Common Text Style Class Reference
   ) -> (then, Output:User Widget Object Reference)
   > Return Node!(execute, ReturnValue:Widget Object Reference=@n3)
  ]}
 HandlePartybarUIFeatureChanged {nodes [
   n1 = HandlePartybarUIFeatureChanged!() -> (then, Feature:EFortUIFeature Enum=ShowHome, FeatureState:EFortUIFeatureState Enum=Enabled) {is=title K2Node_FunctionEntry}
   n2 = Get FortUIDataConfigurationContext()
   > Branch(Condition=AND Boolean(A="Equal (Enum)"(A:EFortUIFeature Enum=@n1, B:EFortUIFeature Enum=ShowPartyBar), B=ArePartyFeaturesEnabled(self=@n2)))
   > SetVisibility(InVisibility=Select!(
     Enabled:ESlateVisibility Enum=SelfHitTestInvisible
     Disabled:ESlateVisibility Enum=Collapsed
     Hidden:ESlateVisibility Enum=Collapsed
     Invalid:ESlateVisibility Enum=Collapsed
     Index:EFortUIFeatureState Enum=@n1.FeatureState
    ) -> (ReturnValue))
  ]}
]
variables [NewLocalPlayerWidget {type=TeamMemberManage_C, default=null}]
widgets {
 info {inheritedWidgetCount=0, parentClass=&CommonUserWidget, rootWidgetClass=&CommonBorder, widgetCount=9}
 widgets [
  CommonBorder_1 {
   widgetClassPath=&CommonBorder
   widget=CommonBorder_1
   properties {
    verticalAlignment=VAlign_Center
    background {
     bIsDynamicallyLoaded=false
     drawAs=NoDrawType
     imageType=NoImage
     mirroring=NoMirror
     resourceName=None
     resourceObject=null
     tiling=NoTile
     imageSize=<32 32>
     margin=<ltrb 0.1 0.45 0.9 0.5>
     outlineSettings {
      bUseBrushTransparency=false
      roundingType=HalfHeightRadius
      width=0
      color {colorUseRule=UseColor_Specified, specifiedColor=<rgba 0 0 0 0>}
      cornerRadii=<0 0 0 0>
     }
     tintColor {colorUseRule=UseColor_Specified, specifiedColor=<rgba 1 1 1 1>}
     uVRegion {bIsValid=false, max=<0 0>, min=<0 0>}
    }
    padding=<ltrb 0 0 0 0>
   }
  }
  HorizontalBox_1 {
   widgetClassPath=&HorizontalBox
   parent=CommonBorder_1
   widget=HorizontalBox_1
   slot {class=BorderSlot, properties {verticalAlignment=VAlign_Center, padding=<ltrb 0 0 0 0>}}
  }
  Overlay_0 {
   widgetClassPath=&Overlay
   parent=HorizontalBox_1
   widget=Overlay_0
   slot {class=HorizontalBoxSlot, properties {horizontalAlignment=HAlign_Center, verticalAlignment=VAlign_Center}}
  }
  LocalPlayer {
   widgetClassPath=&/TeamMemberBanner
   bIsVariable=true
   parent=Overlay_0
   widget=LocalPlayer
   properties {
    onButtonBaseClicked=(&self:WidgetTree.LocalPlayer.HandleOnButtonClicked)
    onButtonBaseDoubleClicked=(&self:WidgetTree.LocalPlayer.HandleOnButtonDoubleClicked)
    onButtonBaseHovered=(&self:WidgetTree.LocalPlayer.HandleOnButtonHovered)
    onButtonBaseUnhovered=(&self:WidgetTree.LocalPlayer.HandleOnButtonUnhovered)
    onSelectedChangedBase=(&self:WidgetTree.LocalPlayer.HandleOnSelectedChanged)
    representedMemberIndex=0
    hoveredSlateSoundOverride {resourceObject=/Game/Sounds/Fort_UI_Sounds/FrontEnd/DefaultSounds/PressedButtons/MSS_UI_Button_Hover_Default}
    pressedSlateSoundOverride {resourceObject=/Game/Sounds/Fort_UI_Sounds/FrontEnd/DefaultSounds/PressedButtons/MSS_UI_Button_Pressed_Default}
   }
   slot {class=OverlaySlot}
  }
  ButtonLocalPlayer {
   widgetClassPath=&IconTextButton
   bIsVariable=true
   parent=Overlay_0
   widget=ButtonLocalPlayer
   properties {
    buttonText=""
    onButtonBaseClicked=(&self:WidgetTree.ButtonLocalPlayer.HandleOnButtonClicked)
    onButtonBaseDoubleClicked=(&self:WidgetTree.ButtonLocalPlayer.HandleOnButtonDoubleClicked)
    onButtonBaseHovered=(&self:WidgetTree.ButtonLocalPlayer.HandleOnButtonHovered)
    onButtonBaseUnhovered=(&self:WidgetTree.ButtonLocalPlayer.HandleOnButtonUnhovered)
    onSelectedChangedBase=(&self:WidgetTree.ButtonLocalPlayer.HandleOnSelectedChanged)
    style=/FortUILibrary/Foundation/Buttons/ButtonStyles/ButtonStyle-MediumTransparentNoCues.ButtonStyle-MediumTransparentNoCues_C
    visibility=Visible
   }
   slot {class=OverlaySlot, properties {horizontalAlignment=HAlign_Fill, verticalAlignment=VAlign_Fill}}
  }
  LineSeparator {
   widgetClassPath=&Image
   bIsVariable=true
   parent=HorizontalBox_1
   widget=LineSeparator
   properties {
    brush {
     bIsDynamicallyLoaded=false
     drawAs=Image
     imageType=NoImage
     mirroring=NoMirror
     resourceName=None
     resourceObject=null
     tiling=NoTile
     imageSize=<2 38>
     margin=<ltrb 0 0 0 0>
     outlineSettings {
      bUseBrushTransparency=false
      roundingType=HalfHeightRadius
      width=0
      color {colorUseRule=UseColor_Specified, specifiedColor=<rgba 0 0 0 0>}
      cornerRadii=<0 0 0 0>
     }
     tintColor {colorUseRule=UseColor_Specified, specifiedColor=<rgba 1 1 1 1>}
     uVRegion {bIsValid=false, max=<0 0>, min=<0 0>}
    }
    colorAndOpacity=<rgba 1 1 1 0.4>
   }
   slot {class=HorizontalBoxSlot, properties {verticalAlignment=VAlign_Top, padding=<ltrb 6 0 0 0>}}
  }
  RemotePlayer1 {
   widgetClassPath=&TeamMemberBanner
   bIsVariable=true
   parent=HorizontalBox_1
   widget=RemotePlayer1
   properties {
    onButtonBaseClicked=(&self:WidgetTree.RemotePlayer1.HandleOnButtonClicked)
    onButtonBaseDoubleClicked=(&self:WidgetTree.RemotePlayer1.HandleOnButtonDoubleClicked)
    onButtonBaseHovered=(&self:WidgetTree.RemotePlayer1.HandleOnButtonHovered)
    onButtonBaseUnhovered=(&self:WidgetTree.RemotePlayer1.HandleOnButtonUnhovered)
    onSelectedChangedBase=(&self:WidgetTree.RemotePlayer1.HandleOnSelectedChanged)
    representedMemberIndex=1
    hoveredSlateSoundOverride {resourceObject=&MSS_UI_Button_Hover_Default}
    pressedSlateSoundOverride {resourceObject=&MSS_UI_Button_Pressed_Default}
   }
   slot {class=HorizontalBoxSlot, properties {padding=<ltrb 6 0 8 0>}}
  }
  RemotePlayer2 {
   widgetClassPath=&TeamMemberBanner
   bIsVariable=true
   parent=HorizontalBox_1
   widget=RemotePlayer2
   properties {
    onButtonBaseClicked=(&self:WidgetTree.RemotePlayer2.HandleOnButtonClicked)
    onButtonBaseDoubleClicked=(&self:WidgetTree.RemotePlayer2.HandleOnButtonDoubleClicked)
    onButtonBaseHovered=(&self:WidgetTree.RemotePlayer2.HandleOnButtonHovered)
    onButtonBaseUnhovered=(&self:WidgetTree.RemotePlayer2.HandleOnButtonUnhovered)
    onSelectedChangedBase=(&self:WidgetTree.RemotePlayer2.HandleOnSelectedChanged)
    representedMemberIndex=2
    hoveredSlateSoundOverride {resourceObject=&MSS_UI_Button_Hover_Default}
    pressedSlateSoundOverride {resourceObject=&MSS_UI_Button_Pressed_Default}
   }
   slot {class=HorizontalBoxSlot, properties {padding=<ltrb 0 0 8 0>}}
  }
  RemotePlayer3 {
   widgetClassPath=&TeamMemberBanner
   bIsVariable=true
   parent=HorizontalBox_1
   widget=RemotePlayer3
   properties {
    onButtonBaseClicked=(&self:WidgetTree.RemotePlayer3.HandleOnButtonClicked)
    onButtonBaseDoubleClicked=(&self:WidgetTree.RemotePlayer3.HandleOnButtonDoubleClicked)
    onButtonBaseHovered=(&self:WidgetTree.RemotePlayer3.HandleOnButtonHovered)
    onButtonBaseUnhovered=(&self:WidgetTree.RemotePlayer3.HandleOnButtonUnhovered)
    onSelectedChangedBase=(&self:WidgetTree.RemotePlayer3.HandleOnSelectedChanged)
    representedMemberIndex=3
    hoveredSlateSoundOverride {resourceObject=&MSS_UI_Button_Hover_Default}
    pressedSlateSoundOverride {resourceObject=&MSS_UI_Button_Pressed_Default}
   }
   slot {class=HorizontalBoxSlot}
  }
 ]
}
