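# Sound cue for the front-end social panel UI sound in the Fort_UI sound class.
# It plays the mission objective pop-up wave through a modulator that pitches it up by a small random range on each play.
# Overridden concurrency allows one instance at a time, stopping the oldest when it retriggers.
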SoundCue /Game/Sounds/Fort_UI_Sounds/FrontEnd/New_FrontEnd/Fort_UI_SocialPanel_Cue
root=Modulator_1
nodes [
 {id=Modulator_1, children [WavePlayer_3], properties {pitchMax=1.35, pitchMin=1.25}}
 {id=WavePlayer_3, properties {soundWaveAssetPtr=&/UI_FN_Mission_Obj_PopUp_01}}
]
properties {
 bOverrideConcurrency=true
 maxDistance=2097152
 soundClassObject=&Fort_UI
 volumeMultiplier=1
 concurrencyOverrides {maxCount=1, resolutionRule=StopOldest}
}
