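# Social panel widget based on FortSocialPanel: a fixed-width column with top bar and user lists over a blurred lightbox.
# One OnOpened animation plays forward for intro and reversed for outro, with sound; finish callbacks end each transition, and Destruct completes an interrupted one.
# The top bar and a mouse-down handler call HandleSocialPanelClose unless a slide-out menu is open.
# ClosePanel pops it from the modal layer.
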WidgetBlueprint /FortniteCoreUI/UI/SocialPanel/SocialPanel
animations [OnOpened {
  binding_count=6
  display_label=""
  length_seconds=0.2
  object=&self:OnOpened
  sequence {bindings [
    SocialPanelWidthSB {
     id=#1
     possessed_class=SizeBox
     tracks [
      {
       type=MovieScene2DTransformTrack
       sections [{
         type=MovieScene2DTransformSection
         channels [
          Translation.X {type=MovieSceneScriptingFloatChannel, keys [[0,-256],[4,0]]}
          Translation.Y {type=MovieSceneScriptingFloatChannel, keys [[0,64],[4,0]]}
          Angle {type=MovieSceneScriptingFloatChannel}
          Scale.X {type=MovieSceneScriptingFloatChannel, keys [[0,0.375],[4,1]]}
          Scale.Y {type=MovieSceneScriptingFloatChannel, keys [[0,0.875],[4,1]]}
          Shear.X {type=MovieSceneScriptingFloatChannel}
          Shear.Y {type=MovieSceneScriptingFloatChannel, keys [[0,-16],[4,0]]}
         ]
        }]
      }
      {
       type=MovieSceneFloatTrack
       display_name=Render Opacity
       sections [{type=MovieSceneFloatSection, start_frame=0, channels [None {type=MovieSceneScriptingFloatChannel, keys [[0,0],[2,1]]}]}]
      }
     ]
    }
    PanelView_UserLists {
     id=#2
     possessed_class=SocialPanelView_UserLists_C
     tracks [{
       type=MovieSceneColorTrack
       display_name=Color and Opacity
       sections [{
         type=MovieSceneColorSection
         channels [
          Color.R {type=MovieSceneScriptingFloatChannel, keys [[0,0],[4,1]]}
          Color.G {type=MovieSceneScriptingFloatChannel, keys [[0,0.0044],[4,1]]}
          Color.B {type=MovieSceneScriptingFloatChannel, keys [[0,0.0319],[4,1]]}
          Color.A {type=MovieSceneScriptingFloatChannel}
         ]
        }]
      }]
    }
    BackgroundBlur_0 {
     id=#3
     possessed_class=BackgroundBlur
     tracks [{
       type=MovieSceneFloatTrack
       display_name=Blur Strength
       sections [{type=MovieSceneFloatSection, channels [None {type=MovieSceneScriptingFloatChannel, keys [[0,0],[2,4],[4,4]]}]}]
      }]
    }
    SocialPanel_TopBarExtension {
     id=#4
     possessed_class=SocialPanel_TopBarExtension_C
     tracks [{
       type=MovieSceneColorTrack
       display_name=Color and Opacity
       sections [{
         type=MovieSceneColorSection
         channels [
          Color.R {type=MovieSceneScriptingFloatChannel, keys [[0,0],[4,1]]}
          Color.G {type=MovieSceneScriptingFloatChannel, keys [[0,0.0044],[4,1]]}
          Color.B {type=MovieSceneScriptingFloatChannel, keys [[0,0.0319],[4,1]]}
          Color.A {type=MovieSceneScriptingFloatChannel}
         ]
        }]
      }]
    }
    Lightbox {
     id=#5
     possessed_class=Lightbox_C
     tracks [{
       type=MovieSceneFloatTrack
       display_name=Render Opacity
       sections [{type=MovieSceneFloatSection, channels [None {type=MovieSceneScriptingFloatChannel, keys [[0,0],[2,1]]}]}]
      }]
    }
   ]}
 }]
graphs [
 ClosePanelAfterAnimation {nodes [
   ClosePanelAfterAnimation() {is=title K2Node_FunctionEntry}
   > PopActivatablePanelInModalLayer!(execute, Panel:Common Activatable Panel Legacy Object Reference=Self-Reference()) -> (then)
  ]}
 EventGraph {nodes [
   "On Button Clicked (SocialPanel_TopBarExtension)"!() -> (OutputDelegate, then, Button:Common Button Legacy Object Reference) {is=title K2Node_ComponentBoundEvent}
   > n1 = Branch(Condition=IsAnySlideOutMenuOpen!(self:Fort Social Panel Object Reference) -> (ReturnValue:bool))
   HandleSocialPanelClose(execute=@n1.else, self:Fort Social Panel Object Reference)
   n2 = HandleIntroEnded() {is=title K2Node_CustomEvent}
   > UnbindAllFromAnimationFinished!(
    execute
    self:User Widget Object Reference
    Animation:Widget Animation Object Reference=Get OnOpened() -> (OnOpened:Widget Animation Object Reference)
   ) -> (then)
   > EndIntro(self:Common Activatable Panel Legacy Object Reference)
   n3 = HandleOutroEnded() {is=title K2Node_CustomEvent}
   > UnbindAllFromAnimationFinished(Animation=Get OnOpened())
   > EndOutro(self:Common Activatable Panel Legacy Object Reference)
   Event Destruct()
   > Branch(Condition=IsAnimationPlaying(InAnimation=Get OnOpened()))
   > UnbindAllFromAnimationFinished(Animation=Get OnOpened())
   > n4 = IsAnimationPlayingForward!(execute, self:User Widget Object Reference, InAnimation:Widget Animation Object Reference=Get OnOpened()) -> (then, ReturnValue:bool)
   > StopAllAnimations(self:User Widget Object Reference)
   > n5 = Branch(Condition=@n4)
   > EndIntro()
   EndOutro(execute=@n5.else)
   Event ClosePanel() {is=title K2Node_Event}
   > ClosePanelAfterAnimation() {is=title K2Node_CallFunction}
   Event OnBeginOutro() {is=title K2Node_Event}
   n6 = Get OnOpened()
   > BindToAnimationFinished(Animation=@n6, Delegate=@n3)
   > PlayAnimation(InAnimation=@n6, PlayMode=Reverse)
   > Play Panel Animation Sound()
   Event OnBeginIntro() {is=title K2Node_Event}
   n7 = Get OnOpened()
   > BindToAnimationFinished(Animation=@n7, Delegate=@n2)
   > PlayAnimation(InAnimation=@n7)
   > Play Panel Animation Sound()
  ]}
 OnMouseButtonDown_0 {nodes [
   OnMouseButtonDown_0!() -> (then, MyGeometry:Geometry Structure, MouseEvent:"Pointer Event Structure (by ref)") {is=title K2Node_FunctionEntry}
   n1 = IsActivated!(self:Common Activatable Widget Object Reference) -> (ReturnValue:bool)
   > n2 = Branch(Condition=OR Boolean(A=IsAnySlideOutMenuOpen(), B=NOT Boolean(A=@n1)))
   > Return Node!(execute, ReturnValue:Event Reply Structure=Unhandled())
   HandleSocialPanelClose(execute=@n2.else)
   > Return Node(ReturnValue=Handled())
  ]}
 Play Panel Animation Sound {nodes [
   Play Panel Animation Sound() {is=title K2Node_FunctionEntry}
   > Branch(Condition=IsAnimationPlaying(InAnimation=Get OnOpened()))
   > PlaySound2D(Sound=/Game/Sounds/Fort_UI_Sounds/FrontEnd/New_FrontEnd/Fort_UI_SocialPanel_Cue)
  ]}
]
variables []
widgets {
 info {inheritedWidgetCount=1, parentClass=/Script/FortniteUI.FortSocialPanel, rootWidgetClass=&Overlay, widgetCount=15}
 widgets [
  Overlay_1 {widgetClassPath=&Overlay, widget=Overlay_1}
  Lightbox {
   widgetClassPath=/Game/UI/Foundation/Lightbox/Lightbox
   bIsVariable=true
   parent=Overlay_1
   widget=Lightbox
   properties {renderOpacity=0.5, showWindowControlButtons=false, visibility=HitTestInvisible}
   slot {class=OverlaySlot, properties {horizontalAlignment=HAlign_Fill, verticalAlignment=VAlign_Fill}}
  }
  BackgroundBlur_0 {
   widgetClassPath=&BackgroundBlur
   parent=Overlay_1
   widget=BackgroundBlur_0
   properties {blurStrength=8, navigation=&self:WidgetTree.BackgroundBlur_0.WidgetNavigation_0}
   slot {class=OverlaySlot, properties {horizontalAlignment=HAlign_Fill, verticalAlignment=VAlign_Fill}}
  }
  SafeZone_0 {
   widgetClassPath=&SafeZone
   bIsVariable=true
   parent=BackgroundBlur_0
   widget=SafeZone_0
   slot {class=BackgroundBlurSlot, properties {padding=<ltrb 0 0 0 0>}}
  }
  Overlay_0 {widgetClassPath=&Overlay, parent=SafeZone_0, widget=Overlay_0, slot {class=SafeZoneSlot}}
  Image_Back {
   widgetClassPath=&Image
   bIsVariable=true
   parent=Overlay_0
   widget=Image_Back
   properties {colorAndOpacity=<rgba 1 1 1 0>}
   slot {class=OverlaySlot, properties {horizontalAlignment=HAlign_Fill, verticalAlignment=VAlign_Fill}}
  }
  SocialPanelWidthSB {
   widgetClassPath=&SizeBox
   parent=Overlay_0
   widget=SocialPanelWidthSB
   properties {bOverride_WidthOverride=true, widthOverride=544}
   slot {class=OverlaySlot, properties {verticalAlignment=VAlign_Fill}}
  }
  Overlay_7 {widgetClassPath=&Overlay, parent=SocialPanelWidthSB, widget=Overlay_7, slot {class=SizeBoxSlot}}
  VerticalBox_2 {
   widgetClassPath=&VerticalBox
   parent=Overlay_7
   widget=VerticalBox_2
   slot {class=OverlaySlot, properties {horizontalAlignment=HAlign_Fill, verticalAlignment=VAlign_Fill, padding=<ltrb -112 -64 0 -64>}}
  }
  Border_0 {
   widgetClassPath=&Border
   parent=VerticalBox_2
   widget=Border_0
   properties {brushColor=<rgba 0 0.0044 0.0319 1>, padding=<ltrb 0 0 0 0>}
   slot {class=VerticalBoxSlot}
  }
  Spacer_0 {
   widgetClassPath=&Spacer
   parent=Border_0
   widget=Spacer_0
   properties {size=<136 136>}
   slot {class=BorderSlot, properties {padding=<ltrb 0 0 0 0>}}
  }
  Border_1 {
   widgetClassPath=&Border
   parent=VerticalBox_2
   widget=Border_1
   properties {
    background {
     bIsDynamicallyLoaded=false
     drawAs=Image
     imageType=NoImage
     mirroring=NoMirror
     resourceName=None
     resourceObject=/FortniteCoreUI/UI/SocialPanel/Art/Materials/M_UI_SocialPanelBackground
     tiling=NoTile
     imageSize=<32 32>
     margin=<ltrb 0 0 0 0>
     outlineSettings {
      bUseBrushTransparency=false
      roundingType=HalfHeightRadius
      width=0
      color {colorUseRule=UseColor_Specified, specifiedColor=<rgba 0 0 0 0>}
      cornerRadii=<0 0 0 0>
     }
     tintColor {colorUseRule=UseColor_Specified, specifiedColor=<rgba 1 1 1 1>}
     uVRegion {bIsValid=false, max=<0 0>, min=<0 0>}
    }
    padding=<ltrb 0 0 0 0>
   }
   slot {class=VerticalBoxSlot, properties {size {sizeRule=Fill, value=1}}}
  }
  VerticalBox_0 {
   widgetClassPath=&VerticalBox
   parent=Overlay_7
   widget=VerticalBox_0
   slot {class=OverlaySlot, properties {horizontalAlignment=HAlign_Fill, verticalAlignment=VAlign_Fill}}
  }
  SocialPanel_TopBarExtension {
   widgetClassPath=&/SocialPanel_TopBarExtension
   bIsVariable=true
   parent=VerticalBox_0
   widget=SocialPanel_TopBarExtension
   properties {
    bHideInputActionWithKeyboard=true
    clickMethod=MouseDown
    onButtonBaseClicked=(&self:WidgetTree.SocialPanel_TopBarExtension.HandleOnButtonClicked)
    onButtonBaseDoubleClicked=(&self:WidgetTree.SocialPanel_TopBarExtension.HandleOnButtonDoubleClicked)
    onButtonBaseHovered=(&self:WidgetTree.SocialPanel_TopBarExtension.HandleOnButtonHovered)
    onButtonBaseUnhovered=(&self:WidgetTree.SocialPanel_TopBarExtension.HandleOnButtonUnhovered)
    onSelectedChangedBase=(&self:WidgetTree.SocialPanel_TopBarExtension.HandleOnSelectedChanged)
    pressMethod=ButtonPress
    renderTransformPivot=<0 0.5>
    triggeringInputAction {dataTable=&InputActionDataTable, rowName=SocialPanelLobbyOpen}
   }
   slot {class=VerticalBoxSlot}
  }
  PanelView_UserLists {
   widgetClassPath=&/SocialPanelView_UserLists
   bInherited=true
   bIsVariable=true
   parent=VerticalBox_0
   widget=PanelView_UserLists
   properties {bIsFocusable=true, visibility=Visible, renderTransformPivot=<0.5 0>}
   slot {class=VerticalBoxSlot, properties {size {sizeRule=Fill, value=1}}}
  }
 ]
}
