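# LocalUserSettingSection is a social panel widget, derived from FortLocalUserSettingsSection, that lays out a titled, apparently collapsible group of local user settings.
# A transparent border wraps a vertical box holding the inherited TitleButton above EntryBox_Options.
# EntryBox_Options is a vertical DynamicEntryBox that spawns LocalUserSettingButton entries and starts collapsed.
# The event graph is empty, so behavior comes from the parent class.
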WidgetBlueprint /Game/UI/SocialPanel/LocalUserSettingSection
graphs [EventGraph {links [], nodes []}]
variables []
widgets {
 info {inheritedWidgetCount=2, parentClass=/Script/FortniteUI.FortLocalUserSettingsSection, rootWidgetClass=&Border, widgetCount=4}
 widgets [
  Border_0 {
   widgetClassPath=&Border
   widget=Border_0
   properties {
    background {
     bIsDynamicallyLoaded=false
     drawAs=Image
     imageType=NoImage
     mirroring=NoMirror
     resourceName=None
     resourceObject=null
     tiling=NoTile
     imageSize=<32 32>
     margin=<ltrb 0 0 0 0>
     outlineSettings {
      bUseBrushTransparency=false
      roundingType=HalfHeightRadius
      width=0
      color {colorUseRule=UseColor_Specified, specifiedColor=<rgba 0 0 0 0>}
      cornerRadii=<0 0 0 0>
     }
     tintColor {colorUseRule=UseColor_Specified, specifiedColor=<rgba 1 1 1 0>}
     uVRegion {bIsValid=false, max=<0 0>, min=<0 0>}
    }
    padding=<ltrb 0 0 0 0>
   }
  }
  VerticalBox_0 {widgetClassPath=&VerticalBox, parent=Border_0, widget=VerticalBox_0, slot {class=BorderSlot, properties {padding=<ltrb 0 0 0 0>}}}
  TitleButton {
   widgetClassPath=&/LocalUserSettingTitle
   bInherited=true
   bIsVariable=true
   parent=VerticalBox_0
   widget=TitleButton
   properties {
    onButtonBaseClicked=(&self:WidgetTree.TitleButton.HandleOnButtonClicked)
    onButtonBaseDoubleClicked=(&self:WidgetTree.TitleButton.HandleOnButtonDoubleClicked)
    onButtonBaseHovered=(&self:WidgetTree.TitleButton.HandleOnButtonHovered)
    onButtonBaseUnhovered=(&self:WidgetTree.TitleButton.HandleOnButtonUnhovered)
    onSelectedChangedBase=(&self:WidgetTree.TitleButton.HandleOnSelectedChanged)
   }
   slot {class=VerticalBoxSlot, properties {padding=<ltrb 0 0 0 4>}}
  }
  EntryBox_Options {
   widgetClassPath=&DynamicEntryBox
   bInherited=true
   bIsVariable=true
   parent=VerticalBox_0
   widget=EntryBox_Options
   properties {
    entryBoxType=Vertical
    entryWidgetClass=&/LocalUserSettingButton.LocalUserSettingButton_C
    visibility=Collapsed
    entrySpacing=<0 4>
   }
   slot {class=VerticalBoxSlot, properties {padding=<ltrb 0 0 0 4>}}
  }
 ]
}
